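# Editor utility actor that lets the level viewport camera jump to a player-view position, apparently for checking scale in levels.
# Ticking previewPlayerView works as a button: the construction script resets it to false and calls TeleportToHead.
# TeleportToHead moves the viewport camera to the SceneCameraLocation component's world location, keeping pitch and yaw but dropping roll.
# The isCrouched variable is declared but unused in these graphs.
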EditorUtilityBlueprint /Game/Blueprints/Blutilities/ForLDs/BPs/TestScaleEditor/TestScaleEditorActor
graphs [
 EventGraph {links [], nodes []}
 TeleportToHead {nodes [
   TeleportToHead() {is=title K2Node_FunctionEntry}
   n1 = Get World Transform!(self:Scene Component Object Reference=Get SceneCameraLocation() -> (SceneCameraLocation:Scene Component Object Reference)) -> (ReturnValue_Location:Vector, ReturnValue_Rotation:Rotator, ReturnValue_Scale:Vector) {is=title K2Node_CallFunction}
   n2 = BreakRotator(InRot=@n1.ReturnValue_Rotation)
   > SetLevelViewportCameraInfo!(
    execute
    self:Unreal Editor Subsystem Object Reference=Get UnrealEditorSubsystem()
    CameraLocation:Vector=@n1
    CameraRotation:Rotator=MakeRotator(Pitch=@n2.Pitch, Yaw=@n2.Yaw)
   ) -> (then)
  ]}
 UserConstructionScript {nodes [
   Construction Script()
   > Branch(Condition=Get previewPlayerView() -> (previewPlayerView:bool))
   > Set previewPlayerView(previewPlayerView:bool=false) -> (Output_Get:bool)
   > TeleportToHead() {is=title K2Node_CallFunction}
  ]}
]
variables [
 isCrouched {type=boolean}
 previewPlayerView {type=boolean}
]
