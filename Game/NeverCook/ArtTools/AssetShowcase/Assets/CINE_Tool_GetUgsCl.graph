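# Editor utility that records the workspace's current UnrealGameSync changelist for the asset showcase tools.
# On Run, it loads the SHOWCASE_Globals save slot as SG_ShowcaseRel, or creates that save object if none exists.
# Embedded Python reads CurrentChangeNumber from .ugs/state.json in the project's parent folder, falling back to -1 on a file error.
# When the script succeeds, the changelist is written to the save object and saved back to the slot.
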EditorUtilityBlueprint /Game/NeverCook/ArtTools/AssetShowcase/Assets/CINE_Tool_GetUgsCl
graphs [EventGraph {nodes [
   Event Run()
   > n1 = DoesSaveGameExist!(execute, SlotName:String=SHOWCASE_Globals, UserIndex:int=0) -> (then, ReturnValue:bool)
   > n2 = Branch(Condition=@n1)
   > n3 = LoadGameFromSlot!(execute, SlotName:String=SHOWCASE_Globals, UserIndex:int=0) -> (then, ReturnValue:Save Game Object Reference)
   > n4 = Cast To SG_ShowcaseRel!(execute, Object=@n3) -> (then, CastFailed, AsSG Showcase Rel:SG Showcase Rel Object Reference) {is=title K2Node_DynamicCast}
   > n5 = Set SaveData(SaveData:SG Showcase Rel Object Reference=@n4) -> (Output_Get:SG Showcase Rel Object Reference)
   n6 = Split(SourceString=Split(SourceString=GetProjectDirectory!() -> (ReturnValue:String), InStr=/, SearchDir=FromEnd), InStr=/, SearchDir=FromEnd)
   n7 = Execute Python Script!(
    execute=[@n5.then, @n10.then]
    PythonScript="import json\r\nimport unreal\r\n\r\nunreal.log(\"CineTools - Fetching Current UGS Changelist\")\r\ntry:\r\n    with open (path, \"r\") as openfile:\r\n        data = json.load(openfile)\r\n\r\n    change = data.get(\"CurrentChangeNumber\")\r\n    unreal.log(\"CineTools - UGS Changelist Found - \" + str(change))\r\n\r\nexcept IOError:\r\n    unreal.log_warning(\"CineTools - Unable to determine Current UGS Changelist\")\r\n    change = -1"
    path:String="To String (Text)"(InText=Format Text!(Format="LOCTEXT(\"{BUILD}/.ugs/state.json\")", BUILD:String=@n6) -> (Result))
   ) -> (then, ReturnValue, change:int) {inputs [path], outputs [change]}
   > Branch(Condition=@n7)
   n8 = Get SaveData() -> (SaveData:SG Showcase Rel Object Reference)
   > Set changelist!(execute, changelist:int=@n7.change, self:SG Showcase Rel Object Reference=@n8) -> (then, Output_Get:int=0)
   > SaveGameToSlot!(execute, SaveGameObject:Save Game Object Reference=@n8, SlotName:String=SHOWCASE_Globals, UserIndex:int=0) -> (then, ReturnValue:bool)
   n9 = CreateSaveGameObject!(execute=@n2.else, SaveGameClass:Save Game Class Reference=/Game/NeverCook/ArtTools/AssetShowcase/Data/SG_ShowcaseRel.SG_ShowcaseRel_C) -> (then, ReturnValue:SG Showcase Rel Object Reference)
   > n10 = Set SaveData(SaveData=@n9)
  ]}]
variables [SaveData {type=SG_ShowcaseRel_C, default=null}]
