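# LS_Showcase is an editor-only Level Sequence among the Movie Render Queue assets of the AssetShowcase art tool, under NeverCook.
# It binds no objects and holds just two event tracks, TriggerEvents and RepeaterEvents, at 30 fps.
# The trigger section shows no keyed events, and the short repeater section at the start shows no bound event, so it appears to be a near-empty shell.
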LevelSequence /Game/NeverCook/ArtTools/AssetShowcase/MRQAssets/LS_Showcase
bindings []
display_rate {denominator=1, numerator=30}
tick_resolution {denominator=1, numerator=24000}
tracks [
 {
  type=MovieSceneEventTrack
  display_name=TriggerEvents
  sections [{type=MovieSceneEventTriggerSection, channels [None {type=MovieSceneScriptingEventChannel}]}]
 }
 {type=MovieSceneEventTrack, display_name=RepeaterEvents, sections [{type=MovieSceneEventRepeaterSection, end_frame=3000, start_frame=0}]}
]
