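# BP_AssetShowcase is an editor utility widget that builds and optionally renders showcase Level Sequences for selected actors.
# Render duplicates a base sequence into a temp folder, then adds the actor, a Nanite-disabled fallback copy and an optional scale reference as spawnables.
# Helpers key turntable and camera cuts, drive time-of-day, and read mesh stats through inline Python.
# UI events set duration, angle, FOV and TOD; selection changes gate the render buttons.
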EditorUtilityWidgetBlueprint /Game/NeverCook/ArtTools/AssetShowcase/BP_AssetShowcase
graphs [
 CreateTempSequence {nodes [
   CreateTempSequence() {is=title K2Node_FunctionEntry}
   n1 = Get selectedActor() -> (selectedActor:Actor Object Reference)
   n2 = "To String (Guid)"!(InGuid:"Guid Structure (by ref)"=NewGuid!() -> (ReturnValue:Guid Structure)) -> (ReturnValue:String)
   > n3 = DuplicateAsset!(
    execute
    SourceAssetPath:String=Get LSBasePath() -> (LSBasePath:String)
    DestinationAssetPath:String=Append!(
     A=Get tempDir() -> (tempDir:String)
     B=GetActorLabel(self=@n1, bCreateIfNone=true)
     C:String=_
     D:String=GetSubstring!(SourceString:String=@n2, StartIndex:int=0, Length:int=6) -> (ReturnValue:String)
    ) -> (ReturnValue)
   ) -> (then, ReturnValue:Object Reference)
   > n4 = Cast To LevelSequence!(execute, Object=@n3) -> (then, CastFailed, AsLevel Sequence:Level Sequence Object Reference) {is=title K2Node_DynamicCast}
   > n5 = Set LSTemp(LSTemp:Level Sequence Object Reference=@n4) -> (Output_Get:Level Sequence Object Reference)
   > SetBurnIn!(
    execute:Exec
    self:Object Reference=@n5
    BurninRef:BP Showcase Burn In Rel Object Reference=Get BurnIn() -> (BurnIn:BP Showcase Burn In Rel Object Reference)
   ) -> (then:Exec) {is=title K2Node_Message}
   > Return Node!(execute, outSequence:Level Sequence Object Reference=@n5, success:bool=true)
   Return Node(execute=@n4.CastFailed, success=false)
  ]}
 EditTempSequence {nodes [
   n1 = DuplicateActor!(
    execute
    self:Editor Actor Subsystem Object Reference=Get EditorActorSubsystem()
    ActorToDuplicate:Actor Object Reference=Get curActor() -> (curActor:Actor Object Reference)
    ToWorld:World Object Reference
    Offset:Vector
   ) -> (then, ReturnValue:Actor Object Reference)
   > n2 = Is Valid(InputObject=@n1)
   PrintString(execute=@n2.Is Not Valid, InString="Actor not duplicated!")
   OpenLevelSequence!(execute=@n2.Is Valid, LevelSequence:Level Sequence Object Reference=Get LSTemp() -> (LSTemp:Level Sequence Object Reference)) -> (then, ReturnValue:bool)
   > n3 = disableNanite!(execute, self, object:Object Reference=@n1, disable:bool=true) -> (then, isSM:bool, isProp:bool, failed:bool, asSM:Static Mesh Actor Object Reference, asProp:Building Prop Object Reference)
   > n4 = Branch(Condition=@n3.failed)
   > PrintString(InString="Disable Nanite Failed!")
   n5 = AddPossessable!(execute=@n4.else, Sequence:Movie Scene Sequence Object Reference=Get LSTemp(), ObjectToPossess:Object Reference=@n1) -> (then, ReturnValue:Movie Scene Binding Proxy Structure)
   > n6 = ConvertToSpawnable!(
    execute
    self:Level Sequence Editor Subsystem Object Reference=Get LevelSequenceEditorSubsystem() -> (ReturnValue:Level Sequence Editor Subsystem Object Reference) {is=title K2Node_GetEditorSubsystem}
    ObjectBinding:"Movie Scene Binding Proxy Structure (by ref)"=@n5
   ) -> (then, ReturnValue:Array of Movie Scene Binding Proxy Structures)
   > Set FallbackRef(FallbackRef:Movie Scene Binding Proxy Structure="Get (a copy)"(Array:Array of Movie Scene Binding Proxy Structures=@n6) -> (Output:Movie Scene Binding Proxy Structure)) -> (Output_Get:Movie Scene Binding Proxy Structure)
   n7 = Length(TargetArray:Array of Movie Scene Binding Proxy Structures=@n6)
   > PrintString(InString=Append!(A="Fallback bindings: ", B="To String (Integer)"(InInt=@n7)) -> (ReturnValue))
   n8 = GetPathNameForLoadedAsset()
   > Execute Python Script!(
    execute
    PythonScript=MakeLiteralString(Value="import unreal\r\n\r\nsmload = unreal.load_asset(sm,None,True)\r\nasset_options = unreal.GeometryScriptCopyMeshFromAssetOptions()\r\nrequested_lod = unreal.GeometryScriptMeshReadLOD()\r\ndm = unreal.DynamicMesh()\r\nunreal.GeometryScript_AssetUtils.copy_mesh_from_static_mesh(smload, dm, asset_options, requested_lod)\r\n\r\n\r\nmesh_editor_ss = unreal.get_editor_subsystem(unreal.StaticMeshEditorSubsystem)\r\nbuildsetting = mesh_editor_ss.get_lod_build_settings(smload,0)\r\nsubsys = unreal.get_editor_subsystem(unreal.StaticMeshEditorSubsystem)\r\n\r\n\r\nlod_for_collision= smload.get_editor_property(\"lod_for_collision\")\r\n \r\n\r\ndist_field_bias = smload.get_editor_property(\"distance_field_self_shadow_bias\")\r\ndist2sided = buildsetting.get_editor_property(\"generate_distance_field_as_if_two_sided\" )\r\nvertcount = subsys.get_number_verts(smload,0)\r\nnanitevertcount = dm.get_vertex_count()\r\ncollcount = subsys.get_convex_collision_count(smload)\r\nleafcol = subsys.is_section_collision_enabled(smload,lod_for_collision,1) and subsys.is_section_collision_enabled(smload,lod_for_collision,0)")
    sm:String=@n8
   ) -> (
    then
    ReturnValue
    dist_field_bias:Wildcard
    dist2sided:Wildcard
    lod_for_collision:Wildcard
    vertcount:Wildcard
    nanitevertcount:Wildcard
    collcount:Wildcard
    leafcol:Wildcard
   ) {inputs [sm], outputs [dist_field_bias, dist2sided, lod_for_collision, vertcount, nanitevertcount, collcount, leafcol]}
   n9 = Get FallbackRef() -> (FallbackRef:Movie Scene Binding Proxy Structure)
   SetupAnimData!(execute, self, binding:Movie Scene Binding Proxy Structure=@n9) -> (then, 3DTransform Track:Movie Scene 3DTransform Track Object Reference, Section:Movie Scene Section Object Reference)
   > setupFallback!(execute, self, binding:Movie Scene Binding Proxy Structure=@n9) -> (then)
   n10 = Set curActor(execute=@n15.then, curActor:Actor Object Reference=Get selectedActor()) -> (Output_Get:Actor Object Reference)
   n11 = Cast To FortStaticMeshActor!(execute, Object=@n10) -> (then, CastFailed, AsFort Static Mesh Actor:Fort Static Mesh Actor Object Reference) {is=title K2Node_DynamicCast}
   n12 = Cast To BuildingSMActor!(execute=@n11.CastFailed, Object=@n10) -> (then, CastFailed, AsBuilding Static Mesh:Building Static Mesh Object Reference) {is=title K2Node_DynamicCast}
   EditTempSequence!() -> (then, LSTemp:Level Sequence Object Reference) {is=title K2Node_FunctionEntry}
   > n13 = Sequence!(execute) -> (then_0, then_1, then_2:Exec, then_3:Exec, then_4:Exec, then_5:Exec)
   > n14 = Set SeqFrames(SeqFrames:int=getTotalDuration!() -> (sequenceFrames:int=0)) -> (Output_Get:int=0)
   > n15 = SetPlaybackEnd!(execute, Sequence:Movie Scene Sequence Object Reference=Get LSTemp(), EndFrame:int=@n14) -> (then)
   n16 = Sequence!(execute=@n13.then_2) -> (then_0, then_1, then_2)
   n17 = Get startFrame() -> (startFrame:int=0)
   > n18 = Set endFrame(endFrame:int=int + int(A=@n17, B=Get CDuration() -> (CDuration:int=0))) -> (Output_Get:int=0)
   > n19 = cineCameraSetup!(
    execute
    self
    cutTrack:Movie Scene Camera Cut Track Object Reference=Get localCutTrack() -> (localCutTrack:Movie Scene Camera Cut Track Object Reference)
    startFrame:int=@n17
    endFrame:int=@n18
    fPrerollFrames:int=Get PreRollFrames() -> (PreRollFrames:int=0)
   ) -> (
    then
    CameraOutBinding:Movie Scene Binding Proxy Structure
    CameraOutComponentBinding:Movie Scene Binding Proxy Structure
    outCutTrack:Movie Scene Camera Cut Track Object Reference
   )
   > Set localCutTrack(localCutTrack:Movie Scene Camera Cut Track Object Reference=@n19.outCutTrack) -> (Output_Get:Movie Scene Camera Cut Track Object Reference)
   > Set CineCamRef(CineCamRef:Movie Scene Binding Proxy Structure=@n19) -> (Output_Get:Movie Scene Binding Proxy Structure)
   > Set CineCamCompRef(CineCamCompRef:Movie Scene Binding Proxy Structure=@n19.CameraOutComponentBinding) -> (Output_Get:Movie Scene Binding Proxy Structure)
   > n20 = SetupAnimData(binding=Get CineCamRef() -> (CineCamRef:Movie Scene Binding Proxy Structure))
   > SetPreRollFrames!(
    execute
    self:Movie Scene Section Object Reference=@n20.Section
    InPreRollFrames:int=getSequenceFrames!(levelSeq:Level Sequence Object Reference=Get LSTemp(), frames:int=Get PreRollFrames() -> (PreRollFrames=0)) -> (Frame Number Value:int=0)
   ) -> (then)
   n21 = GetLevelViewportCameraInfo!(WorldContextObject:Object Reference=Get LevelEditorSubsystem()) -> (CameraLocation:Vector, CameraRotation:Rotator, ReturnValue:bool)
   n22 = Get startFrame() -> (startFrame=0)
   n23 = Get CineCamRef()
   addGroupVectorKey!(
    execute=@n16.then_1
    self
    binding:Movie Scene Binding Proxy Structure=@n23
    group:ENU_TransformGroups Enum=NewEnumerator0
    vector:Vector=@n21
    frame:int=@n22
   ) -> (then)
   n24 = BreakRotator(InRot=@n21.CameraRotation)
   > addGroupVectorKey(binding=@n23, group=NewEnumerator1, vector=MakeVector(X=@n24, Y=@n24.Pitch, Z=@n24.Yaw), frame=@n22)
   Branch(execute=@n13.then_3, Condition=Get useTurnTable() -> (useTurnTable:bool))
   n25 = Get CurAssetRef() -> (CurAssetRef:Movie Scene Binding Proxy Structure)
   > SetupAnimData(binding=@n25)
   > addGroupVectorKey(binding=@n25, group=NewEnumerator0, vector=Get Actor Location(self=Get curActor()), frame=Get startFrame() -> (startFrame=0))
   > addGroupVectorKey(binding=@n25, group=NewEnumerator1, vector="0, 0, 0", frame=Get startFrame() -> (startFrame=0))
   > addGroupVectorKey!(
    execute
    self
    binding=@n25
    group=NewEnumerator1
    vector_X:double=0
    vector_Y:double=0
    vector_Z:double=Get MaxAngle() -> (MaxAngle:double)
    frame=Get endFrame() -> (endFrame:int=0)
   ) -> (then)
   > Set startFrame(startFrame:int=int + int!(A=Get startFrame() -> (startFrame=0), B=Get endFrame() -> (endFrame=0), C:int=1) -> (ReturnValue)) -> (Output_Get:int=0)
   Branch(execute=@n13.then_4, Condition=Get orbitCamera() -> (orbitCamera:bool))
   n26 = Get startFrame() -> (startFrame=0)
   > n27 = Set endFrame(endFrame=int + int!(A=@n26, B=Get CDuration() -> (CDuration=0)) -> (ReturnValue)) -> (Output_Get=0)
   > n28 = cineCameraSetup(
    cutTrack=Get localCutTrack()
    startFrame=@n26
    endFrame=@n27
    fPrerollFrames=SelectInt(A=0, B=Get PreRollFrames() -> (PreRollFrames=0), bPickA=Get useTurnTable())
   )
   > n29 = SetupAnimData(binding=@n28)
   > SetPreRollFrames(
    self=@n29.Section
    InPreRollFrames=getSequenceFrames(levelSeq=Get LSTemp(), frames=Get PreRollFrames() -> (PreRollFrames=0)) -> (Frame Number Value=0)
   )
   n30 = GetLevelViewportCameraInfo(WorldContextObject=Get LevelEditorSubsystem())
   n31 = Get startFrame() -> (startFrame=0)
   > addGroupVectorKey!(execute, self, binding=@n28, group=NewEnumerator0, vector:Vector=@n30, frame=@n31) -> (then)
   n32 = BreakRotator(InRot=@n30.CameraRotation)
   > addGroupVectorKey(binding=@n28, group=NewEnumerator1, vector=MakeVector(X=@n32, Y=@n32.Pitch, Z=@n32.Yaw), frame=@n31)
   Set startFrame(execute=@n13.then_1, startFrame=0) -> (Output_Get=0)
   > Clear!(
    execute
    TargetMap:Map of Movie Scene Binding Proxy Structures to ST Transform Groups Structures=Get TransformData() -> (TransformData:Map of Movie Scene Binding Proxy Structures to ST Transform Groups Structures)
   ) -> (then) {is=title K2Node_CallFunction}
   n33 = Get ScaleActorRef() -> (ScaleActorRef:Movie Scene Binding Proxy Structure)
   > Branch(Condition=IsValid!(InBinding:"Movie Scene Binding Proxy Structure (by ref)"=@n33) -> (ReturnValue))
   > SetupAnimData(binding=@n33)
   > n34 = getScaleActorLocation!(execute, self, selActor:Actor Object Reference=Get curActor()) -> (then, spawnLocation:Vector, ViewLookat:Rotator)
   n35 = Get startFrame() -> (startFrame=0)
   > addGroupVectorKey(binding=@n33, group=NewEnumerator0, vector=@n34, frame=@n35)
   n36 = Combine Rotators!(A:Rotator=@n34.ViewLookat, B:Rotator="0,-90,0") -> (ReturnValue:Rotator)
   n37 = IsChecked(self=Get JunoRef() -> (JunoRef:Editor Utility Check Box Object Reference))
   > addGroupVectorKey(
    binding=@n33
    group=NewEnumerator1
    vector=rotatorToVector!(rot:Rotator=SelectRotator!(A:Rotator=@n36, B:Rotator=@n34.ViewLookat, bPickA:bool=@n37) -> (ReturnValue:Rotator)) -> (vec:Vector)
    frame=@n35
   )
   setupTimeOfDay(execute=@n13.then_5)
   > setupLODInfo()
   > Return Node!(execute, outSequence=Get LSTemp())
   n38 = Get StaticMeshEditorSubsystem()
   n39 = Get StaticMeshComponent(self:Building Static Mesh Object Reference=@n12) -> (StaticMeshComponent:Base Building Static Mesh Component Object Reference)
   n40 = Get StaticMesh(self=@n39)
   Break Mesh Build Settings!(MeshBuildSettings:"Mesh Build Settings Structure (by ref)"=GetLodBuildSettings!(execute, self:Static Mesh Editor Subsystem Object Reference=@n38, StaticMesh:Static Mesh Object Reference=@n40, LodIndex:int=0) -> (then, OutBuildOptions:Mesh Build Settings Structure)) -> (
    bUseMikkTSpace:bool=~
    bRecomputeNormals:bool=~
    bRecomputeTangents:bool=~
    bComputeWeightedNormals:bool=~
    bRemoveDegenerates:bool=~
    bBuildReversedIndexBuffer:bool=~
    bUseHighPrecisionTangentBasis:bool=~
    bUseFullPrecisionUVs:bool=~
    bUseBackwardsCompatibleF16TruncUVs:bool=~
    bGenerateLightmapUVs:bool=~
    bGenerateDistanceFieldAsIfTwoSided:bool=~
    bSupportFaceRemap:bool=~
    MinLightmapResolution:int
    SrcLightmapIndex:int
    DstLightmapIndex:int
    BuildScale3D:Vector=~
    DistanceFieldResolutionScale:float=~
    DistanceFieldReplacementMesh:Static Mesh Object Reference
    MaxLumenMeshCards:int
   ) {is=title K2Node_BreakStruct}
   Break Mesh Nanite Settings!(MeshNaniteSettings:"Mesh Nanite Settings Structure (by ref)"=GetNaniteSettings!(self:Static Mesh Editor Subsystem Object Reference=@n38, StaticMesh:Static Mesh Object Reference=@n40) -> (ReturnValue:Mesh Nanite Settings Structure)) -> (
    bEnabled:bool=~
    bExplicitTangents:bool=~
    bLerpUVs:bool=~
    bSeparable:bool=~
    bVoxelNDF:bool=~
    bVoxelOpacity:bool=~
    ShapePreservation:ENaniteShapePreservation Enum
    PositionPrecision:int
    NormalPrecision:int
    TangentPrecision:int
    BoneWeightPrecision:int
    KeepPercentTriangles:float=~
    TrimRelativeError:float=~
    GenerateFallback:ENaniteGenerateFallback Enum
    FallbackTarget:ENaniteFallbackTarget Enum
    FallbackPercentTriangles:float=~
    FallbackRelativeError:float=~
    MaxEdgeLengthFactor:float=~
    NumRays:int
    VoxelLevel:int
    RayBackUp:float=~
    DisplacementUVChannel:int
   ) {is=title K2Node_BreakStruct}
   GetLodCount!(execute, self:Static Mesh Editor Subsystem Object Reference=@n38, StaticMesh:Static Mesh Object Reference=@n40) -> (then, ReturnValue:int=0)
   CloseLevelSequence!(execute) -> (then)
   Append(
    A="import unreal\r\n"
    B="import unreal\r\n\r\nsmload = unreal.load_asset(sm,None,True)\r\nasset_options = unreal.GeometryScriptCopyMeshFromAssetOptions()\r\nrequested_lod = unreal.GeometryScriptMeshReadLOD()\r\ndm = unreal.DynamicMesh()\r\nunreal.GeometryScript_AssetUtils.copy_mesh_from_static_mesh(smload, dm, asset_options, requested_lod)\r\n\r\n\r\nmesh_editor_ss = unreal.get_editor_subsystem(unreal.StaticMeshEditorSubsystem)\r\nbuildsetting = mesh_editor_ss.get_lod_build_settings(smload,0)\r\nsubsys = unreal.get_editor_subsystem(unreal.StaticMeshEditorSubsystem)\r\n\r\n\r\nlod_for_collision= smload.get_editor_property(\"lod_for_collision\")\r\n \r\n\r\ndist_field_bias = smload.get_editor_property(\"distance_field_self_shadow_bias\")\r\ndist2sided = buildsetting.get_editor_property(\"generate_distance_field_as_if_two_sided\" )\r\nvertcount = subsys.get_number_verts(smload,0)\r\nnanitevertcount = dm.get_vertex_count()\r\ncollcount = subsys.get_convex_collision_count(smload)\r\nleafcol = subsys.is_section_collision_enabled(smload,lod_for_collision,1) and subsys.is_section_collision_enabled(smload,lod_for_collision,0)"
   )
   "Get (a copy)"(Array:Array of Actor Object References=Get CurSel() -> (CurSel:Array of Actor Object References)) -> (Output:Actor Object Reference)
   Get CDuration() -> (CDuration=0)
   Get bDisallowNanite!(self:Static Mesh Component Object Reference=Get StaticMeshComponent(self:Static Mesh Actor Object Reference=@n11) -> (StaticMeshComponent:Static Mesh Component Object Reference)) -> (bDisallowNanite:bool)
   Get bDisallowNanite(self=@n39)
   n41 = Get JunoRef()
   n42 = AddSpawnableFromClass!(
    execute=@n47.then
    self:Level Sequence Editor Subsystem Object Reference=Get LevelSequenceEditorSubsystem()
    Sequence:Movie Scene Sequence Object Reference=Get LSTemp()
    ClassToSpawn:Object Class Reference=SelectClass!(
     A:Object Class Reference=Get JunoScaleActor() -> (JunoScaleActor:Actor Class Reference)
     B:Object Class Reference=Get ScaleActor() -> (ScaleActor:Test Scale Editor Actor Class Reference)
     bSelectA:bool=IsChecked(self=@n41)
    ) -> (ReturnValue:Object Class Reference)
   ) -> (then, ReturnValue:Movie Scene Binding Proxy Structure)
   > Set ScaleActorRef(ScaleActorRef:Movie Scene Binding Proxy Structure=@n42) -> (Output_Get:Movie Scene Binding Proxy Structure)
   n43 = Set Actor Hidden In Game!(execute=@n45.then, self:Actor Object Reference=Get selectedActor(), bNewHidden:bool=true) -> (then) {is=title K2Node_CallFunction}
   n44 = AddSpawnableFromInstance!(
    execute=@n49.then
    self:Level Sequence Editor Subsystem Object Reference=Get LevelSequenceEditorSubsystem()
    Sequence:Movie Scene Sequence Object Reference=Get LSTemp()
    ObjectToSpawn:Object Reference=Get selectedActor()
   ) -> (then, ReturnValue:Movie Scene Binding Proxy Structure)
   > n45 = Set CurAssetRef(CurAssetRef:Movie Scene Binding Proxy Structure=@n44) -> (Output_Get:Movie Scene Binding Proxy Structure)
   n46 = addVisiblityTrack!(execute=@n43.then, self, InBinding:"Movie Scene Binding Proxy Structure (by ref)"=@n45, duration:int=Get SeqFrames() -> (SeqFrames:int=0)) -> (then, success:bool, binding:Movie Scene Binding Proxy Structure, boolChannel:Movie Scene Scripting Bool Channel Object Reference)
   > "Add Key (Bool)"!(
    execute
    self:Movie Scene Scripting Bool Channel Object Reference=@n46.boolChannel
    InTime:"Frame Number Structure (by ref)"=Make Frame Number!(Value:int=0) -> (FrameNumber:Frame Number Structure) {is=title K2Node_MakeStruct}
    NewValue:bool=true
    SubFrame:float=0
    TimeUnit:EMovieSceneTimeUnit Enum=DisplayRate
   ) -> (then, ReturnValue:Movie Scene Scripting Bool Key Object Reference)
   > n47 = Branch(Condition=IsChecked(self=Get SizeRef() -> (SizeRef:Editor Utility Check Box Object Reference)))
   n48 = Get bHidden!(self:Actor Object Reference=@n10) -> (bHidden:bool)
   n49 = Set selectedWasHidden(execute=@n10.then, selectedWasHidden:bool=@n48) -> (Output_Get:bool)
  ]}
 EventGraph {nodes [
   n1 = GetSelectedLevelActors(self=Get EditorActorSubsystem())
   > Set CurSel(CurSel:Array of Actor Object References=@n1) -> (Output_Get:Array of Actor Object References)
   > Branch(Condition=Is Empty!(TargetArray:Array of Actor Object References=@n1) -> (ReturnValue:bool))
   > printMessage!(
    execute
    self
    Message_Message_4_#1:Text="LOCTEXT(\"Warning: no selected Actor! Skipping...\")"
    Message_Level_10_#2:ENU_MessageLevel Enum=NewEnumerator1
    displayTime:int=5
   ) -> (then)
   n2 = "On Value Changed (TODSlider)"!() -> (OutputDelegate, then, Value:float) {is=title K2Node_ComponentBoundEvent}
   > TODCall_Event!(execute, self, value:double=@n2.Value) -> (then)
   n3 = "On Value Changed (cutDurationSlider)"!() -> (OutputDelegate, then, Value:float) {is=title K2Node_ComponentBoundEvent}
   > manageDuration!(execute, self, value:double=@n3.Value) -> (then)
   n4 = "On Check State Changed (rotateActors)"!() -> (OutputDelegate, then, bIsChecked:bool) {is=title K2Node_ComponentBoundEvent}
   > Set useTurnTable(useTurnTable:bool=@n4.bIsChecked) -> (Output_Get:bool)
   "On Clicked (getTODActor)"() {is=title K2Node_ComponentBoundEvent}
   > getTODActor_evt()
   n5 = "On Value Changed (turnAngle)"!() -> (OutputDelegate, then, Value:float) {is=title K2Node_ComponentBoundEvent}
   > manageAngle!(execute, self, value:double=@n5.Value) -> (then)
   "On Clicked (angle90)"() {is=title K2Node_ComponentBoundEvent}
   > presetAngle!(execute, self, angle:int=90) -> (then)
   "On Clicked (angle360)"() {is=title K2Node_ComponentBoundEvent}
   > presetAngle(angle=360)
   "On Clicked (angle180)"() {is=title K2Node_ComponentBoundEvent}
   > presetAngle(angle=180)
   "On Clicked (angle270)"() {is=title K2Node_ComponentBoundEvent}
   > presetAngle(angle=270)
   n6 = "On Check State Changed (rotateCamera)"!() -> (OutputDelegate, then, bIsChecked:bool) {is=title K2Node_ComponentBoundEvent}
   > Set orbitCamera(orbitCamera:bool=@n6.bIsChecked) -> (Output_Get:bool)
   n7 = "On Value Changed (camFOV_slider)"!() -> (OutputDelegate, then, Value:float) {is=title K2Node_ComponentBoundEvent}
   > manageFOV!(execute, self, value:double=@n7.Value) -> (then)
   "On Clicked (fov35)"() {is=title K2Node_ComponentBoundEvent}
   > presetFOV!(execute, self, fov:int=35) -> (then)
   "On Clicked (renderSelected)"() {is=title K2Node_ComponentBoundEvent}
   > n8 = Sequence!(execute) -> (then_0, then_1)
   > fetchUGSChange()
   n9 = CreateTempSequence!(execute=@n8.then_1, self) -> (then, outSequence:Level Sequence Object Reference, success:bool) {is=title K2Node_CallFunction}
   > n10 = Branch(Condition=@n9.success)
   > n11 = EditTempSequence!(execute, self, LSTemp:Level Sequence Object Reference=@n9) -> (then, outSequence:Level Sequence Object Reference) {is=title K2Node_CallFunction}
   > n12 = Branch(Condition=Get RenderImmediately() -> (RenderImmediately:bool))
   > n13 = renderQueue!(execute, self, InSequence:Level Sequence Object Reference=@n11) -> (then, success:bool)
   > Branch(Condition=@n13)
   n14 = Get LSTemp()
   > printMessage(
    Message_Message_4_#1=Format Text!(Format="LOCTEXT(\"Rendering Sequence:\\r\\n{seq}\")", seq:String=GetDisplayName(Object=@n14)) -> (Result)
    Message_Level_10_#2=NewEnumerator0
    displayTime=5
   )
   n15 = Get LSTemp()
   printMessage(
    execute=@n12.else
    Message_Message_4_#1=Format Text(Format="LOCTEXT(\"Sequence Saved in Temp:\\r\\n{seq}\")", seq=GetDisplayName(Object=@n15))
    Message_Level_10_#2=NewEnumerator0
    displayTime=5
   )
   > OpenLevelSequence!(execute, self:Self Object Reference) -> (then)
   printMessage(
    execute=@n10.else
    Message_Message_4_#1="LOCTEXT(\"Error: could not create Temp Sequence! Stopping...\")"
    Message_Level_10_#2=NewEnumerator2
    displayTime=5
   )
   "On Clicked (fov75)"() {is=title K2Node_ComponentBoundEvent}
   > presetFOV(fov=75)
   "On Clicked (fov90)"() {is=title K2Node_ComponentBoundEvent}
   > presetFOV(fov=90)
   "On Clicked (fov120)"() {is=title K2Node_ComponentBoundEvent}
   > presetFOV(fov=120)
   n16 = "On Check State Changed (renderNow)"!() -> (OutputDelegate, then, bIsChecked:bool) {is=title K2Node_ComponentBoundEvent}
   > Set RenderImmediately(RenderImmediately:bool=@n16.bIsChecked) -> (Output_Get:bool)
   n17 = "On Value Committed (TODMinTime)"!() -> (OutputDelegate, then, InValue:float, CommitMethod:ETextCommit Enum=Default) {is=title K2Node_ComponentBoundEvent}
   n18 = getTODTimeValue!(value:double=@n17.InValue) -> (outValue:double)
   > SetValue!(execute, self:Spin Box Object Reference=Get TODMinTime() -> (TODMinTime:Editor Utility Spin Box Object Reference), NewValue:float=@n18) -> (then)
   > Set TODMin(TODMin:double=@n18) -> (Output_Get:double)
   > TODCall_Event(value=float / float(A=@n17.InValue, B=23.99))
   n19 = "On Value Committed (TODMaxTime)"!() -> (OutputDelegate, then, InValue:float, CommitMethod:ETextCommit Enum=Default) {is=title K2Node_ComponentBoundEvent}
   n20 = getTODTimeValue(value=@n19.InValue)
   > SetValue(self=Get TODMaxTime() -> (TODMaxTime:Editor Utility Spin Box Object Reference), NewValue=@n20)
   > Set TODMax(TODMax:double=@n20) -> (Output_Get:double)
   > TODCall_Event(value=float / float(A=@n19.InValue, B=23.99))
   n21 = "On Check State Changed (TODRange)"!() -> (OutputDelegate, then, bIsChecked:bool) {is=title K2Node_ComponentBoundEvent}
   > SetIsEnabled(self=Get TODSliderBox() -> (TODSliderBox:Vertical Box Object Reference), bInIsEnabled=NOT Boolean(A=@n21.bIsChecked))
   > SetIsEnabled(self=Get TODRangeBox() -> (TODRangeBox:Horizontal Box Object Reference), bInIsEnabled=@n21.bIsChecked)
   > n22 = Set useTODRange(useTODRange:bool=@n21.bIsChecked) -> (Output_Get:bool)
   > Set TODMin(TODMin=GetValue!(self:Spin Box Object Reference=Get TODMinTime()) -> (ReturnValue:float))
   > Set TODMax(TODMax=GetValue(self=Get TODMaxTime()))
   > n23 = Branch(Condition=@n22)
   TODCall_Event(execute=@n23.else, value=GetValue(self:Radial Slider Object Reference=Get TODSlider() -> (TODSlider:Radial Slider Object Reference)))
   n24 = "On Check State Changed (openSequence)"!() -> (OutputDelegate, then, bIsChecked:bool) {is=title K2Node_ComponentBoundEvent}
   > Set OpenLevelSeq(OpenLevelSeq:bool=@n24.bIsChecked) -> (Output_Get:bool)
   "On Check State Changed (SizeRef)"!() -> (OutputDelegate, then, bIsChecked:bool) {is=title K2Node_ComponentBoundEvent}
   > manageSizeRef()
   n25 = "On Value Changed (yOffsetSlider)"!() -> (OutputDelegate, then, Value:float) {is=title K2Node_ComponentBoundEvent}
   > manageOffset!(execute, self, YOffset:double=@n25.Value) -> (then)
   "On Clicked (cleanupTempFiles)"() {is=title K2Node_ComponentBoundEvent}
   > cleanupTempSequences()
   "On Clicked (placeScaleActor)"() {is=title K2Node_ComponentBoundEvent}
   > spawnJonesy()
   "On Clicked (disableNaniteButton)"() {is=title K2Node_ComponentBoundEvent}
   n26 = Get selectedActor()
   > Is Valid(InputObject=@n26)
   > n27 = disableNanite(object=@n26, disable=true)
   > n28 = Branch(Condition=@n27.failed)
   > PrintString(InString="Failed to disable Nanite!")
   PrintString(execute=@n28.else, InString="Nanite disabled!")
   "On Clicked (secs_10)"() {is=title K2Node_ComponentBoundEvent}
   > presetCutDuration!(execute, self, seconds:int=10) -> (then)
   "On Clicked (secs_20)"() {is=title K2Node_ComponentBoundEvent}
   > presetCutDuration(seconds=20)
   "On Clicked (secs_30)"() {is=title K2Node_ComponentBoundEvent}
   > presetCutDuration(seconds=30)
   n29 = ActorSelectionChanged!() -> (OutputDelegate, then, SelectionSet:Typed Element Selection Set Object Reference) {is=title K2Node_CustomEvent}
   n30 = getSelectedAsset!(execute=[@n50.then, @n29.then], self, set:Typed Element Selection Set Object Reference=@n29.SelectionSet) -> (then, success:bool, outSM:Static Mesh Object Reference, outActor:Actor Object Reference, outClass:Actor Class Reference)
   > M UISelect Render Type!(execute:Exec, Condition:bool=@n30) -> (then:Exec) {is=title K2Node_MacroInstance}
   > manageSizeRef()
   n31 = getTODActor_evt() {is=title K2Node_CustomEvent}
   n32 = GetAllActorsOfClass!(
    execute=[@n31.then, @n48.then_0]
    WorldContextObject:Object Reference
    ActorClass:Actor Class Reference=/Script/FortniteGame.FortDaySequenceActor
   ) -> (then, OutActors:Array of Fort Day Sequence Actor Object References)
   > n33 = Branch(Condition=Is Valid Index!(TargetArray:Array of Fort Day Sequence Actor Object References=@n32, IndexToTest:int) -> (ReturnValue:bool))
   > Set TODActor(TODActor:Fort Day Sequence Actor Object Reference="Get (a copy)"(Array:Array of Fort Day Sequence Actor Object References=@n32) -> (Output:Fort Day Sequence Actor Object Reference)) -> (Output_Get:Fort Day Sequence Actor Object Reference)
   > SetIsEnabled(self=Get TODSlider(), bInIsEnabled=true)
   > n34 = SetIsEnabled(self=Get TODRange() -> (TODRange:Editor Utility Check Box Object Reference), bInIsEnabled=true)
   TODCall_Event(execute=[@n34.then, @n37.then], value=GetValue(self=Get TODSlider()))
   n35 = GetAllActorsOfClass(execute=@n33.else, ActorClass=/Game/TimeOfDay/TODM/MASTER_TODM.MASTER_TODM_C) -> (OutActors:Array of MASTER TODM Object References)
   > n36 = Branch(Condition=Is Valid Index(TargetArray:Array of MASTER TODM Object References=@n35))
   > Set TODRef(TODRef:MASTER TODM Object Reference="Get (a copy)"(Array:Array of MASTER TODM Object References=@n35) -> (Output:MASTER TODM Object Reference)) -> (Output_Get:MASTER TODM Object Reference)
   > SetIsEnabled(self=Get TODSlider(), bInIsEnabled=true)
   > SetIsEnabled(self=Get TODRange(), bInIsEnabled=false)
   > n37 = PrintString(InString="Warning: No TimeOfDay Manager found! Time Range disabled!")
   SetIsEnabled(execute=@n36.else, self=Get TODSlider(), bInIsEnabled=false)
   > SetIsEnabled(self=Get TODRange(), bInIsEnabled=false)
   > PrintString(InString="Warning: No TimeOfDay Actor found! TOD Controls disabled!")
   n38 = TODCall_Event!() -> (OutputDelegate, then, value:double=~) {is=title K2Node_CustomEvent}
   n39 = IsValid!(Object:Object Reference=Get TODRef() -> (TODRef:MASTER TODM Object Reference)) -> (ReturnValue)
   > n40 = Branch(Condition=@n39)
   > n41 = setTOD!(execute, self, value:double=@n38.value) -> (then, curTime:double)
   n42 = previewTimeOfDay!(execute=@n40.else, self, value:double=@n38.value, TODActorRef:Fort Day Sequence Actor Object Reference) -> (then, curTime:double)
   Set CurTOD(execute=[@n41.then, @n42.then], CurTOD:double=SelectFloat(A=@n41, B=@n42, bPickA=@n39)) -> (Output_Get:double)
   MRQDone!() -> (OutputDelegate, then, PipelineExecutor:Movie Pipeline Executor Base Object Reference, bSuccess:bool=~) {is=title K2Node_CustomEvent}
   n43 = Get TODActor() -> (TODActor:Fort Day Sequence Actor Object Reference)
   > n44 = Is Valid(InputObject=@n43)
   n45 = getTODActor_evt(execute=@n44.Is Not Valid) {is=title K2Node_CallFunction}
   SetEditorProperty(
    execute=[@n45.then, @n44.Is Valid]
    Object=@n43
    PropertyName=String To Name(InString=MakeLiteralString(Value=bRunDayCycle))
    PropertyValue=MakeLiteralBool(Value=false)
   )
   > Set Actor Hidden In Game(self=Get selectedActor(), bNewHidden=Get selectedWasHidden() -> (selectedWasHidden:bool))
   > OpenLevelSequence()
   OpenLevelSequence() {is=title K2Node_CustomEvent}
   > Branch(Condition=Get OpenLevelSeq() -> (OpenLevelSeq:bool))
   n46 = Get LSTemp()
   > Is Valid(InputObject=@n46)
   > Delay!(execute, WorldContextObject:Object Reference, Duration=2) -> (then)
   > OpenLevelSequence!(execute, LevelSequence:Level Sequence Object Reference=@n46) -> (then, ReturnValue:bool) {is=title K2Node_CallFunction}
   n47 = OnMapChanged_Event!() -> (OutputDelegate, then, MapChangeEventFlags:int) {is=title K2Node_CustomEvent}
   > Set TODRef()
   > Set TODActor()
   > Set selectedActor(selectedActor:Actor Object Reference) -> (Output_Get:Actor Object Reference)
   > Set selectedActorSM(selectedActorSM:Static Mesh Object Reference) -> (Output_Get:Static Mesh Object Reference)
   > Set LSBase(LSBase:Level Sequence Object Reference) -> (Output_Get:Level Sequence Object Reference)
   > Set LSTemp()
   > PrintString(InString="Deleted TOD Reference!")
   Event PreConstruct()
   > n48 = Sequence!(execute) -> (then_0, then_1, then_2:Exec, then_3:Exec)
   manageDuration(
    execute=@n48.then_1
    value=GetValue(self:Slider Object Reference=Get cutDurationSlider() -> (cutDurationSlider:Editor Utility Slider Object Reference))
   )
   > n49 = GetSelectionSet(self=Get LevelEditorSubsystem())
   > n50 = Bind Event to On Selection Change(self:Typed Element Selection Set Object Reference=@n49, Delegate=@n29) {is=title K2Node_AddDelegate}
   manageOffset(execute=@n48.then_2, YOffset=GetValue(self=Get yOffsetSlider() -> (yOffsetSlider:Editor Utility Slider Object Reference)))
   > manageAngle(value=GetValue(self=Get turnAngle() -> (turnAngle:Editor Utility Slider Object Reference)))
   > manageFOV(value=GetValue(self=Get camFOV_slider() -> (camFOV_slider:Editor Utility Slider Object Reference)))
   n51 = Get rotateActors() -> (rotateActors:Editor Utility Check Box Object Reference)
   > Set useTurnTable(useTurnTable=Select!(
     Unchecked:bool=false
     Checked:bool=true
     Undetermined:bool=true
     Index:ECheckBoxState Enum=GetCheckedState!(self:Check Box Object Reference=@n51) -> (ReturnValue:ECheckBoxState Enum=Unchecked)
    ) -> (ReturnValue:bool))
   n52 = Get rotateCamera() -> (rotateCamera:Editor Utility Check Box Object Reference)
   > Set orbitCamera(orbitCamera=Select(Unchecked=false, Checked=true, Undetermined=true, Index=GetCheckedState(self=@n52) -> (ReturnValue=Unchecked)))
   > Set TODMin(TODMin=GetValue(self:Spin Box Object Reference=Get TODMinTime()))
   > Set TODMax(TODMax=GetValue(self=Get TODMaxTime()))
   n53 = getTempDir!(execute=@n48.then_3, self) -> (then, LSBaseSourcePath:String, TempDir:String, success:bool)
   > n54 = Branch(Condition=@n53.success)
   > printMessage!(execute, self, Message:ST Showcase Message Structure=Get DefaultMessage() -> (DefaultMessage:ST Showcase Message Structure), displayTime=0) -> (then)
   printMessage!(
    execute=@n54.else
    self
    Message_Message_4_#1:Text=Format Text!(Format="LOCTEXT(\"Error: could not create Temp Folder at path:\\r\\n{tempPath}\")", tempPath:String=@n53.TempDir) -> (Result)
    Message_Level_10_#2:ENU_MessageLevel Enum=NewEnumerator2
    displayTime=0
   ) -> (then)
   Event Construct()
   > Assign On Map Changed!(execute:Exec, self:Level Editor Subsystem Object Reference=Get LevelEditorSubsystem(), Delegate:"Delegate (by ref)"=@n47) -> (then:Exec) {is=title K2Node_AssignDelegate}
   TryRun!(
    execute
    self:Editor Utility Subsystem Object Reference=Get EditorUtilitySubsystem()
    Asset:Object Reference=/Game/NeverCook/ArtTools/AssetShowcase/Assets/CINE_Tool_GetUgsCl
   ) -> (then, ReturnValue:bool)
   PrintString(InString="Destructed!")
   debugFunc()
   Set Actor Hidden In Game(self=Get selectedActor(), bNewHidden=Get selectedWasHidden())
   MRQJobStarted!() -> (OutputDelegate, then, StartedJob:Movie Pipeline Executor Job Object Reference) {is=title K2Node_CustomEvent}
   Event Tick()
   Event Destruct()
   Event OnInitialized()
  ]}
 FOVtoFocalLength {nodes [
   n1 = FOVtoFocalLength!() -> (then, fov:double=~) {is=title K2Node_FunctionEntry}
   n2 = float * float(A=@n1, B=0.5) {is=title K2Node_PromotableOperator}
   > Return Node!(execute, ReturnValue:double=float / float(A=11.88, B="Tan (Degrees)"!(A:double=@n2) -> (ReturnValue:double)))
  ]}
 Get Pre Roll Frame for Event {nodes [
   Get Pre Roll Frame for Event() {is=title K2Node_FunctionEntry}
   n1 = Get PreRollFrames() -> (PreRollFrames=0)
   > Return Node(ReturnValue:int=int * int(A=int - int(A=@n1, B=2), B=-1))
  ]}
 Get Pre Roll Frame for Section {nodes [
   Get Pre Roll Frame for Section() {is=title K2Node_FunctionEntry}
   > Return Node(ReturnValue=int * int(A=Get PreRollFrames() -> (PreRollFrames=0), B=-1))
  ]}
 M_UISelectRenderType {nodes [
   n1 = Inputs!() -> (execute, Condition:bool=~)
   > n2 = Branch(Condition=@n1)
   n3 = Get selectedActorClass() -> (selectedActorClass:Actor Class Reference)
   n4 = "Equal (Class)"(A=@n3, B=/Script/FortniteGame.BuildingSMActor)
   > SetIsEnabled(self=Get renderBP() -> (renderBP:Editor Utility Button Object Reference), bInIsEnabled=@n4)
   n5 = OR Boolean(A="Equal (Class)"(A=@n3, B=/Script/FortniteGame.FortStaticMeshActor), B=@n4)
   > SetIsEnabled(self=Get renderSelected() -> (renderSelected:Editor Utility Button Object Reference), bInIsEnabled=@n5)
   > n6 = SetIsEnabled(self=Get renderLODs() -> (renderLODs:Editor Utility Button Object Reference), bInIsEnabled=@n5)
   Outputs!(then:Exec=[@n6.then, @n7.then])
   SetIsEnabled(execute=@n2.else, self=[Get renderBP(), Get renderLODs(), Get renderSelected()], bInIsEnabled=false)
   > n7 = printMessage(
    Message_Message_4_#1="LOCTEXT(\"Info: no valid Actor selected...\")"
    Message_Level_10_#2=NewEnumerator0
    displayTime=3
   )
  ]}
 M_UIprintErrorInvalidObject {nodes [
   n1 = Inputs!() -> (in:Exec, InputObject:Object Reference, execPhase:String)
   > n2 = Is Valid(InputObject=@n1)
   n3 = printMessage(
    execute=@n2.Is Not Valid
    Message_Message_4_#1=Format Text!(Format="LOCTEXT(\"Error: Object not valid during {phase}. Stopping...\")", phase:String=@n1.execPhase) -> (Result)
    Message_Level_10_#2=NewEnumerator2
    displayTime=5
   )
   Outputs!(Is Valid:Exec=@n2.Is Valid, Not Valid:Exec=@n3.then)
  ]}
 M_UIprintErrorPhase {nodes [
   n1 = Inputs!() -> (in, execPhase)
   > printMessage(
    Message_Message_4_#1=Format Text(Format="LOCTEXT(\"Error: could not {phase}. Stopping...\")", phase=@n1)
    Message_Level_10_#2=NewEnumerator2
    displayTime=5
   )
   > Outputs!(out:Exec)
  ]}
 SetupAnimData {nodes [
   n1 = SetupAnimData!() -> (then, binding:Movie Scene Binding Proxy Structure) {is=title K2Node_FunctionEntry}
   n2 = AddTrack!(
    execute=@n23.then_0
    InBinding:"Movie Scene Binding Proxy Structure (by ref)"=@n1
    TrackType:Movie Scene Track Class Reference=/Script/MovieSceneTracks.MovieScene3DTransformTrack
   ) -> (then, ReturnValue:Movie Scene Track Object Reference)
   n3 = AddSection!(execute=@n13.then, Track:Movie Scene Track Object Reference=@n2) -> (then, ReturnValue:Movie Scene Section Object Reference)
   n4 = SetStartFrame!(execute, Section:Movie Scene Section Object Reference=@n3, StartFrame:int=int * int(A=Get PreRollFrames() -> (PreRollFrames=0), B=-1)) -> (then)
   SetEndFrame!(execute=[@n4.then, @n3.then], Section:Movie Scene Section Object Reference=@n3, EndFrame:int=Get SeqFrames() -> (SeqFrames=0)) -> (then)
   > n5 = GetAllChannels!(execute, Section:Movie Scene Section Object Reference=@n3) -> (then, ReturnValue:Array of Movie Scene Scripting Channel Object References)
   > n6 = For Loop(LastIndex=2)
   n7 = For Loop(execute=@n6.Completed, LastIndex=2)
   n8 = For Loop(execute=@n7.Completed, LastIndex=2)
   n9 = Make ST Transform Channels!(TransformChannels_7_#3:Map of ENU_TransformChannels Enums to Movie Scene Scripting Double Channel Object References=Get tRotation() -> (tRotation:Map of ENU_TransformChannels Enums to Movie Scene Scripting Double Channel Object References=NewEnumerator0)) -> (ST_TransformChannels:ST Transform Channels Structure) {is=title K2Node_MakeStruct}
   n10 = Make ST Transform Channels(TransformChannels_7_#3=Get tScale() -> (tScale:Map of ENU_TransformChannels Enums to Movie Scene Scripting Double Channel Object References=NewEnumerator0))
   n11 = Make ST Transform Channels(TransformChannels_7_#3=Get tLocation() -> (tLocation:Map of ENU_TransformChannels Enums to Movie Scene Scripting Double Channel Object References=NewEnumerator0))
   n12 = Add!(
    execute=@n8.Completed
    TargetMap:Map of Movie Scene Binding Proxy Structures to ST Transform Groups Structures=Get TransformData()
    Key:"Movie Scene Binding Proxy Structure (by ref)"=Get binding() -> (binding:Movie Scene Binding Proxy Structure)
    Value:"ST Transform Groups Structure (by ref)"=Make ST Transform Groups!(ChannelGroup_8_#3:Map of ENU_TransformGroups Enums to ST Transform Channels Structures=Make Map!(
      Key 0:ENU_TransformGroups Enum=NewEnumerator0
      Value 0:ST Transform Channels Structure=@n11
      Key 1:ENU_TransformGroups Enum=NewEnumerator1
      Value 1:ST Transform Channels Structure=@n9
      Key 2:ENU_TransformGroups Enum=NewEnumerator2
      Value 2:ST Transform Channels Structure=@n10
     ) -> (Map:Map of ENU_TransformGroups Enums to ST Transform Channels Structures) {is=title K2Node_MakeMap}) -> (ST_TransformGroups:ST Transform Groups Structure) {is=title K2Node_MakeStruct}
   ) -> (then) {is=title K2Node_CallFunction}
   n13 = Cast To MovieScene3DTransformTrack!(execute=@n2.then, Object=@n2) -> (then, CastFailed, AsMovie Scene 3DTransform Track:Movie Scene 3DTransform Track Object Reference) {is=title K2Node_DynamicCast}
   Return Node!(execute=@n12.then, 3DTransform Track:Movie Scene 3DTransform Track Object Reference=@n13, Section:Movie Scene Section Object Reference=@n3)
   n14 = Cast To MovieSceneScriptingDoubleChannel!(
    execute=@n8.LoopBody
    Object="Get (a copy)"(Array:Array of Movie Scene Scripting Channel Object References=@n5, Dimension 1=int + int(A=@n8, B=6)) -> (Output:Movie Scene Scripting Channel Object Reference)
   ) -> (then, CastFailed, AsMovie Scene Scripting Double Channel:Movie Scene Scripting Double Channel Object Reference) {is=title K2Node_DynamicCast}
   n15 = indexToEnum!(inIndex:int=@n8) -> (ReturnValue:ENU_TransformChannels Enum=NewEnumerator0)
   > Add(
    TargetMap:Map of ENU_TransformChannels Enums to Movie Scene Scripting Double Channel Object References=Get tScale() -> (tScale=NewEnumerator0)
    Key:"ENU_TransformChannels Enum (by ref)"=@n15
    Value:"Movie Scene Scripting Double Channel Object Reference (by ref)"=@n14
   )
   n16 = Cast To MovieSceneScriptingDoubleChannel(execute=@n7.LoopBody, Object="Get (a copy)"(Array=@n5, Dimension 1=int + int(A=@n7, B=3)))
   n17 = indexToEnum(inIndex=@n7) -> (ReturnValue=NewEnumerator0)
   > Add(TargetMap=Get tRotation() -> (tRotation=NewEnumerator0), Key=@n17, Value=@n16)
   n18 = Cast To MovieSceneScriptingDoubleChannel(execute=@n6.LoopBody, Object="Get (a copy)"(Array=@n5, Dimension 1=@n6))
   n19 = indexToEnum(inIndex=@n6) -> (ReturnValue=NewEnumerator0)
   > Add(TargetMap=Get tLocation() -> (tLocation=NewEnumerator0), Key=@n19, Value=@n18)
   n20 = Get ChannelName!(self:Movie Scene Scripting Channel Object Reference=@n14) -> (ChannelName:Name=None)
   PrintString(InString=Append!(
     A="Added Key: "
     B=Enum to String!(Enumerator:ENU_TransformChannels Enum=@n15) -> (ReturnValue:String)
     C:String=" - Channel Name: "
     D:String="To String (Name)"(InName=@n20)
    ) -> (ReturnValue))
   n21 = Get ChannelName(self=@n18) -> (ChannelName=None)
   PrintString(InString=Append(A="Added Key: ", B=Enum to String(Enumerator=@n19), C=" - Channel Name: ", D="To String (Name)"(InName=@n21)))
   n22 = Get ChannelName(self=@n16) -> (ChannelName=None)
   PrintString(InString=Append(A="Added Key: ", B=Enum to String(Enumerator=@n17), C=" - Channel Name: ", D="To String (Name)"(InName=@n22)))
   Get CDuration() -> (CDuration=0)
   n23 = Sequence!(execute=@n1.then) -> (then_0, then_1)
  ]}
 TODCall {nodes [TODCall!() -> (then, value:double=~) {is=title K2Node_FunctionEntry}]}
 ToolTip_AnglePresets {nodes [
   ToolTip_AnglePresets() {is=title K2Node_FunctionEntry}
   > Return Node!(execute, ReturnValue:Text=MakeLiteralText(Value="LOCTEXT(\"Rotation Max Angle preset - in degrees\")"))
  ]}
 ToolTip_CleanupTemp {nodes [
   ToolTip_CleanupTemp() {is=title K2Node_FunctionEntry}
   > Return Node(ReturnValue=MakeLiteralText(Value="LOCTEXT(\"Temp Level Sequences are saved in a \\\"temp\\\" folder located at the root level of this Editor Utility. If you don\\'t need any of the generated Level Sequences you may use this button to delete them up.\\r\\n\\r\\nTODO:\\r\\n- Add an option not to save temp Level Sequences - if possible\")"))
  ]}
 ToolTip_CutDuration {nodes [
   ToolTip_CutDuration() {is=title K2Node_FunctionEntry}
   > Return Node(ReturnValue=MakeLiteralText(Value="LOCTEXT(\"Duration (frames) of each camera cut. At least one cut must be active.\")"))
  ]}
 ToolTip_CutDurationPresets {nodes [
   ToolTip_CutDurationPresets() {is=title K2Node_FunctionEntry}
   > Return Node(ReturnValue=MakeLiteralText(Value="LOCTEXT(\"Preset Duration - in seconds - for each cut\")"))
  ]}
 ToolTip_GetTODActor {nodes [
   ToolTip_GetTODActor() {is=title K2Node_FunctionEntry}
   > Return Node(ReturnValue=MakeLiteralText(Value="LOCTEXT(\"Looks for a Time of Day Actor in the current Level if it couldn\\'t find one. If there is a TOD Actor in the Level, the TOD Slider becomes available.\")"))
  ]}
 ToolTip_JonesyOffset {nodes [
   ToolTip_JonesyOffset() {is=title K2Node_FunctionEntry}
   > Return Node(ReturnValue=MakeLiteralText(Value="LOCTEXT(\"The offset for placing Jonesy - in cm. This is also used to raycast Jonesy position against the selected Actor in order to spawn it in a \\\"safe\\\" location.\")"))
  ]}
 ToolTip_OpenSequence {nodes [
   ToolTip_OpenSequence() {is=title K2Node_FunctionEntry}
   > Return Node(ReturnValue=MakeLiteralText(Value="LOCTEXT(\"Opens the generated Level Sequence in Sequencer\")"))
  ]}
 ToolTip_RenderQueue {nodes [
   ToolTip_RenderQueue() {is=title K2Node_FunctionEntry}
   > Return Node(ReturnValue=MakeLiteralText(Value="LOCTEXT(\"Begins rendering of the generated Level Sequence using MRQ\")"))
  ]}
 ToolTip_RenderSelected {nodes [
   ToolTip_RenderSelected() {is=title K2Node_FunctionEntry}
   > Return Node(ReturnValue=MakeLiteralText(Value="LOCTEXT(\"Renders any active cut of the selected Actor in the Level. This will start rendering in Editor with Movie Render Queue right away. As a default it produces mp4 videos and deletes the source image sequence.\\r\\n\\r\\nTODO:\\r\\n- Selectable Config files\\r\\n- Add to Queue - instead of Rendering one sequence at a time\\r\\n- Save the current Queue in Temp folder for later use\")"))
  ]}
 ToolTip_TurnTableEnable {nodes [
   ToolTip_TurnTableEnable() {is=title K2Node_FunctionEntry}
   > Return Node(ReturnValue=MakeLiteralText(Value="LOCTEXT(\"Development: this is disabled until other options are implemented\")"))
  ]}
 ToolTip_UseTurnTable {nodes [
   ToolTip_UseTurnTable() {is=title K2Node_FunctionEntry}
   > Return Node(ReturnValue=MakeLiteralText(Value="LOCTEXT(\"Rotates the selected Asset (as a possessable) from zero to the specified angle\")"))
  ]}
 ToolTip_UserComment {nodes [
   ToolTip_UserComment() {is=title K2Node_FunctionEntry}
   > Return Node(ReturnValue=MakeLiteralText(Value="LOCTEXT(\"This adds a User Comment to the final Burn-in\")"))
  ]}
 addCameraCut {nodes [
   n1 = addCameraCut!() -> (
    then
    Sequence:Level Sequence Object Reference
    CameraBinding:Movie Scene Binding Proxy Structure
    cutTrack:Movie Scene Camera Cut Track Object Reference
    start:int
    end:int
    preroll:int
   ) {is=title K2Node_FunctionEntry}
   > n2 = Sequence!(execute) -> (then_0, then_1, then_2:Exec)
   > Set curSeq(curSeq:Level Sequence Object Reference=Select!(Option 0:Level Sequence Object Reference=Get LSTemp(), Option 1:Level Sequence Object Reference=@n1, Index:bool=IsValid(Object=@n1)) -> (ReturnValue:Level Sequence Object Reference)) -> (Output_Get:Level Sequence Object Reference)
   n3 = Get cutTrack() -> (cutTrack:Movie Scene Camera Cut Track Object Reference)
   n4 = IsValid(Object=@n3)
   > n5 = Branch(Condition=@n4)
   n6 = AddTrack!(
    execute=@n5.else
    Sequence:Movie Scene Sequence Object Reference=Get curSeq() -> (curSeq:Level Sequence Object Reference)
    TrackType=/Script/MovieSceneTracks.MovieSceneCameraCutTrack
   ) -> (then, ReturnValue:Movie Scene Camera Cut Track Object Reference)
   Set curCutTrack(
    execute=[@n15.then, @n5.then]
    curCutTrack:Movie Scene Camera Cut Track Object Reference=Select(Option 0:Movie Scene Camera Cut Track Object Reference=@n6, Option 1:Movie Scene Camera Cut Track Object Reference=@n3, Index=@n4) -> (ReturnValue:Movie Scene Camera Cut Track Object Reference)
   ) -> (Output_Get:Movie Scene Camera Cut Track Object Reference)
   n7 = AddSection(execute=@n2.then_1, Track=Get curCutTrack() -> (curCutTrack:Movie Scene Camera Cut Track Object Reference))
   > n8 = Cast To MovieSceneCameraCutSection!(execute, Object=@n7) -> (then, CastFailed, AsMovie Scene Camera Cut Section:Movie Scene Camera Cut Section Object Reference) {is=title K2Node_DynamicCast}
   > n9 = Set curSection(curSection:Movie Scene Camera Cut Section Object Reference=@n8) -> (Output_Get:Movie Scene Camera Cut Section Object Reference)
   > n10 = GetAutoSizeHasStartFrame!(execute, Section:Movie Scene Section Object Reference=@n9) -> (then, ReturnValue:bool)
   > PrintString(InString=Append(A="Section AutoSize: ", B="To String (Boolean)"(InBool=@n10), C=" :: "))
   n11 = Get start() -> (start:int=0)
   n12 = int - int(A=@n11, B=Get preroll() -> (preroll:int=0))
   > SetStartFrame(Section=@n9, StartFrame=SelectInt(A=@n12, B=@n11, bPickA="Equal (Integer)"(A=@n11)))
   n13 = Get end() -> (end:int=0)
   > SetEndFrame(Section=@n9, EndFrame=@n13)
   > n14 = GetBindingID!(
    execute
    Sequence:Movie Scene Sequence Object Reference=Get LSTemp()
    InBinding:"Movie Scene Binding Proxy Structure (by ref)"=Get CameraBinding() -> (CameraBinding:Movie Scene Binding Proxy Structure)
   ) -> (then, ReturnValue:Movie Scene Object Binding ID Structure)
   > SetCameraBindingID!(
    execute
    self:Movie Scene Camera Cut Section Object Reference=Get curSection() -> (curSection:Movie Scene Camera Cut Section Object Reference)
    InCameraBindingID:"Movie Scene Object Binding ID Structure (by ref)"=@n14
   ) -> (then)
   > PrintString(InString=Append(A="Added Cut Section: ", B="To String (Integer)"(InInt=@n12), C=" :: ", D="To String (Integer)"(InInt=@n13)))
   Return Node!(
    execute=@n2.then_2
    outCutTrack:Movie Scene Camera Cut Track Object Reference=Get curCutTrack()
    outSection:Movie Scene Camera Cut Section Object Reference=Get curSection()
   )
   n15 = PrintString(execute=@n6.then, InString="New Cut Track Added!")
  ]}
 addFloatTrack {nodes [
   n1 = addFloatTrack!() -> (then, binding:Movie Scene Binding Proxy Structure, duration:int, property:Name, propertyPath:String, noPath:bool=true) {is=title K2Node_FunctionEntry}
   > n2 = AddTrack!(execute, InBinding:"Movie Scene Binding Proxy Structure (by ref)"=@n1, TrackType=/Script/MovieSceneTracks.MovieSceneFloatTrack) -> (then, ReturnValue:Movie Scene Track Object Reference)
   > n3 = Cast To MovieSceneFloatTrack!(execute, Object=@n2) -> (then, CastFailed, AsMovie Scene Float Track:Movie Scene Float Track Object Reference) {is=title K2Node_DynamicCast}
   > SetPropertyNameAndPath!(
    execute
    Track:Movie Scene Property Track Object Reference=@n3
    InPropertyName:"Name (by ref)"=@n1.property
    InPropertyPath:String=SelectString(A="To String (Name)"(InName=@n1.property), B=@n1.propertyPath, bPickA=@n1.noPath)
   ) -> (then)
   > n4 = AddSection(Track=@n3)
   > SetEndFrame(Section=@n4, EndFrame=Get duration() -> (duration:int=0))
   > n5 = GetAllChannels(Section=@n4)
   > n6 = Cast To MovieSceneScriptingFloatChannel!(execute, Object="Get (a copy)"(Array=@n5)) -> (then, CastFailed, AsMovie Scene Scripting Float Channel:Movie Scene Scripting Float Channel Object Reference) {is=title K2Node_DynamicCast}
   > Return Node!(
    execute
    valid:bool=true
    FloatTrack:Movie Scene Float Track Object Reference=@n3
    Section:Movie Scene Section Object Reference=@n4
    FloatChannel:Movie Scene Scripting Float Channel Object Reference=@n6
   )
   Return Node(execute=[@n6.CastFailed, @n3.CastFailed], valid=false)
  ]}
 addGroupKey {nodes [
   n1 = addGroupKey!() -> (then, binding:Movie Scene Binding Proxy Structure, group:ENU_TransformGroups Enum=NewEnumerator0, value:double=~, frame:int) {is=title K2Node_FunctionEntry}
   n2 = Find(
    TargetMap:Map of Movie Scene Binding Proxy Structures to ST Transform Groups Structures=Get TransformData()
    Key:"Movie Scene Binding Proxy Structure (by ref)"=@n1
   ) -> (Value:ST Transform Groups Structure)
   n3 = Break ST Transform Channels!(ST_TransformChannels:"ST Transform Channels Structure (by ref)"=Find(
     TargetMap:Map of ENU_TransformGroups Enums to ST Transform Channels Structures=Break ST Transform Groups!(ST_TransformGroups:"ST Transform Groups Structure (by ref)"=@n2) -> (ChannelGroup_8_#3:Map of ENU_TransformGroups Enums to ST Transform Channels Structures) {is=title K2Node_BreakStruct}
     Key:"ENU_TransformGroups Enum (by ref)"=@n1.group
    ) -> (Value:ST Transform Channels Structure)) -> (TransformChannels_7_#3:Map of ENU_TransformChannels Enums to Movie Scene Scripting Double Channel Object References) {is=title K2Node_BreakStruct}
   > n4 = Keys!(execute, TargetMap:Map of ENU_TransformChannels Enums to Movie Scene Scripting Double Channel Object References=@n3) -> (then, Keys:Array of ENU_TransformChannels Enums)
   > n5 = For Each Loop(Array:Array of ENU_TransformChannels Enums=@n4) -> (Array Element:ENU_TransformChannels Enum)
   > "Add Key (Double)"!(
    execute
    self:Movie Scene Scripting Double Channel Object Reference=Find(
     TargetMap:Map of ENU_TransformChannels Enums to Movie Scene Scripting Double Channel Object References=@n3
     Key:"ENU_TransformChannels Enum (by ref)"=@n5
    ) -> (Value:Movie Scene Scripting Double Channel Object Reference)
    InTime:"Frame Number Structure (by ref)"=Make Frame Number(Value=Get frame() -> (frame:int=0))
    NewValue:double=Get value() -> (value:double)
    SubFrame:float=0
    TimeUnit:EMovieSceneTimeUnit Enum=DisplayRate
    InInterpolation:EMovieSceneKeyInterpolation Enum=Linear
   ) -> (then, ReturnValue:Movie Scene Scripting Double Key Object Reference)
  ]}
 addGroupVectorKey {nodes [
   n1 = addGroupVectorKey!() -> (then, binding:Movie Scene Binding Proxy Structure, group:ENU_TransformGroups Enum=NewEnumerator0, vector:Vector=~, frame:int) {is=title K2Node_FunctionEntry}
   n2 = Find(
    TargetMap:Map of Movie Scene Binding Proxy Structures to ST Transform Groups Structures=Get TransformData()
    Key:"Movie Scene Binding Proxy Structure (by ref)"=@n1
   ) -> (Value:ST Transform Groups Structure)
   > n3 = Branch(Condition=@n2.ReturnValue)
   n4 = Find(
    TargetMap:Map of ENU_TransformGroups Enums to ST Transform Channels Structures=Break ST Transform Groups(ST_TransformGroups=@n2)
    Key:"ENU_TransformGroups Enum (by ref)"=@n1.group
   ) -> (Value:ST Transform Channels Structure)
   > n5 = Branch(Condition=@n4.ReturnValue)
   n6 = Break ST Transform Channels(ST_TransformChannels=@n4)
   > n7 = Keys(TargetMap=@n6)
   > n8 = For Each Loop(Array=@n7)
   n9 = Find(
    TargetMap:Map of ENU_TransformChannels Enums to Movie Scene Scripting Double Channel Object References=@n6
    Key:"ENU_TransformChannels Enum (by ref)"=@n8
   ) -> (Value:Movie Scene Scripting Double Channel Object Reference)
   > n10 = Branch(Condition=@n9.ReturnValue)
   > n11 = Is Valid(InputObject=@n9)
   PrintString(execute=@n11.Is Not Valid, InString=Append!(A="Channel Key: ", B=Enum to String(Enumerator=@n8), C=" - not Valid!") -> (ReturnValue))
   n12 = BreakVector(InVec=Get vector() -> (vector:Vector))
   "Add Key (Double)"(
    execute=@n11.Is Valid
    self=@n9
    InTime=Make Frame Number(Value=Get frame() -> (frame=0))
    NewValue=Select!(NewEnumerator0:double=@n12, NewEnumerator1:double=@n12.Y, NewEnumerator2:double=@n12.Z, Index:ENU_TransformChannels Enum=@n8) -> (ReturnValue:double)
    SubFrame=0
    TimeUnit=DisplayRate
    InInterpolation=Linear
   )
   PrintString(execute=@n10.else, InString="Cannot find Channel!")
   PrintString(execute=@n5.else, InString="Cannot find Group!")
   PrintString(execute=@n3.else, InString="Cannot find Binding!")
   PrintString(InString="Vector Key Added!")
  ]}
 addVisiblityTrack {nodes [
   n1 = addVisiblityTrack!() -> (then, InBinding:"Movie Scene Binding Proxy Structure (by ref)", duration:int) {is=title K2Node_FunctionEntry}
   > n2 = AddTrack(InBinding=@n1, TrackType=/Script/MovieSceneTracks.MovieSceneVisibilityTrack)
   > n3 = Cast To MovieSceneVisibilityTrack!(execute, Object=@n2) -> (then, CastFailed, AsMovie Scene Visibility Track:Movie Scene Visibility Track Object Reference) {is=title K2Node_DynamicCast}
   > SetPropertyNameAndPath(Track=@n3, InPropertyName=MakeLiteralName(Value=Visibility), InPropertyPath=bHidden)
   > n4 = AddSection(Track=@n3)
   > SetEndFrame(Section=@n4, EndFrame=@n1.duration)
   > n5 = GetAllChannels(Section=@n4)
   > n6 = Cast To MovieSceneScriptingBoolChannel!(execute, Object="Get (a copy)"(Array=@n5)) -> (then, CastFailed, AsMovie Scene Scripting Bool Channel:Movie Scene Scripting Bool Channel Object Reference) {is=title K2Node_DynamicCast}
   n7 = Get InBinding() -> (InBinding:"Movie Scene Binding Proxy Structure (by ref)")
   > Return Node!(
    execute
    success:bool=true
    binding:Movie Scene Binding Proxy Structure=@n7
    boolChannel:Movie Scene Scripting Bool Channel Object Reference=@n6
   )
   Return Node(execute=@n6.CastFailed, success=false, binding=@n7, boolChannel=@n6)
  ]}
 cineCameraSetup {nodes [
   n1 = FindTracksByExactType!(
    execute
    InBinding:"Movie Scene Binding Proxy Structure (by ref)"
    TrackType:Movie Scene Track Class Reference=/Script/MovieSceneTracks.MovieSceneObjectPropertyTrack
   ) -> (then, ReturnValue:Array of Movie Scene Object Property Track Object References)
   n2 = Length(TargetArray:Array of Movie Scene Object Property Track Object References=@n1)
   > PrintString(InString=Append!(A="Camera Tracks: ", B="To String (Integer)"(InInt=@n2)) -> (ReturnValue))
   > n3 = For Each Loop(Array:Array of Movie Scene Object Property Track Object References=@n1) -> (Array Element:Movie Scene Object Property Track Object Reference)
   > n4 = AddSection(Track=@n3)
   n5 = AddTrack(TrackType=&MovieSceneObjectPropertyTrack)
   > n6 = AddSection(Track=@n5)
   n7 = GetAllChannels(execute=[@n4.then, @n6.then], Section=@n6)
   > n8 = For Each Loop(Array:Array of Movie Scene Scripting Channel Object References=@n7) -> (Array Element:Movie Scene Scripting Channel Object Reference)
   > PrintString(InString="To String (Name)"(InName=Get ChannelName(self=@n8) -> (ChannelName=None)))
   n9 = AddTrack!(execute, Sequence:Movie Scene Sequence Object Reference=Get LSTemp(), TrackType=&MovieSceneCameraCutTrack) -> (then, ReturnValue:Movie Scene Camera Cut Track Object Reference)
   > n10 = AddSection(Track=@n9)
   > SetEndFrame(Section=@n10, EndFrame=Get localEndFrame() -> (localEndFrame:int=0))
   > n11 = Cast To MovieSceneCameraCutSection(Object=@n10)
   n12 = SpawnActor Cine Camera Actor!(
    execute:Exec=@n27.then_0
    WorldContextObject:Object Reference
    Class:Actor Class Reference=/Script/CinematicCamera.CineCameraActor
    SpawnTransform:Transform=MakeTransform(Rotation="0, 0, 0")
    CollisionHandlingOverride:ESpawnActorCollisionHandlingMethod Enum=Undefined
    TransformScaleMethod:ESpawnActorScaleMethod Enum=MultiplyWithRoot
    Owner:Actor Object Reference
    Instigator:Pawn Object Reference
   ) -> (then:Exec, ReturnValue:Cine Camera Actor Object Reference) {is=title K2Node_SpawnActorFromClass}
   > n13 = Set cameraActor(cameraActor:Cine Camera Actor Object Reference=@n12) -> (Output_Get:Cine Camera Actor Object Reference)
   > n14 = AddSpawnableFromInstance(self=Get LevelSequenceEditorSubsystem(), Sequence=Get LSTemp(), ObjectToSpawn=@n13)
   > n15 = Set cameraBinding(cameraBinding:Movie Scene Binding Proxy Structure=@n14) -> (Output_Get:Movie Scene Binding Proxy Structure)
   n16 = GetBindingID(execute=@n11.then, Sequence=Get LSTemp(), InBinding=@n15)
   > n17 = SetCameraBindingID(self=@n11, InCameraBindingID=@n16)
   Destroy Actor(execute=[@n17.then, @n34.then], self=Get cameraActor() -> (cameraActor:Cine Camera Actor Object Reference))
   n18 = Cast To CineCameraActor!(execute, Object:Wildcard) -> (then, CastFailed, AsCine Camera Actor:Cine Camera Actor Object Reference) {is=title K2Node_DynamicCast}
   > n19 = AddPossessable(
    Sequence=Get LSTemp()
    ObjectToPossess=GetCineCameraComponent!(self:Cine Camera Actor Object Reference=@n18) -> (ReturnValue:Cine Camera Component Object Reference)
   )
   > n20 = Set cameraComponent(cameraComponent:Movie Scene Binding Proxy Structure=@n19) -> (Output_Get:Movie Scene Binding Proxy Structure)
   > SetParent!(execute, InBinding:"Movie Scene Binding Proxy Structure (by ref)"=@n20, InParentBinding:"Movie Scene Binding Proxy Structure (by ref)") -> (then)
   > n21 = AddTrack!(execute, InBinding:"Movie Scene Binding Proxy Structure (by ref)"=@n20, TrackType=&MovieSceneFloatTrack) -> (then, ReturnValue:Movie Scene Track Object Reference)
   > n22 = Cast To MovieSceneFloatTrack(Object=@n21)
   > n23 = AddSection(Track=@n22)
   > SetEndFrame(Section=@n23, EndFrame=Get CDuration() -> (CDuration=0))
   PrintString(execute=@n18.CastFailed, InString="Cast Failed!")
   n24 = cineCameraSetup!() -> (then, cutTrack:Movie Scene Camera Cut Track Object Reference, startFrame:int, endFrame:int, fPrerollFrames:int) {is=title K2Node_FunctionEntry}
   n25 = "Equal (Integer)"(A=@n24.startFrame, B=@n24.endFrame)
   n26 = "Equal (Integer)"(A=@n24.startFrame)
   > Set localEndFrame(localEndFrame:int=SelectInt(A=Get SeqFrames() -> (SeqFrames=0), B=@n24.endFrame, bPickA=AND Boolean(A=@n26, B=@n25))) -> (Output_Get:int=0)
   > n27 = Sequence()
   Return Node!(
    execute=@n27.then_2
    CameraOutBinding:Movie Scene Binding Proxy Structure=Get cameraBinding() -> (cameraBinding:Movie Scene Binding Proxy Structure)
    CameraOutComponentBinding:Movie Scene Binding Proxy Structure=Get cameraComponent() -> (cameraComponent:Movie Scene Binding Proxy Structure)
    outCutTrack:Movie Scene Camera Cut Track Object Reference=Get localCutTrack()
   )
   n28 = addFloatTrack!(
    execute=@n27.then_1
    self
    binding:Movie Scene Binding Proxy Structure=Get cameraComponent()
    duration:int=Get localEndFrame() -> (localEndFrame=0)
    property:Name=MakeLiteralName(Value=CurrentFocalLength)
    propertyPath:String
    noPath:bool=true
   ) -> (
    then
    valid:bool
    FloatTrack:Movie Scene Float Track Object Reference
    Section:Movie Scene Section Object Reference
    FloatChannel:Movie Scene Scripting Float Channel Object Reference
   ) {is=title K2Node_CallFunction}
   > n29 = Branch(Condition=@n28)
   > "Add Key (Float)"!(
    execute
    self:Movie Scene Scripting Float Channel Object Reference=@n28.FloatChannel
    InTime:"Frame Number Structure (by ref)"=Make Frame Number(Value=0)
    NewValue:float=FOVtoFocalLength!(fov:double=Get cameraFOV() -> (cameraFOV:double)) -> (ReturnValue:double) {is=title K2Node_CallFunction}
    SubFrame:float="To Float (Integer)"(InInt=Get startFrame() -> (startFrame=0))
    TimeUnit:EMovieSceneTimeUnit Enum=DisplayRate
    InInterpolation:EMovieSceneKeyInterpolation Enum=Auto
   ) -> (then, ReturnValue:Movie Scene Scripting Float Key Object Reference)
   PrintString(execute=@n29.else, InString=Key NOT Added)
   PrintString(InString=Key Added)
   ConvertToSpawnable(self=Get LevelSequenceEditorSubsystem())
   n30 = GetCineCameraComponent(self=@n13)
   n31 = AddPossessable(execute=@n15.then, Sequence=Get LSTemp(), ObjectToPossess=@n30)
   > n32 = Set cameraComponent(cameraComponent=@n31)
   > SetParent(InBinding=@n32, InParentBinding=@n15)
   > n33 = addCameraCut!(
    execute
    self
    Sequence:Level Sequence Object Reference=Get LSTemp()
    CameraBinding:Movie Scene Binding Proxy Structure=Get cameraBinding()
    cutTrack:Movie Scene Camera Cut Track Object Reference=Get cutTrack()
    start:int=Get startFrame() -> (startFrame=0)
    end:int=Get localEndFrame() -> (localEndFrame=0)
    preroll:int=Get fPrerollFrames() -> (fPrerollFrames:int=0)
   ) -> (then, outCutTrack:Movie Scene Camera Cut Track Object Reference, outSection:Movie Scene Camera Cut Section Object Reference) {is=title K2Node_CallFunction}
   > n34 = Set localCutTrack(localCutTrack=@n33)
  ]}
 cleanupTempSequences {nodes [
   cleanupTempSequences() {is=title K2Node_FunctionEntry}
   n1 = Get tempDir()
   n2 = int - int(A=Len(S=@n1))
   > n3 = GetAssetsByPath!(
    execute
    self:Asset Registry Interface=GetAssetRegistry()
    PackagePath:Name=String To Name(InString=GetSubstring(SourceString=@n1, StartIndex=0, Length=@n2))
    bRecursive:bool=false
    bIncludeOnlyOnDiskAssets:bool=false
   ) -> (then, OutAssetData:Array of Asset Data Structures, ReturnValue:bool)
   > Branch(Condition=@n3.ReturnValue)
   > n4 = Set assets(assets:Array of Asset Data Structures=@n3) -> (Output_Get:Array of Asset Data Structures)
   > n5 = For Each Loop(Array:Array of Asset Data Structures=@n4) -> (Array Element:Asset Data Structure)
   n6 = Get sequences() -> (sequences:String)
   n7 = Show Message Dialog(
    execute=@n5.Completed
    Title=MakeLiteralText(Value="LOCTEXT(\"Cleanup Temp Sequences\")")
    Message="To Text (String)"(InString=Append(A="The following files will be deleted: \r\n\r\n", B=@n6))
    MessageType=OkCancel
    DefaultValue=Cancel
    MessageCategory=Info
   )
   > n8 = Switch on EAppReturnType(Selection=@n7)
   n9 = For Each Loop(Exec=@n8.Ok, Array=Get assets() -> (assets:Array of Asset Data Structures))
   DeleteLoadedAssets!(execute=@n9.Completed, AssetsToDelete:Array of Object References=Get assetsToDel() -> (assetsToDel:Array of Object References)) -> (then, ReturnValue:bool)
   Add(
    execute=@n9.LoopBody
    TargetArray:Array of Object References=Get assetsToDel()
    NewItem:"Object Reference (by ref)"=GetAsset(InAssetData=@n9)
   ) {is=title K2Node_CallArrayFunction}
   n10 = Break Asset Data(AssetData=@n5)
   Set sequences(execute=@n5.LoopBody, sequences:String=Append(A="To String (Name)"(InName=@n10), B="\r\n")) -> (Output_Get:String)
  ]}
 createTODEvent {nodes [
   n1 = createTODEvent!() -> (
    then
    InSequence:Movie Scene Sequence Object Reference
    InSection:Movie Scene Event Section Base Object Reference
    channel:Movie Scene Scripting Event Channel Object Reference
    InEndpoint:"Sequencer Quick Binding Result Structure (by ref)"
    InTOD:double=~
    InTime:"Frame Number Structure (by ref)"
   ) {is=title K2Node_FunctionEntry}
   > n2 = CreateEvent!(
    execute
    InSequence:Movie Scene Sequence Object Reference=@n1
    InSection:Movie Scene Event Section Base Object Reference=@n1.InSection
    InEndpoint:"Sequencer Quick Binding Result Structure (by ref)"=@n1.InEndpoint
    InPayload:Array of Strings=Make Array("[0]":String="To String (Float)"(InDouble=@n1.InTOD)) -> (Array:Array of Strings)
   ) -> (then, ReturnValue:Movie Scene Event Structure)
   > n3 = "Add Key (Event)"!(
    execute
    self:Movie Scene Scripting Event Channel Object Reference=@n1.channel
    InTime:"Frame Number Structure (by ref)"=@n1.InTime
    NewValue:Movie Scene Event Structure=@n2
    SubFrame:float=0
    TimeUnit:EMovieSceneTimeUnit Enum=DisplayRate
   ) -> (then, ReturnValue:Movie Scene Scripting Event Key Object Reference)
   > Return Node!(execute, outKey:Movie Scene Scripting Event Key Object Reference=@n3)
  ]}
 debugFunc {nodes [
   debugFunc() {is=title K2Node_FunctionEntry}
   > Is Valid(InputObject=Get selectedActor())
   > Set SeqFrames(SeqFrames=getTotalDuration() -> (sequenceFrames=0)) -> (Output_Get=0)
   > CreateTempSequence()
   n1 = Get TODActor()
   > Is Valid(InputObject=@n1)
   > n2 = AddPossessable(Sequence=Get LSTemp(), ObjectToPossess=@n1)
   > n3 = AddTrack(InBinding=@n2, TrackType=&MovieSceneFloatTrack)
   > n4 = Cast To MovieSceneFloatTrack(Object=@n3)
   n5 = MakeLiteralString(Value=CurTimeOfDay)
   > SetPropertyNameAndPath(Track=@n4, InPropertyName=String To Name(InString=@n5), InPropertyPath=@n5)
   > n6 = AddSection(Track=@n4)
   > SetEndFrame(Section=@n6, EndFrame=Get SeqFrames() -> (SeqFrames=0))
   > n7 = GetAllChannels(Section=@n6)
   > n8 = Cast To MovieSceneScriptingFloatChannel(Object="Get (a copy)"(Array=@n7))
   > "Add Key (Float)"(
    self=@n8
    InTime=Make Frame Number(Value=0)
    NewValue=GetValue(self=Get TODMinTime())
    SubFrame=0
    TimeUnit=DisplayRate
    InInterpolation=Auto
   )
   > "Add Key (Float)"(
    self=@n8
    InTime=Make Frame Number(Value=Get SeqFrames() -> (SeqFrames=0))
    NewValue=GetValue(self=Get TODMaxTime())
    SubFrame=0
    TimeUnit=DisplayRate
    InInterpolation=Auto
   )
  ]}
 disableNanite {nodes [
   n1 = disableNanite!() -> (then, object:Object Reference, disable:bool=~) {is=title K2Node_FunctionEntry}
   > n2 = Cast To StaticMeshActor!(execute, Object=@n1) -> (then, CastFailed, AsStatic Mesh Actor:Static Mesh Actor Object Reference) {is=title K2Node_DynamicCast}
   > Set bDisallowNanite!(execute, bDisallowNanite:bool=Get disable() -> (disable:bool), self:Static Mesh Component Object Reference=Get StaticMeshComponent(self=@n2)) -> (then, Output_Get:bool)
   > Return Node!(
    execute
    isSM:bool=true
    isProp:bool=false
    failed:bool=false
    asSM:Static Mesh Actor Object Reference=@n2
    asProp:Building Prop Object Reference
   )
   n3 = Cast To BuildingProp!(execute=@n2.CastFailed, Object=@n1) -> (then, CastFailed, AsBuilding Prop:Building Prop Object Reference) {is=title K2Node_DynamicCast}
   > Set bDisallowNanite(
    bDisallowNanite=Get disable()
    self=Get StaticMeshComponent(self:Building Static Mesh Object Reference=@n3) -> (StaticMeshComponent:Base Building Static Mesh Component Object Reference)
   )
   > Return Node(isSM=false, isProp=true, failed=false, asProp=@n3)
   Return Node(execute=@n3.CastFailed, isSM=false, isProp=false, failed=true)
   SetForceDisableNanite!(execute, self:Static Mesh Component Object Reference, bInForceDisableNanite:bool=false) -> (then)
  ]}
 fetchUGSChange {nodes [
   fetchUGSChange() {is=title K2Node_FunctionEntry}
   > n1 = Sequence!(execute) -> (then_0, then_1)
   > n2 = DoesSaveGameExist!(execute, SlotName:String=SHOWCASE_Globals, UserIndex:int=0) -> (then, ReturnValue:bool)
   > n3 = Branch(Condition=@n2)
   > n4 = LoadGameFromSlot!(execute, SlotName:String=SHOWCASE_Globals, UserIndex:int=0) -> (then, ReturnValue:Save Game Object Reference)
   > n5 = Cast To SG_ShowcaseRel!(execute, Object=@n4) -> (then, CastFailed, AsSG Showcase Rel:SG Showcase Rel Object Reference) {is=title K2Node_DynamicCast}
   > n6 = Set SaveData(SaveData:SG Showcase Rel Object Reference=@n5) -> (Output_Get:SG Showcase Rel Object Reference)
   n7 = Split(SourceString=Split(SourceString=GetProjectDirectory!() -> (ReturnValue:String), InStr=/, SearchDir=FromEnd), InStr=/, SearchDir=FromEnd)
   n8 = Execute Python Script!(
    execute=[@n6.then, @n11.then]
    PythonScript="import json\r\nimport unreal\r\n\r\nunreal.log(\"CineTools - Fetching Current UGS Changelist\")\r\ntry:\r\n    with open (path, \"r\") as openfile:\r\n        data = json.load(openfile)\r\n\r\n    change = data.get(\"CurrentChangeNumber\")\r\n    unreal.log(\"CineTools - UGS Changelist Found - \" + str(change))\r\n\r\nexcept IOError:\r\n    unreal.log_warning(\"CineTools - Unable to determine Current UGS Changelist\")\r\n    change = -1"
    path:String="To String (Text)"(InText=Format Text!(Format="LOCTEXT(\"{BUILD}/.ugs/state.json\")", BUILD:String=@n7) -> (Result))
   ) -> (then, ReturnValue, change:int) {inputs [path], outputs [change]}
   > n9 = Branch(Condition=@n8)
   > Set changelist!(execute, changelist:int=@n8.change, self:SG Showcase Rel Object Reference=Get SaveData() -> (SaveData:SG Showcase Rel Object Reference)) -> (then, Output_Get:int=0)
   printMessage(
    execute=[@n9.else, @n5.CastFailed]
    Message_Message_4_#1="LOCTEXT(\"Warning: could not get the current Changelist!\")"
    Message_Level_10_#2=NewEnumerator1
    displayTime=0
   )
   n10 = CreateSaveGameObject!(execute=@n3.else, SaveGameClass:Save Game Class Reference=/Game/NeverCook/ArtTools/AssetShowcase/Data/SG_ShowcaseRel.SG_ShowcaseRel_C) -> (then, ReturnValue:SG Showcase Rel Object Reference)
   > n11 = Set SaveData(SaveData=@n10)
   n12 = Get SaveData()
   Set assetName!(
    execute=@n1.then_1
    assetName:String=GetActorLabel(self=Get selectedActor(), bCreateIfNone=false)
    self:SG Showcase Rel Object Reference=@n12
   ) -> (then, Output_Get:String)
   > SaveGameToSlot!(execute, SaveGameObject:Save Game Object Reference=@n12, SlotName:String=SHOWCASE_Globals, UserIndex:int=0) -> (then, ReturnValue:bool)
  ]}
 getConfig {nodes [
   getConfig() {is=title K2Node_FunctionEntry}
   n1 = Literal enum ENU_ConfigOptions!(Enum:ENU_ConfigOptions Enum=NewEnumerator1) -> (ReturnValue:ENU_ConfigOptions Enum) {is=title K2Node_EnumLiteral}
   n2 = Literal enum ENU_ConfigOptions(Enum=NewEnumerator0)
   n3 = Get useTODRange() -> (useTODRange:bool)
   > Return Node!(
    execute
    Value:Movie Pipeline Primary Config Object Reference=Find(
     TargetMap:Map of ENU_ConfigOptions Enums to Movie Pipeline Primary Config Object References=Get LSConfigMap() -> (LSConfigMap:Map of ENU_ConfigOptions Enums to Movie Pipeline Primary Config Object References=NewEnumerator0)
     Key:"ENU_ConfigOptions Enum (by ref)"=Select!(Option 0:"ENU_ConfigOptions Enum (by ref)"=@n2, Option 1:"ENU_ConfigOptions Enum (by ref)"=@n1, Index:bool=@n3) -> (ReturnValue:"ENU_ConfigOptions Enum (by ref)")
    ) -> (Value:Movie Pipeline Primary Config Object Reference)
   )
  ]}
 getParentFolder {nodes [
   n1 = getParentFolder!() -> (then, InPath:String) {is=title K2Node_FunctionEntry}
   > n2 = Set localCache(localCache:Array of Strings=ParseIntoArray(SourceString=@n1, Delimiter=/)) -> (Output_Get:Array of Strings)
   n3 = Last Index!(TargetArray=@n2) -> (ReturnValue=0)
   > Remove Index!(execute, TargetArray=@n2, IndexToRemove:int=@n3) -> (then)
   n4 = Get localCache() -> (localCache:Array of Strings)
   > Return Node!(execute, OutPath:String=Append!(A=/, B=JoinStringArray(SourceArray=@n4, Separator=/), C:String=/) -> (ReturnValue))
   int - int(A=@n3)
  ]}
 getScaleActorLocation {nodes [
   n1 = getScaleActorLocation!() -> (then, selActor:Actor Object Reference) {is=title K2Node_FunctionEntry}
   n2 = GetLevelViewportCameraInfo(WorldContextObject=Get LevelEditorSubsystem())
   n3 = "To Transform (Rotator)"!(InRotator:"Rotator (by ref)"=@n2.CameraRotation) -> (ReturnValue:Transform)
   > Set offset(offset:Vector=vector * vector(
     A=GetTransformAxisVector!(Transform:Transform=@n3, Axis:EGeometryScriptAxis Enum=Y) -> (ReturnValue:Vector)
     B=Get YOffset() -> (YOffset:double)
    ) -> (ReturnValue=~)) -> (Output_Get:Vector)
   > n4 = For Loop with Break!(Execute:Exec, FirstIndex:int, LastIndex:int=Get maxIterations() -> (maxIterations:int=0), Break:Exec=@n13.then) -> (LoopBody:Exec, Index:int, Completed:Exec) {is=title K2Node_MacroInstance}
   n5 = Get outLocation() -> (outLocation:Vector)
   n6 = BreakVector(InVec=@n5)
   n7 = Get YOffset()
   Return Node!(
    execute=@n4.Completed
    spawnLocation:Vector=MakeVector(X=@n6, Y=float + float(A=@n6.Y, B=@n7), Z=@n6.Z)
    ViewLookat_Roll:float=0
    ViewLookat_Pitch:float=0
    ViewLookat_Yaw:float=BreakRotator(InRot=FindLookAtRotation!(Start:"Vector (by ref)"=@n5, Target:"Vector (by ref)"=@n2) -> (ReturnValue:Rotator)).Yaw
   )
   n8 = vector + vector(
    A=Get Actor Bounds(self=@n1, bOnlyCollidingComponents=false)
    B=vector * vector(A:int=@n4, B:Vector=Get offset() -> (offset:Vector)) -> (ReturnValue=~)
   )
   n9 = BreakVector(InVec=@n8)
   > n10 = Line Trace For Objects!(
    execute
    WorldContextObject:Object Reference
    Start:Vector=@n8
    End:Vector=MakeVector(X=@n9, Y=@n9.Y, Z=float - float(A=@n9.Z, B=5000))
    ObjectTypes:Array of EObjectTypeQuery Enums=Make Array("[0]":EObjectTypeQuery Enum=ObjectTypeQuery1) -> (Array:Array of EObjectTypeQuery Enums)
    bTraceComplex:bool=true
    ActorsToIgnore:Array of Actor Object References
    DrawDebugType:EDrawDebugTrace Enum
    bIgnoreSelf:bool=false
    TraceColor:Linear Color Structure=(R=1,G=0,B=0,A=1)
    TraceHitColor:Linear Color Structure=(R=0,G=1,B=0,A=1)
    DrawTime:float=10
   ) -> (then, OutHit:Hit Result Structure, ReturnValue:bool)
   > Branch(Condition=@n10.ReturnValue)
   n11 = BreakHitResult!(Hit:"Hit Result Structure (by ref)"=@n10) -> (
    bBlockingHit:bool
    bInitialOverlap:bool
    Time:float
    Distance:float
    Location:Vector
    ImpactPoint:Vector
    Normal:Vector
    ImpactNormal:Vector
    PhysMat:Physical Material Object Reference
    HitActor:Actor Object Reference
    HitComponent:Primitive Component Object Reference
    HitBoneName:Name
    BoneName:Name
    HitItem:int
    ElementIndex:int
    FaceIndex:int
    TraceStart:Vector
    TraceEnd:Vector
   )
   > n12 = Branch(Condition="Equal (Object)"(A=Get selActor() -> (selActor:Actor Object Reference), B=@n11.HitActor))
   n13 = Set outLocation(execute=@n12.else, outLocation:Vector=vector + vector(A=@n11.ImpactPoint, B=Get offset())) -> (Output_Get:Vector)
  ]}
 getSelectedAsset {nodes [
   n1 = getSelectedAsset!() -> (then, set:Typed Element Selection Set Object Reference) {is=title K2Node_FunctionEntry}
   > n2 = Is Valid(InputObject=@n1)
   Return Node!(
    execute=@n2.Is Not Valid
    success:bool=false
    outSM:Static Mesh Object Reference
    outActor:Actor Object Reference
    outClass:Actor Class Reference
   )
   > n3 = GetSelectedObjects!(execute, self:Typed Element Selection Set Object Reference=@n1, InRequiredClass:Object Class Reference=&Actor) -> (then, ReturnValue:Array of Object References)
   > n4 = Branch(Condition=Is Valid Index(TargetArray:Array of Object References=@n3))
   > n5 = Cast To Actor!(execute, Object="Get (a copy)"(Array:Array of Object References=@n3) -> (Output:Object Reference)) -> (then, CastFailed, AsActor:Actor Object Reference) {is=title K2Node_DynamicCast}
   > n6 = Set selectedActor(selectedActor=@n5)
   > n7 = Cast To FortStaticMeshActor(Object=@n6)
   > n8 = Set selectedActorClass(selectedActorClass:Actor Class Reference=&FortStaticMeshActor) -> (Output_Get:Actor Class Reference)
   > n9 = Set selectedActorSM(selectedActorSM=Get StaticMesh(self=Get StaticMeshComponent(self:Static Mesh Actor Object Reference=@n7) -> (StaticMeshComponent:Static Mesh Component Object Reference)))
   n10 = Get selectedActor()
   > Return Node(success=true, outSM=@n9, outActor=@n10, outClass=@n8)
   n11 = Cast To BuildingSMActor(execute=@n7.CastFailed, Object=@n6)
   > n12 = Set selectedActorClass(selectedActorClass=&BuildingSMActor)
   > n13 = Set selectedActorSM(selectedActorSM=Get StaticMesh(self:Building Static Mesh Object Reference=@n11))
   > Return Node(success=true, outSM=@n13, outActor=@n10, outClass=@n12)
   Return Node(execute=@n11.CastFailed, success=false)
   Return Node(execute=[@n5.CastFailed, @n4.else], success=false)
  ]}
 getSequenceFrames {nodes [
   n1 = getSequenceFrames!() -> (then, levelSeq:Level Sequence Object Reference, frames:int) {is=title K2Node_FunctionEntry}
   > n2 = GetTickResolution!(execute, Sequence:Movie Scene Sequence Object Reference=@n1) -> (then, ReturnValue:Frame Rate Structure)
   > n3 = GetDisplayRate!(execute, Sequence:Movie Scene Sequence Object Reference=@n1) -> (then, ReturnValue:Frame Rate Structure)
   > Return Node!(
    execute
    Frame Number Value:int=Break Frame Time!(FrameTime:"Frame Time Structure (by ref)"=Transform Frame Time!(
      SourceTime_FrameNumber_Value:int=@n1.frames
      SourceTime_SubFrame:float=0
      SourceRate:"Frame Rate Structure (by ref)"=@n3
      DestinationRate:"Frame Rate Structure (by ref)"=@n2
     ) -> (ReturnValue:Frame Time Structure)) -> (FrameNumber_Value:int, SubFrame:float=~) {is=title K2Node_BreakStruct}
   )
  ]}
 getTODHoursMins {nodes [
   n1 = getTODHoursMins!() -> (then, curTOD:double=~) {is=title K2Node_FunctionEntry}
   n2 = Fraction!(A:double=@n1) -> (ReturnValue:double)
   > Return Node!(execute, hours:int=Floor(A=@n1) {is=title K2Node_CallFunction}, minutes:int=Truncate(A=float * float(A=@n2, B=60)))
  ]}
 getTODTimeValue {nodes [
   n1 = getTODTimeValue!() -> (then, value:double=~) {is=title K2Node_FunctionEntry}
   n2 = "Clamp (Float)"(Value=@n1, Min=0.01, Max=23.99)
   n3 = float * float(A=Fraction(A=@n2), B=60)
   n4 = Append!(A=0., B="To String (Integer)"(InInt=Truncate(A=@n3))) -> (ReturnValue)
   n5 = Floor(A=@n2)
   > Return Node!(execute, outValue:double=float + float(A="To Float (Integer)"(InInt=@n5), B=String To Float(InString=@n4)))
  ]}
 getTempDir {nodes [
   getTempDir() {is=title K2Node_FunctionEntry}
   n1 = Get Object Path String(Object=Get LSBase() -> (LSBase:Level Sequence Object Reference))
   > n2 = getParentFolder!(execute, self, InPath:String=GetPath!(InPath:String=@n1) -> (ReturnValue:String)) -> (then, OutPath:String) {is=title K2Node_CallFunction}
   n3 = Append(A=@n2, B=temp/)
   > n4 = DoesDirectoryExist!(execute, DirectoryPath:String=@n3) -> (then, ReturnValue:bool)
   > n5 = Branch(Condition=@n4)
   n6 = Set tempDir(execute=[@n8.then, @n5.then], tempDir:String=@n3) -> (Output_Get:String)
   > n7 = Set LSBasePath(LSBasePath:String=@n1) -> (Output_Get:String)
   n8 = MakeDirectory!(execute=@n5.else, DirectoryPath:String=@n3) -> (then, ReturnValue:bool)
   Return Node!(
    execute=@n7.then
    LSBaseSourcePath:String=@n7
    TempDir:String=@n6
    success:bool=Select(Option 0:bool=@n8, Option 1:bool=@n4, Index=@n4) -> (ReturnValue:bool)
   )
  ]}
 getTotalDuration {nodes [
   getTotalDuration() {is=title K2Node_FunctionEntry}
   n1 = Get CDuration() -> (CDuration=0)
   n2 = int + int(A=SelectInt(A=@n1, B=0, bPickA=Get useTurnTable()), B=SelectInt(A=@n1, B=0, bPickA=Get orbitCamera()))
   > Return Node!(execute, sequenceFrames:int=SelectInt(A=Get CDuration() -> (CDuration=0), B=@n2, bPickA="Equal (Integer)"(A=@n2)))
  ]}
 indexToEnum {nodes [
   n1 = indexToEnum!() -> (then, inIndex:int) {is=title K2Node_FunctionEntry}
   > Return Node!(
    execute
    ReturnValue:ENU_TransformChannels Enum=Select!(
     Option 0:ENU_TransformChannels Enum=NewEnumerator0
     Option 1:ENU_TransformChannels Enum=NewEnumerator1
     Option 2:ENU_TransformChannels Enum=NewEnumerator2
     Index:int=@n1
    ) -> (ReturnValue:ENU_TransformChannels Enum)
   )
  ]}
 manageAngle {nodes [
   n1 = manageAngle!() -> (then, value:double=~) {is=title K2Node_FunctionEntry}
   > n2 = Set MaxAngle(MaxAngle:double=@n1) -> (Output_Get:double)
   n3 = "To String (Integer)"(InInt=Truncate(A=@n2))
   > "SetText (Text)"(
    self=Get turnTableAngle_text() -> (turnTableAngle_text:Text Block Object Reference)
    InText="To Text (String)"(InString=Append(A="Angle (deg): ", B=@n3))
   )
  ]}
 manageDuration {nodes [
   n1 = manageDuration!() -> (then, value:double=~) {is=title K2Node_FunctionEntry}
   > n2 = Set CDuration(CDuration:int=Truncate(A=@n1)) -> (Output_Get:int=0)
   n3 = "To String (Integer)"(InInt=@n2)
   > "SetText (Text)"(
    self=Get cutDuration() -> (cutDuration:Text Block Object Reference)
    InText="To Text (String)"(InString=Append(A="Cut Duration (frames): ", B=@n3))
   )
  ]}
 manageFOV {nodes [
   n1 = manageFOV!() -> (then, value:double=~) {is=title K2Node_FunctionEntry}
   > Sequence()
   > n2 = Set cameraFOV(cameraFOV:double=@n1) -> (Output_Get:double)
   n3 = "To String (Integer)"(InInt=Truncate(A=@n2))
   > "SetText (Text)"(self=Get camFOV_text() -> (camFOV_text:Text Block Object Reference), InText="To Text (String)"(InString=Append(A="Camera FOV: ", B=@n3)))
   ExecuteConsoleCommand(Command=FOVAngle 10)
  ]}
 manageOffset {nodes [
   n1 = manageOffset!() -> (then, YOffset:double=~) {is=title K2Node_FunctionEntry}
   > n2 = Sequence()
   > n3 = Branch(Condition="float < float"(A="Absolute (Float)"(A=@n1), B=10))
   > n4 = Set YOffset(YOffset:double=float * float(A="Sign (Float)"!(A:double=@n1) -> (ReturnValue:double), B=10)) -> (Output_Get:double)
   > SetValue!(execute, self:Slider Object Reference=Get yOffsetSlider(), InValue:float=@n4) -> (then)
   Set YOffset(execute=@n3.else, YOffset=@n1)
   n5 = "To String (Integer)"(InInt=Truncate(A=Get YOffset()))
   n6 = Get OffsetName() -> (OffsetName:String)
   "SetText (Text)"(
    execute=@n2.then_1
    self=Get yOffsetText() -> (yOffsetText:Text Block Object Reference)
    InText="To Text (String)"(InString=Append(A=@n6, B=@n5))
   )
  ]}
 manageSizeRef {nodes [
   manageSizeRef() {is=title K2Node_FunctionEntry}
   > SetIsEnabled(self=[Get JunoRef(), Get yOffsetText(), Get yOffsetSlider()], bInIsEnabled=IsChecked(self=Get SizeRef()))
  ]}
 presetAngle {nodes [
   n1 = presetAngle!() -> (then, angle:int) {is=title K2Node_FunctionEntry}
   > SetValue(self=Get turnAngle(), InValue="To Float (Integer)"(InInt=@n1))
  ]}
 presetCutDuration {nodes [
   n1 = presetCutDuration!() -> (then, seconds:int) {is=title K2Node_FunctionEntry}
   > SetValue(self=Get cutDurationSlider(), InValue="To Float (Integer)"(InInt=int * int(A=@n1, B=30)))
  ]}
 presetFOV {nodes [
   n1 = presetFOV!() -> (then, fov:int) {is=title K2Node_FunctionEntry}
   > SetValue(self=Get camFOV_slider(), InValue="To Float (Integer)"(InInt=@n1))
  ]}
 previewTimeOfDay {nodes [
   n1 = previewTimeOfDay!() -> (then, value:double=~, TODActorRef:Fort Day Sequence Actor Object Reference) {is=title K2Node_FunctionEntry}
   n2 = Select!(
    Option 0:Day Sequence Actor Object Reference=Get TODActor()
    Option 1:Day Sequence Actor Object Reference=@n1.TODActorRef
    Index:bool=IsValid(Object=@n1.TODActorRef)
   ) -> (ReturnValue:Day Sequence Actor Object Reference)
   > Is Valid(InputObject=@n2)
   n3 = float * float(A=@n1, B=23.99)
   > SetTimeOfDayPreview!(execute, self:Day Sequence Actor Object Reference=@n2, InHours:float=@n3) -> (then)
   > SetEditorProperty(Object=@n2, PropertyName=bRunDayCycle, PropertyValue=Get runDayCycle() -> (runDayCycle:bool))
   n4 = Truncate(A=float * float(A=Fraction(A=@n3), B=60))
   n5 = Floor(A=@n3)
   > SetEditorProperty(
    Object=@n2
    PropertyName=InitialTimeOfDay
    PropertyValue:Timecode Structure=Make Timecode!(Hours:int=@n5, Minutes:int=@n4, Seconds:int=0, Frames:int=0, Subframe:float=0, bDropFrameFormat:bool=False) -> (Timecode:Timecode Structure) {is=title K2Node_MakeStruct}
   )
   n6 = Append!(A="To String (Integer)"(InInt=@n5), B=:, C:String="To String (Integer)"(InInt=@n4)) -> (ReturnValue)
   > Set Text(
    Text="To Text (String)"(InString=Append!(A="Time of Day: ", B=@n6) -> (ReturnValue))
    self=Get CurrentTODText() -> (CurrentTODText:Text Block Object Reference)
   )
   > Return Node!(execute, curTime:double=@n3)
  ]}
 printDefaultMessage {nodes [
   printDefaultMessage() {is=title K2Node_FunctionEntry}
   > printMessage!(execute, self, Message:ST Showcase Message Structure=Get DefaultMessage(), displayTime=0) -> (then)
   > Clear and Invalidate Timer by Handle!(execute, WorldContextObject:Object Reference, Handle=Get defaultMessageTimer() -> (defaultMessageTimer:Timer Handle Structure)) -> (then)
  ]}
 printErrorInvalidObject {nodes [
   n1 = printErrorInvalidObject!() -> (then, object:Object Reference, execPhase:String) {is=title K2Node_FunctionEntry}
   > n2 = Is Valid(InputObject=@n1)
   printMessage!(
    execute=@n2.Is Not Valid
    self
    Message_Message_4_#1:Text=Format Text!(Format="LOCTEXT(\"Error: Object not valid during {phase}. Stopping...\")", phase:String=@n1.execPhase) -> (Result)
    Message_Level_10_#2:ENU_MessageLevel Enum=NewEnumerator2
    displayTime=5
   ) -> (then)
   > Return Node!(execute, valid:bool=false)
   Return Node(execute=@n2.Is Valid, valid=true)
  ]}
 printMessage {nodes [
   n1 = printMessage!() -> (then, Message:ST Showcase Message Structure, displayTime:int) {is=title K2Node_FunctionEntry}
   > n2 = Sequence()
   n3 = Get defaultMessageTimer()
   > Branch(Condition=Is Valid Timer Handle!(Handle:Timer Handle Structure=@n3) -> (ReturnValue:bool))
   > Clear and Invalidate Timer by Handle(Handle=@n3)
   n4 = Break ST Showcase Message!(ST_ShowcaseMessage:"ST Showcase Message Structure (by ref)"=@n1) -> (Message_4_#1:Text, Level_10_#2:ENU_MessageLevel Enum) {is=title K2Node_BreakStruct}
   n5 = Get MessageBox() -> (MessageBox:Rich Text Block Object Reference)
   Set Text(execute=@n2.then_1, Text=@n4, self:Rich Text Block Object Reference=@n5)
   > Set TextStyleSet!(
    execute
    TextStyleSet:Data Table Object Reference=Find(
     TargetMap:Map of ENU_MessageLevel Enums to Data Table Object References=Get STMessageLevels() -> (STMessageLevels:Map of ENU_MessageLevel Enums to Data Table Object References=NewEnumerator0)
     Key:"ENU_MessageLevel Enum (by ref)"=@n4.Level_10_#2
    ) -> (Value:Data Table Object Reference)
    self:Rich Text Block Object Reference=@n5
   ) -> (then, Output_Get:Data Table Object Reference)
   > SetBrushColor(
    self=Get MessageBorder() -> (MessageBorder:Border Object Reference)
    InBrushColor=Find(
     TargetMap:Map of ENU_MessageLevel Enums to Linear Color Structures=Get STMsgLevelColors() -> (STMsgLevelColors:Map of ENU_MessageLevel Enums to Linear Color Structures=NewEnumerator0)
     Key=@n4.Level_10_#2
    ) -> (Value:Linear Color Structure)
   )
   > Branch(Condition="integer > integer"(A=Get displayTime() -> (displayTime:int=0)))
   > n6 = Set Timer by Function Name!(
    execute
    Object:Object Reference
    FunctionName:String=printDefaultMessage
    Time:float="To Float (Integer)"(InInt=Get displayTime() -> (displayTime=0))
    bLooping:bool=false
    bMaxOncePerFrame:bool=false
    InitialStartDelay:float=0
    InitialStartDelayVariance:float=0
   ) -> (then, ReturnValue:Timer Handle Structure) {is=title K2Node_CallFunction}
   > Set defaultMessageTimer(defaultMessageTimer:Timer Handle Structure=@n6) -> (Output_Get:Timer Handle Structure)
  ]}
 renderQueue {nodes [
   n1 = GetCurrentLevel!(execute, self:Level Editor Subsystem Object Reference=Get LevelEditorSubsystem()) -> (then, ReturnValue:Level Object Reference)
   > Set Map!(execute, Map:Soft Object Path Structure=GetSoftObjectPath(Object=@n1), self:Movie Pipeline Executor Job Object Reference) -> (then, Output_Get:Soft Object Path Structure)
   n2 = renderQueue!() -> (then, InSequence:Level Sequence Object Reference) {is=title K2Node_FunctionEntry}
   > n3 = Construct Object from Class(
    Class=/Script/MovieRenderPipelineCore.MoviePipelineQueue
    self=Get MoviePipelineQueueSubsystem() -> (ReturnValue:Movie Pipeline Queue Subsystem Object Reference) {is=title K2Node_GetEditorSubsystem}
   ) -> (ReturnValue:Movie Pipeline Queue Object Reference)
   > n4 = M UIprint Error Invalid Object!(in:Exec, InputObject=@n3, execPhase:String=Constructing MPQ) -> (Is Valid, Not Valid:Exec) {is=title K2Node_MacroInstance}
   > n5 = CreateJobFromSequence!(execute, InPipelineQueue:Movie Pipeline Queue Object Reference=@n3, InSequence:Level Sequence Object Reference=@n2) -> (then, ReturnValue:Movie Pipeline Executor Job Object Reference)
   > n6 = M UIprint Error Invalid Object(InputObject=@n5, execPhase=Creating Job)
   > Set Sequence!(execute, Sequence:Soft Object Path Structure=GetSoftObjectPath(Object=@n2), self:Movie Pipeline Executor Job Object Reference=@n5) -> (then, Output_Get:Soft Object Path Structure)
   > SetConfiguration!(
    execute
    self:Movie Pipeline Executor Job Object Reference=@n5
    InPreset:Movie Pipeline Primary Config Object Reference=getConfig!() -> (Value:Movie Pipeline Primary Config Object Reference) {is=title K2Node_CallFunction}
   ) -> (then)
   n7 = Get MLineComments() -> (MLineComments:Editor Utility Multi Line Editable Text Box Object Reference)
   > Set Comment!(
    execute
    Comment:String="To String (Text)"(InText=Get Text(self:"Text Box (Multi- Line) Object Reference"=@n7))
    self:Movie Pipeline Executor Job Object Reference=@n5
   ) -> (then, Output_Get:String)
   > n8 = Branch(Condition=LoadQueue!(
     self:Movie Pipeline Queue Subsystem Object Reference=Get MoviePipelineQueueSubsystem()
     QueueToLoad:Movie Pipeline Queue Object Reference=@n3
     bPromptOnReplacingDirtyQueue:bool=false
    ) -> (ReturnValue:bool))
   > UpdateJobShotListFromSequence!(
    execute
    InSequence:Level Sequence Object Reference=Get InSequence() -> (InSequence:Level Sequence Object Reference)
    InJob:Movie Pipeline Executor Job Object Reference=@n5
   ) -> (then, bShotsChanged:bool)
   > n9 = RenderQueueWithExecutor!(
    execute
    self:Movie Pipeline Queue Subsystem Object Reference=Get MoviePipelineQueueSubsystem()
    InExecutorType:Movie Pipeline Executor Base Class Reference=/Script/MovieRenderPipelineEditor.MoviePipelinePIEExecutor
   ) -> (then, ReturnValue:Movie Pipeline PIEExecutor Object Reference)
   > Bind Event to On Individual Job Started Delegate(self:Movie Pipeline PIEExecutor Object Reference=@n9, Delegate=Create Event()) {is=title K2Node_AddDelegate}
   > Bind Event to On Executor Finished Delegate(self:Movie Pipeline Executor Base Object Reference=@n9, Delegate=Create Event()) {is=title K2Node_AddDelegate}
   > Return Node!(execute, success:bool=true)
   M UIprint Error Phase!(in:Exec=@n8.else, execPhase:String=Load Queue) -> (out:Exec) {is=title K2Node_MacroInstance}
   > Return Node(success=false)
   Return Node(execute=[@n4.Not Valid, @n6.Not Valid], success=false)
   Get Text(self:Editable Text Object Reference=Get UserComment() -> (UserComment:UEFN Editable Text Object Reference))
  ]}
 rotatorToVector {nodes [
   n1 = rotatorToVector!() -> (then, rot_Roll:float, rot_Pitch:float, rot_Yaw:float) {is=title K2Node_FunctionEntry}
   > Return Node!(execute, vec_X:double=@n1, vec_Y:double=@n1.rot_Pitch, vec_Z:double=@n1.rot_Yaw)
  ]}
 setTOD {nodes [
   n1 = setTOD!() -> (then, value:double=~) {is=title K2Node_FunctionEntry}
   n2 = Get TODRef()
   n3 = Select(
    Option 0:Fort Time Of Day Manager Object Reference=Cast To FortTimeOfDayManager!(Object:Fort Time Of Day Manager Interface Interface=GetContextualTimeOfDayManager!(WorldContextObject:Object Reference) -> (ReturnValue:Fort Time Of Day Manager Interface Interface)) -> (AsFort Time Of Day Manager:Fort Time Of Day Manager Object Reference, bSuccess:bool=~) {is=title K2Node_DynamicCast}
    Option 1:Fort Time Of Day Manager Object Reference=@n2
    Index=IsValid(Object=@n2)
   ) -> (ReturnValue:Fort Time Of Day Manager Object Reference)
   > SetUseSpecificTimeOfDayEditing!(execute, self:Fort Time Of Day Manager Object Reference=@n3, bShouldUseSpecificTimeOfDay:bool=true) -> (then)
   n4 = float * float(A=@n1, B=23.99)
   > SetSpecificTimeOfDayInEditor!(execute, self:Fort Time Of Day Manager Object Reference=@n3, NewTimeOfDayInEditor:float=@n4) -> (then)
   n5 = float * float(A=Fraction(A=@n4), B=60)
   n6 = Append!(A="To String (Integer)"(InInt=Floor(A=@n4)), B=:, C:String="To String (Integer)"(InInt=Truncate(A=@n5))) -> (ReturnValue)
   > Set Text(Text="To Text (String)"(InString=Append!(A="Time of Day: ", B=@n6) -> (ReturnValue)), self:Text Object Reference=Get CurrentTODText())
   > Return Node!(execute, curTime:double=@n4)
   GetGlobalTimeOfDayManager!(WorldContextObject:Object Reference) -> (ReturnValue:Fort Time Of Day Manager Interface Interface)
  ]}
 setupFallback {nodes [
   n1 = Get LSTemp()
   n2 = setupFallback!() -> (then, binding:Movie Scene Binding Proxy Structure) {is=title K2Node_FunctionEntry}
   n3 = LocateBoundObjects!(
    execute
    Sequence:Movie Scene Sequence Object Reference=@n1
    InBinding:"Movie Scene Binding Proxy Structure (by ref)"=@n2
    Context:Object Reference=GetMovieScene!(execute, Sequence:Movie Scene Sequence Object Reference=@n1) -> (then, ReturnValue:Movie Scene Object Reference)
   ) -> (then, ReturnValue:Array of Object References)
   > n4 = Branch(Condition=Is Empty(TargetArray:Array of Object References=@n3))
   > PrintString(InString="Could not locate Binding instance!")
   n5 = "Get (a copy)"(Array=@n3)
   n6 = Cast To StaticMeshActor(execute=@n4.else, Object=@n5)
   > Set SMActor(SMActor:Static Mesh Actor Object Reference=@n6) -> (Output_Get:Static Mesh Actor Object Reference)
   > n7 = SetForceDisableNanite(self=Get StaticMeshComponent(self=@n6), bInForceDisableNanite=true)
   PrintString(execute=[@n7.then, @n10.then], InString="Resolved Binding instance!")
   n8 = Cast To BuildingProp(execute=@n6.CastFailed, Object=@n5)
   > n9 = Set BPropActor(BPropActor:Building Prop Object Reference=@n8) -> (Output_Get:Building Prop Object Reference)
   > n10 = SetForceDisableNanite(
    self=Get StaticMeshComponent(self:Building Static Mesh Object Reference=@n9) -> (StaticMeshComponent:Base Building Static Mesh Component Object Reference)
    bInForceDisableNanite=true
   )
   PrintString(execute=@n8.CastFailed, InString="Cast to Building Prop failed!")
   PrintString(execute=@n12.else, InString="Visibility channel not created!")
   PrintString(execute=@n13.then, InString=Visibility Key added)
   Get LevelSequenceEditorSubsystem()
   GetBoundObjectForSequencer!(
    execute:Exec
    self:Object Reference=GetObjectTemplate!(execute, InBinding:"Movie Scene Binding Proxy Structure (by ref)"=@n2) -> (then, ReturnValue:Object Reference)
    ResolvedObject:Object Reference
   ) -> (then:Exec, ReturnValue:Object Reference) {is=title K2Node_Message}
   n11 = addVisiblityTrack(execute=@n14.then_1, InBinding=@n2, duration=Get CDuration() -> (CDuration=0)) {is=title K2Node_CallFunction}
   > n12 = Branch(Condition=@n11)
   > n13 = "Add Key (Bool)"(self=@n11.boolChannel, InTime=Make Frame Number(Value=0), NewValue=false, SubFrame=0, TimeUnit=DisplayRate)
   n14 = Sequence(execute=@n2.then)
  ]}
 setupLODInfo {nodes [
   setupLODInfo() {is=title K2Node_FunctionEntry}
   > n1 = Sequence()
   > n2 = GetPathNameForLoadedAsset(LoadedAsset=Get selectedActorSM() -> (selectedActorSM:Static Mesh Object Reference))
   > n3 = Execute Python Script!(
    execute
    PythonScript=MakeLiteralString(Value="import unreal\r\n\r\nsmload = unreal.load_asset(sm,None,True)\r\nasset_options = unreal.GeometryScriptCopyMeshFromAssetOptions()\r\nrequested_lod = unreal.GeometryScriptMeshReadLOD()\r\ndm = unreal.DynamicMesh()\r\nunreal.GeometryScript_AssetUtils.copy_mesh_from_static_mesh(smload, dm, asset_options, requested_lod)\r\n\r\n\r\nmesh_editor_ss = unreal.get_editor_subsystem(unreal.StaticMeshEditorSubsystem)\r\nbuildsetting = mesh_editor_ss.get_lod_build_settings(smload,0)\r\nsubsys = unreal.get_editor_subsystem(unreal.StaticMeshEditorSubsystem)\r\n\r\n\r\nlod_for_collision= smload.get_editor_property(\"lod_for_collision\")\r\n \r\n\r\ndist_field_bias = smload.get_editor_property(\"distance_field_self_shadow_bias\")\r\ndist2sided = buildsetting.get_editor_property(\"generate_distance_field_as_if_two_sided\" )\r\nvertcount = subsys.get_number_verts(smload,0)\r\nnanitevertcount = dm.get_vertex_count()\r\ncollcount = subsys.get_convex_collision_count(smload)\r\nleafcol = subsys.is_section_collision_enabled(smload,lod_for_collision,1) and subsys.is_section_collision_enabled(smload,lod_for_collision,0)")
    sm:String=@n2
   ) -> (
    then
    ReturnValue
    dist_field_bias:double=~
    dist2sided:bool=~
    lod_for_collision:int
    vertcount:int
    nanitevertcount:int
    collcount:int
    leafcol:bool=~
   ) {inputs [sm], outputs [dist_field_bias, dist2sided, lod_for_collision, vertcount, nanitevertcount, collcount, leafcol]}
   > Set DistFieldBias(DistFieldBias:double=@n3.dist_field_bias) -> (Output_Get:double)
   > Set DistField2Sided(DistField2Sided:bool=@n3.dist2sided) -> (Output_Get:bool)
   > Set LodForCollision(LodForCollision:int=@n3.lod_for_collision) -> (Output_Get:int=0)
   > Set vtxCount(vtxCount:int=@n3.vertcount) -> (Output_Get:int=0)
   > Set NaniteVtxCount(NaniteVtxCount:int=@n3.nanitevertcount) -> (Output_Get:int=0)
   > Set CollisionCount(CollisionCount:int=@n3.collcount) -> (Output_Get:int=0)
   > Set LeafCollisions(LeafCollisions:bool=@n3.leafcol) -> (Output_Get:bool)
   > PrintString(InString=Append(A="LOD Data Gathered: ", B="To String (Boolean)"(InBool=@n3)))
   n4 = AddTrack(
    execute=@n1.then_1
    InBinding=Get TODBindingProxy() -> (TODBindingProxy:Movie Scene Binding Proxy Structure)
    TrackType=&MovieSceneFloatTrack
   )
   > n5 = Cast To MovieSceneFloatTrack(Object=@n4)
   n6 = MakeLiteralString(Value=LODInfo_Nanite)
   > SetPropertyNameAndPath(Track=@n5, InPropertyName=String To Name(InString=@n6), InPropertyPath=@n6)
   > n7 = AddSection(Track=@n5)
   > SetEndFrame(Section=@n7, EndFrame=Get SeqFrames() -> (SeqFrames=0))
   > n8 = GetAllChannels(Section=@n7)
   > n9 = Cast To MovieSceneScriptingFloatChannel(Object="Get (a copy)"(Array:Array of Movie Scene Scripting Channel Object References=@n8) -> (Output:Movie Scene Scripting Channel Object Reference))
   > n10 = Set fChan(fChan:Movie Scene Scripting Float Channel Object Reference=@n9) -> (Output_Get:Movie Scene Scripting Float Channel Object Reference)
   n11 = "To Float (Integer)"(InInt=Get NaniteVtxCount() -> (NaniteVtxCount:int=0))
   > "Add Key (Float)"(
    self=@n10
    InTime=Make Frame Number(Value=Get Pre Roll Frame for Event!() -> (ReturnValue:int=0) {is=title K2Node_CallFunction})
    NewValue=@n11
    SubFrame=0
    TimeUnit=DisplayRate
    InInterpolation=Linear
   )
   > "Add Key (Float)"(
    self=Get fChan() -> (fChan:Movie Scene Scripting Float Channel Object Reference)
    InTime=Make Frame Number(Value=0)
    NewValue=@n11
    SubFrame=0
    TimeUnit=DisplayRate
    InInterpolation=Linear
   )
   > "Add Key (Float)"(
    self=Get fChan()
    InTime=Make Frame Number(Value=Get SeqFrames() -> (SeqFrames=0))
    NewValue=@n11
    SubFrame=0
    TimeUnit=DisplayRate
    InInterpolation=Linear
   )
  ]}
 setupTimeOfDay {nodes [
   n1 = AddTrack(InBinding=Get TODBindingProxy(), TrackType=/Script/MovieSceneTracks.MovieSceneEventTrack)
   > n2 = Cast To MovieSceneEventTrack!(execute, Object=@n1) -> (then, CastFailed, AsMovie Scene Event Track:Movie Scene Event Track Object Reference) {is=title K2Node_DynamicCast}
   > n3 = AddEventTriggerSection!(execute, InTrack:Movie Scene Event Track Object Reference=@n2) -> (then, ReturnValue:Movie Scene Event Trigger Section Object Reference)
   > n4 = Set eventSection(eventSection:Movie Scene Event Trigger Section Object Reference=@n3) -> (Output_Get:Movie Scene Event Trigger Section Object Reference)
   > SetPreRollFrames(
    self=@n3
    InPreRollFrames=getSequenceFrames(levelSeq=Get LSTemp(), frames=Get PreRollFrames() -> (PreRollFrames=0)) -> (Frame Number Value=0) {is=title K2Node_CallFunction}
   )
   > SetStartFrame(Section=@n4, StartFrame=Get Pre Roll Frame for Section!() -> (ReturnValue:int=0) {is=title K2Node_CallFunction})
   > SetEndFrame(Section=@n4, EndFrame=Get SeqFrames() -> (SeqFrames=0))
   > n5 = GetChannelsByType!(
    execute
    Section:Movie Scene Section Object Reference=@n4
    ChannelType:Movie Scene Scripting Channel Class Reference=/Script/SequencerScripting.MovieSceneScriptingEventChannel
   ) -> (then, ReturnValue:Array of Movie Scene Scripting Channel Object References)
   > n6 = Branch(Condition=Is Not Empty!(TargetArray:Array of Movie Scene Scripting Channel Object References=@n5) -> (ReturnValue:bool))
   > n7 = Cast To MovieSceneScriptingEventChannel!(execute, Object="Get (a copy)"(Array=@n5)) -> (then, CastFailed, AsMovie Scene Scripting Event Channel:Movie Scene Scripting Event Channel Object Reference) {is=title K2Node_DynamicCast}
   > Set eventChannel(eventChannel:Movie Scene Scripting Event Channel Object Reference=@n7) -> (Output_Get:Movie Scene Scripting Event Channel Object Reference)
   > n8 = CreateQuickBinding!(
    execute
    InSequence:Movie Scene Sequence Object Reference=Get LSTemp()
    InObject:Object Reference=Get TODActor()
    InFunctionName:String=MakeLiteralString(Value=SetStaticTimeOfDay)
    bCallInEditor:bool=true
   ) -> (then, ReturnValue:Sequencer Quick Binding Result Structure)
   n9 = Get LSTemp()
   n10 = Get eventSection() -> (eventSection:Movie Scene Event Trigger Section Object Reference)
   n11 = Get eventChannel() -> (eventChannel:Movie Scene Scripting Event Channel Object Reference)
   n12 = Get CurTOD() -> (CurTOD:double)
   > createTODEvent!(
    execute
    self
    InSequence:Movie Scene Sequence Object Reference=@n9
    InSection:Movie Scene Event Section Base Object Reference=@n10
    channel:Movie Scene Scripting Event Channel Object Reference=@n11
    InEndpoint:"Sequencer Quick Binding Result Structure (by ref)"=@n8
    InTOD:double=@n12
    InTime:"Frame Number Structure (by ref)"=Make Frame Number(Value=Get Pre Roll Frame for Event() -> (ReturnValue=0))
   ) -> (then, outKey:Movie Scene Scripting Event Key Object Reference) {is=title K2Node_CallFunction}
   > createTODEvent(InSequence=@n9, InSection=@n10, channel=@n11, InEndpoint=@n8, InTOD=@n12, InTime=Make Frame Number(Value=0))
   PrintString(execute=@n7.CastFailed, InString="TOD Channel Cast failed!")
   PrintString(execute=@n6.else, InString="TOD Failed to add Channel!")
   setupTimeOfDay() {is=title K2Node_FunctionEntry}
   n13 = Get TODActor()
   > n14 = Is Valid(InputObject=@n13)
   PrintString(execute=@n14.Is Not Valid, InString="TOD Actor not valid!")
   n15 = AddPossessable(execute=@n14.Is Valid, Sequence=Get LSTemp(), ObjectToPossess=@n13)
   > n16 = Set TODBindingProxy(TODBindingProxy:Movie Scene Binding Proxy Structure=@n15) -> (Output_Get:Movie Scene Binding Proxy Structure)
   > n17 = AddTrack(InBinding=@n16, TrackType=&MovieSceneFloatTrack)
   > n18 = Cast To MovieSceneFloatTrack(Object=@n17)
   n19 = MakeLiteralString(Value=CurTimeOfDay)
   > SetPropertyNameAndPath(Track=@n18, InPropertyName=String To Name(InString=@n19), InPropertyPath=@n19)
   > n20 = AddSection(Track=@n18)
   > SetEndFrame(Section=@n20, EndFrame=Get SeqFrames() -> (SeqFrames=0))
   > n21 = GetAllChannels(Section=@n20)
   > n22 = Cast To MovieSceneScriptingFloatChannel(Object="Get (a copy)"(Array=@n21))
   > n23 = Set todFloatChan(todFloatChan:Movie Scene Scripting Float Channel Object Reference=@n22) -> (Output_Get:Movie Scene Scripting Float Channel Object Reference)
   n24 = SelectFloat(A=Get TODMin() -> (TODMin:double), B=Get CurTOD(), bPickA=Get useTODRange())
   > "Add Key (Float)"(
    self=@n23
    InTime=Make Frame Number(Value=Get Pre Roll Frame for Event() -> (ReturnValue=0))
    NewValue=@n24
    SubFrame=0
    TimeUnit=DisplayRate
    InInterpolation=Linear
   )
   > "Add Key (Float)"(
    self=Get todFloatChan() -> (todFloatChan:Movie Scene Scripting Float Channel Object Reference)
    InTime=Make Frame Number(Value=0)
    NewValue=@n24
    SubFrame=0
    TimeUnit=DisplayRate
    InInterpolation=Linear
   )
   > "Add Key (Float)"(
    self=Get todFloatChan()
    InTime=Make Frame Number(Value=Get SeqFrames() -> (SeqFrames=0))
    NewValue=SelectFloat(A=Get TODMax() -> (TODMax:double), B=Get CurTOD(), bPickA=Get useTODRange())
    SubFrame=0
    TimeUnit=DisplayRate
    InInterpolation=Linear
   )
   n25 = For Loop(LastIndex=int - int(A=Get SeqFrames() -> (SeqFrames=0)))
   > n26 = Branch(Condition="Equal (Integer)"(A=@n25))
   > createTODEvent(
    InSequence=Get LSTemp()
    InSection=Get eventSection()
    channel=Get eventChannel()
    InEndpoint=@n8
    InTOD=Get CurTOD()
    InTime=Make Frame Number(Value=Get Pre Roll Frame for Event() -> (ReturnValue=0))
   )
   > createTODEvent(
    InSequence=Get LSTemp()
    InSection=Get eventSection()
    channel=Get eventChannel()
    InEndpoint=@n8
    InTOD=Get CurTOD()
    InTime=Make Frame Number(Value=0)
   )
   n27 = "To Float (Integer)"(InInt=@n25)
   n28 = "To Float (Integer)"(InInt=int - int(A=Get SeqFrames() -> (SeqFrames=0)))
   createTODEvent(
    execute=@n26.else
    InSequence=Get LSTemp()
    InSection=Get eventSection()
    channel=Get eventChannel()
    InEndpoint=@n8
    InTOD=Lerp(A=Get TODMin(), B=Get TODMax(), Alpha=float / float(A=@n27, B=@n28))
    InTime=Make Frame Number(Value=@n25)
   )
   SetEditorProperty(
    Object=Get TODActor()
    PropertyName=String To Name(InString=MakeLiteralString(Value=bRunDayCycle))
    PropertyValue:bool=MakeLiteralBool(Value=true)
   )
   n29 = Make Timecode(Hours=24, Minutes=0, Seconds=0, Frames=0, Subframe=0, bDropFrameFormat=False)
   > SetEditorProperty(Object=Get TODActor(), PropertyName=String To Name(InString=MakeLiteralString(Value=DayLength)), PropertyValue:Timecode Structure=@n29)
   n30 = getTODHoursMins!(curTOD:double=Get TODMin()) -> (hours:int=0, minutes:int=0) {is=title K2Node_CallFunction}
   n31 = Make Timecode(Hours=@n30, Minutes=@n30.minutes, Seconds=0, Frames=0, Subframe=0, bDropFrameFormat=False)
   > SetEditorProperty(Object=Get TODActor(), PropertyName=String To Name(InString=MakeLiteralString(Value=InitialTimeOfDay)), PropertyValue=@n31)
   n32 = "To Float (Integer)"(InInt=Get SeqFrames() -> (SeqFrames=0))
   n33 = Make Timecode(Hours=0, Minutes=0, Seconds=Truncate(A=float / float(A=@n32, B=30)), Frames=0, Subframe=0, bDropFrameFormat=False)
   > SetEditorProperty(Object=Get TODActor(), PropertyName=String To Name(InString=MakeLiteralString(Value=TimePerCycle)), PropertyValue=@n33)
   n34 = Get useTODRange()
   PrintString(InString=Append!(
     A="TOD Commands: "
     B="\r\n"
     C:String="To String (Boolean)"(InBool=@n34)
     D:String="\r\n"
     E:String=Timecode To String!(InTimecode:"Timecode Structure (by ref)"=@n29, bForceSignDisplay:bool=false) -> (ReturnValue:String)
     F:String="\r\n"
     G:String=Timecode To String(InTimecode=@n31, bForceSignDisplay=false)
     H:String="\r\n"
     I:String=Timecode To String(InTimecode=@n33, bForceSignDisplay=false)
    ) -> (ReturnValue))
   Get CDuration() -> (CDuration=0)
  ]}
 setupTimeOfDay_0 {nodes [
   setupTimeOfDay_0() {is=title K2Node_FunctionEntry}
   n1 = Get TODActor()
   > n2 = Is Valid(InputObject=@n1)
   PrintString(execute=@n2.Is Not Valid, InString="TOD Actor not valid!")
   n3 = AddPossessable(execute=@n2.Is Valid, Sequence=Get LSTemp(), ObjectToPossess=@n1)
   > n4 = AddTrack(InBinding=@n3, TrackType=&MovieSceneEventTrack)
   > n5 = Cast To MovieSceneEventTrack(Object=@n4)
   > n6 = AddEventTriggerSection(InTrack=@n5)
   > n7 = Set eventSection(eventSection=@n6)
   > SetPreRollFrames(
    self=@n6
    InPreRollFrames=getSequenceFrames(levelSeq=Get LSTemp(), frames=Get PreRollFrames() -> (PreRollFrames=0)) -> (Frame Number Value=0)
   )
   > SetStartFrame(Section=@n7, StartFrame=Get Pre Roll Frame for Section() -> (ReturnValue=0))
   > SetEndFrame(Section=@n7, EndFrame=Get SeqFrames() -> (SeqFrames=0))
   > n8 = GetChannelsByType(Section=@n7, ChannelType=&MovieSceneScriptingEventChannel)
   > n9 = Branch(Condition=Is Not Empty(TargetArray=@n8))
   > n10 = Cast To MovieSceneScriptingEventChannel(Object="Get (a copy)"(Array=@n8))
   n11 = Get LSTemp()
   > n12 = CreateQuickBinding(InSequence=@n11, InObject=Get TODActor(), InFunctionName=MakeLiteralString(Value=SetStaticTimeOfDay), bCallInEditor=true)
   n13 = Get CurTOD()
   > n14 = CreateEvent(
    InSequence=@n11
    InSection=Get eventSection()
    InEndpoint=@n12
    InPayload=Make Array("[0]":String="To String (Float)"(InDouble=@n13)) -> (Array:Array of Strings)
   )
   > "Add Key (Event)"(
    self=@n10
    InTime=Make Frame Number(Value=Get Pre Roll Frame for Event() -> (ReturnValue=0))
    NewValue=@n14
    SubFrame=0
    TimeUnit=DisplayRate
   )
   > "Add Key (Event)"(self=@n10, InTime=Make Frame Number(Value=0), NewValue=@n14, SubFrame=0, TimeUnit=DisplayRate)
   PrintString(execute=@n10.CastFailed, InString="TOD Channel Cast failed!")
   PrintString(execute=@n9.else, InString="TOD Failed to add Channel!")
   PrintString(InString=Append!(A="CurTOD: ", B="To String (Float)"(InDouble=@n13)) -> (ReturnValue))
   createTODEvent(InTOD=0)
   Get CDuration() -> (CDuration=0)
  ]}
 spawnJonesy {nodes [
   spawnJonesy() {is=title K2Node_FunctionEntry}
   > n1 = GetSelectedLevelActors(self=Get EditorActorSubsystem())
   > Branch(Condition="integer > integer"(A=Length(TargetArray:Array of Actor Object References=@n1)))
   > n2 = getScaleActorLocation(selActor="Get (a copy)"(Array:Array of Actor Object References=@n1) -> (Output:Actor Object Reference)) {is=title K2Node_CallFunction}
   n3 = Combine Rotators(A=@n2.ViewLookat, B="0,-90,0")
   n4 = IsChecked(self=Get JunoRef())
   > SpawnActor!(
    execute:Exec
    WorldContextObject:Object Reference
    Class:Actor Class Reference=Select(Option 0:Actor Class Reference=Get ScaleActor(), Option 1:Actor Class Reference=Get JunoScaleActor(), Index=@n4) -> (ReturnValue:Actor Class Reference)
    SpawnTransform:Transform=MakeTransform(Location=@n2, Rotation=SelectRotator(A=@n3, B=@n2.ViewLookat, bPickA=@n4))
    CollisionHandlingOverride:ESpawnActorCollisionHandlingMethod Enum=AlwaysSpawn
    TransformScaleMethod:ESpawnActorScaleMethod Enum=MultiplyWithRoot
    Owner:Actor Object Reference
    Instigator:Pawn Object Reference
   ) -> (then:Exec, ReturnValue:Actor Object Reference) {is=title K2Node_SpawnActorFromClass}
  ]}
 warnTempSequences {nodes [
   warnTempSequences() {is=title K2Node_FunctionEntry}
   n1 = Get tempDir()
   n2 = int - int(A=Len(S=@n1))
   > n3 = GetAssetsByPath(
    self=GetAssetRegistry()
    PackagePath=String To Name(InString=GetSubstring(SourceString=@n1, StartIndex=0, Length=@n2))
    bRecursive=false
    bIncludeOnlyOnDiskAssets=false
   )
   > Branch(Condition=@n3.ReturnValue)
   > n4 = Set assets(assets=@n3)
   > n5 = For Each Loop(Array=@n4)
   n6 = Get sequences()
   Show Message Dialog(
    execute=@n5.Completed
    Title=MakeLiteralText(Value="LOCTEXT(\"Cleanup Temp Sequences\")")
    Message="To Text (String)"(InString=Append(A="Some files need a cleanup: \r\n", B=@n6))
    MessageType=Ok
    DefaultValue=Ok
    MessageCategory=Info
   )
   n7 = Break Asset Data(AssetData=@n5)
   Set sequences(execute=@n5.LoopBody, sequences=Append(A="To String (Name)"(InName=@n7), B="\r\n"))
  ]}
 xyzToEnum {nodes [
   n1 = xyzToEnum!() -> (then, inStr:String) {is=title K2Node_FunctionEntry}
   n2 = Len(S=@n1)
   n3 = GetCharacterAsNumber!(SourceString:String=GetSubstring(SourceString=@n1, StartIndex=int - int(A=@n2), Length=1), Index:int=0) -> (ReturnValue:int=0)
   > Return Node!(
    execute
    ReturnValue:ENU_TransformChannels Enum=Select!(
     Option 0:ENU_TransformChannels Enum=NewEnumerator0
     Option 1:ENU_TransformChannels Enum=NewEnumerator1
     Option 2:ENU_TransformChannels Enum=NewEnumerator2
     Index:int=int - int(A=@n3, B=88)
    ) -> (ReturnValue:ENU_TransformChannels Enum)
   )
  ]}
]
variables [
 BurnIn {
  type=BP_Showcase_BurnInRel_C
  default=&/BP_AssetShowcase.Default__BP_AssetShowcase_C:BP_Showcase_BurnInRel_C_0
 }
 cameraFOV {type=number}
 CDuration {type=integer}
 CurSel {type="array<Actor>"}
 CurTOD {type=number}
 DefaultMessage {type=ST_ShowcaseMessage, default {level=Information, message=All good...}}
 defaultMessageTimer {type=TimerHandle}
 JunoScaleActor {type=Class, default=/Game/NeverCook/ArtTools/AssetShowcase/Assets/BP_JunoMinifig.BP_JunoMinifig_C}
 LSBase {type=LevelSequence, default=/Game/NeverCook/ArtTools/AssetShowcase/MRQAssets/LS_Showcase}
 LSBasePath {type=string, default=""}
 LSConfigMap {default {Default=/Game/NeverCook/ArtTools/AssetShowcase/MRQAssets/MPC_default, TimeOfDay=/Game/NeverCook/ArtTools/AssetShowcase/MRQAssets/MPC_TOD}}
 LSTemp {type=LevelSequence, default=&LS_Showcase}
 MaxAngle {type=number}
 MQBase {type=MoviePipelineQueue, default=/Game/NeverCook/ArtTools/AssetShowcase/MRQAssets/MQ_Base}
 OffsetName {type=string, default="Jonesy Offset: "}
 OpenLevelSeq {type=boolean}
 orbitCamera {type=boolean}
 PreRollFrames {type=integer, default=30}
 RenderImmediately {type=boolean}
 SaveData {type=SG_ShowcaseRel_C, default=null}
 ScaleActor {type=Class, default=/Game/Blueprints/Blutilities/ForLDs/BPs/TestScaleEditor/TestScaleEditorActor.TestScaleEditorActor_C}
 selectedActor {type=Actor, default=null}
 selectedActorClass {type=Class, default=null}
 selectedActorSM {type=StaticMesh, default=null}
 selectedWasHidden {type=boolean}
 SeqFrames {type=integer}
 STMessageLevels {default {
   Error=/Game/NeverCook/ArtTools/AssetShowcase/Data/AS_ErrorStyle
   Information=/Game/NeverCook/ArtTools/AssetShowcase/Data/AS_InfoStyle
   Warning=/Game/NeverCook/ArtTools/AssetShowcase/Data/AS_WarningStyle
  }}
 STMsgLevelColors {default {Error=<rgba 0.0312 0.0005 0 0.5>, Information=<rgba 0.0381 0.0729 0.0417 1>, Warning=<rgba 0.276 0.139 0 0.5>}}
 tempDir {type=string, default=""}
 TODActor {type=FortDaySequenceActor, default=null}
 TODBindingProxy {type=MovieSceneBindingProxy, default {bindingId=#4, sequence=null}}
 TODCall {default=()}
 TODMax {type=number}
 TODMin {type=number}
 TODPlayer {type=DaySequencePlayer, default=null}
 TODRef {type=MASTER_TODM_C, default=null}
 TODSequence {type=MovieSceneSequence, default=null}
 TransformData {default {}}
 useTODRange {type=boolean}
 useTurnTable {type=boolean}
 YOffset {type=number}
]
widgets {
 info {rootWidgetClass=&CanvasPanel, widgetCount=85}
 widgets [
  CanvasPanel_0 {widgetClassPath=&CanvasPanel, widget=CanvasPanel_0}
  VerticalBox_53 {
   widgetClassPath=&VerticalBox
   parent=CanvasPanel_0
   widget=VerticalBox_53
   slot {
    class=CanvasPanelSlot
    properties {bAutoSize=true, layoutData {alignment=<0.5 0>, anchors {maximum=<0.5 0>, minimum=<0.5 0>}, offsets=<ltrb 0 0 1921.9219 1081.0811>}}
   }
  }
  HorizontalBox_195 {widgetClassPath=&HorizontalBox, parent=VerticalBox_53, widget=HorizontalBox_195, slot {class=VerticalBoxSlot}}
  VerticalBox_110 {
   widgetClassPath=&VerticalBox
   parent=HorizontalBox_195
   widget=VerticalBox_110
   slot {class=HorizontalBoxSlot, properties {padding=<ltrb 5 5 10 10>}}
  }
  renderNow {
   widgetClassPath=&EditorUtilityCheckBox
   bIsVariable=true
   parent=VerticalBox_110
   widget=renderNow
   properties {widgetStyle {
     checkBoxType=CheckBox
     backgroundHoveredImage {
      bIsDynamicallyLoaded=false
      drawAs=RoundedBox
      imageType=NoImage
      mirroring=NoMirror
      resourceName=None
      resourceObject=null
      tiling=NoTile
      imageSize=<18 18>
      margin=<ltrb 0 0 0 0>
      outlineSettings {
       bUseBrushTransparency=false
       roundingType=FixedRadius
       width=1
       color {colorUseRule=UseColor_ColorTable, specifiedColor=<rgba 0 0 0 0>}
       cornerRadii=<3 3 3 3>
      }
      tintColor {colorUseRule=UseColor_Specified, specifiedColor=<rgba 0.0048 0.0048 0.0048 1>}
      uVRegion {bIsValid=false, max=<0 0>, min=<0 0>}
     }
     backgroundImage {
      bIsDynamicallyLoaded=false
      drawAs=RoundedBox
      imageType=NoImage
      mirroring=NoMirror
      resourceName=None
      resourceObject=null
      tiling=NoTile
      imageSize=<18 18>
      margin=<ltrb 0 0 0 0>
      outlineSettings {
       bUseBrushTransparency=false
       roundingType=FixedRadius
       width=1
       color {colorUseRule=UseColor_ColorTable, specifiedColor=<rgba 0 0 0 0>}
       cornerRadii=<3 3 3 3>
      }
      tintColor {colorUseRule=UseColor_Specified, specifiedColor=<rgba 0.0048 0.0048 0.0048 1>}
      uVRegion {bIsValid=false, max=<0 0>, min=<0 0>}
     }
     backgroundPressedImage {
      bIsDynamicallyLoaded=false
      drawAs=RoundedBox
      imageType=NoImage
      mirroring=NoMirror
      resourceName=None
      resourceObject=null
      tiling=NoTile
      imageSize=<18 18>
      margin=<ltrb 0 0 0 0>
      outlineSettings {
       bUseBrushTransparency=false
       roundingType=FixedRadius
       width=1
       color {colorUseRule=UseColor_ColorTable, specifiedColor=<rgba 0 0 0 0>}
       cornerRadii=<3 3 3 3>
      }
      tintColor {colorUseRule=UseColor_Specified, specifiedColor=<rgba 0.0048 0.0048 0.0048 1>}
      uVRegion {bIsValid=false, max=<0 0>, min=<0 0>}
     }
     borderBackgroundColor {colorUseRule=UseColor_Specified, specifiedColor=<rgba 1 1 1 1>}
     checkedForeground {colorUseRule=UseColor_Specified, specifiedColor=<rgba 1 1 1 1>}
     checkedHoveredForeground {colorUseRule=UseColor_Specified, specifiedColor=<rgba 1 1 1 1>}
     checkedHoveredImage {
      bIsDynamicallyLoaded=false
      drawAs=Image
      imageType=Vector
      mirroring=NoMirror
      resourceName=../../../Engine/Content/Slate/Starship/CoreWidgets/CheckBox/check.svg
      resourceObject=null
      tiling=NoTile
      imageSize=<16 16>
      margin=<ltrb 0 0 0 0>
      outlineSettings {
       bUseBrushTransparency=false
       roundingType=HalfHeightRadius
       width=0
       color {colorUseRule=UseColor_Specified, specifiedColor=<rgba 0 0 0 0>}
       cornerRadii=<0 0 0 0>
      }
      tintColor {colorUseRule=UseColor_Specified, specifiedColor=<rgba 0.0044 0.2384 1 1>}
      uVRegion {bIsValid=false, max=<0 0>, min=<0 0>}
     }
     checkedImage {
      bIsDynamicallyLoaded=false
      drawAs=Image
      imageType=Vector
      mirroring=NoMirror
      resourceName=../../../Engine/Content/Slate/Starship/CoreWidgets/CheckBox/check.svg
      resourceObject=null
      tiling=NoTile
      imageSize=<16 16>
      margin=<ltrb 0 0 0 0>
      outlineSettings {
       bUseBrushTransparency=false
       roundingType=HalfHeightRadius
       width=0
       color {colorUseRule=UseColor_Specified, specifiedColor=<rgba 0 0 0 0>}
       cornerRadii=<0 0 0 0>
      }
      tintColor {colorUseRule=UseColor_Specified, specifiedColor=<rgba 0 0.162 0.7454 1>}
      uVRegion {bIsValid=false, max=<0 0>, min=<0 0>}
     }
     checkedPressedForeground {colorUseRule=UseColor_Specified, specifiedColor=<rgba 1 1 1 1>}
     checkedPressedImage {
      bIsDynamicallyLoaded=false
      drawAs=Image
      imageType=Vector
      mirroring=NoMirror
      resourceName=../../../Engine/Content/Slate/Starship/CoreWidgets/CheckBox/check.svg
      resourceObject=null
      tiling=NoTile
      imageSize=<16 16>
      margin=<ltrb 0 0 0 0>
      outlineSettings {
       bUseBrushTransparency=false
       roundingType=HalfHeightRadius
       width=0
       color {colorUseRule=UseColor_Specified, specifiedColor=<rgba 0 0 0 0>}
       cornerRadii=<0 0 0 0>
      }
      tintColor {colorUseRule=UseColor_Specified, specifiedColor=<rgba 0 0.0802 0.3515 1>}
      uVRegion {bIsValid=false, max=<0 0>, min=<0 0>}
     }
     checkedSlateSound {resourceObject=null}
     foregroundColor {colorUseRule=UseColor_Specified, specifiedColor=<rgba 1 1 1 1>}
     hoveredForeground {colorUseRule=UseColor_Specified, specifiedColor=<rgba 1 1 1 1>}
     hoveredSlateSound {resourceObject=null}
     padding=<ltrb 2 0 0 0>
     pressedForeground {colorUseRule=UseColor_Specified, specifiedColor=<rgba 1 1 1 1>}
     uncheckedHoveredImage {
      bIsDynamicallyLoaded=false
      drawAs=NoDrawType
      imageType=NoImage
      mirroring=NoMirror
      resourceName=None
      resourceObject=null
      tiling=NoTile
      imageSize=<0 0>
      margin=<ltrb 0 0 0 0>
      outlineSettings {
       bUseBrushTransparency=false
       roundingType=HalfHeightRadius
       width=0
       color {colorUseRule=UseColor_Specified, specifiedColor=<rgba 0 0 0 0>}
       cornerRadii=<0 0 0 0>
      }
      tintColor {colorUseRule=UseColor_Specified, specifiedColor=<rgba 1 1 1 1>}
      uVRegion {bIsValid=false, max=<0 0>, min=<0 0>}
     }
     uncheckedImage {
      bIsDynamicallyLoaded=false
      drawAs=NoDrawType
      imageType=NoImage
      mirroring=NoMirror
      resourceName=None
      resourceObject=null
      tiling=NoTile
      imageSize=<0 0>
      margin=<ltrb 0 0 0 0>
      outlineSettings {
       bUseBrushTransparency=false
       roundingType=HalfHeightRadius
       width=0
       color {colorUseRule=UseColor_Specified, specifiedColor=<rgba 0 0 0 0>}
       cornerRadii=<0 0 0 0>
      }
      tintColor {colorUseRule=UseColor_Specified, specifiedColor=<rgba 1 1 1 1>}
      uVRegion {bIsValid=false, max=<0 0>, min=<0 0>}
     }
     uncheckedPressedImage {
      bIsDynamicallyLoaded=false
      drawAs=NoDrawType
      imageType=NoImage
      mirroring=NoMirror
      resourceName=None
      resourceObject=null
      tiling=NoTile
      imageSize=<0 0>
      margin=<ltrb 0 0 0 0>
      outlineSettings {
       bUseBrushTransparency=false
       roundingType=HalfHeightRadius
       width=0
       color {colorUseRule=UseColor_Specified, specifiedColor=<rgba 0 0 0 0>}
       cornerRadii=<0 0 0 0>
      }
      tintColor {colorUseRule=UseColor_Specified, specifiedColor=<rgba 1 1 1 1>}
      uVRegion {bIsValid=false, max=<0 0>, min=<0 0>}
     }
     uncheckedSlateSound {resourceObject=null}
     undeterminedForeground {colorUseRule=UseColor_Specified, specifiedColor=<rgba 1 1 1 1>}
     undeterminedHoveredImage {
      bIsDynamicallyLoaded=false
      drawAs=Image
      imageType=Vector
      mirroring=NoMirror
      resourceName=../../../Engine/Content/Slate/Starship/CoreWidgets/CheckBox/indeterminate.svg
      resourceObject=null
      tiling=NoTile
      imageSize=<16 16>
      margin=<ltrb 0 0 0 0>
      outlineSettings {
       bUseBrushTransparency=false
       roundingType=HalfHeightRadius
       width=0
       color {colorUseRule=UseColor_Specified, specifiedColor=<rgba 0 0 0 0>}
       cornerRadii=<0 0 0 0>
      }
      tintColor {colorUseRule=UseColor_Specified, specifiedColor=<rgba 0.0044 0.2384 1 1>}
      uVRegion {bIsValid=false, max=<0 0>, min=<0 0>}
     }
     undeterminedImage {
      bIsDynamicallyLoaded=false
      drawAs=Image
      imageType=Vector
      mirroring=NoMirror
      resourceName=../../../Engine/Content/Slate/Starship/CoreWidgets/CheckBox/indeterminate.svg
      resourceObject=null
      tiling=NoTile
      imageSize=<16 16>
      margin=<ltrb 0 0 0 0>
      outlineSettings {
       bUseBrushTransparency=false
       roundingType=HalfHeightRadius
       width=0
       color {colorUseRule=UseColor_Specified, specifiedColor=<rgba 0 0 0 0>}
       cornerRadii=<0 0 0 0>
      }
      tintColor {colorUseRule=UseColor_Specified, specifiedColor=<rgba 0 0.162 0.7454 1>}
      uVRegion {bIsValid=false, max=<0 0>, min=<0 0>}
     }
     undeterminedPressedImage {
      bIsDynamicallyLoaded=false
      drawAs=Image
      imageType=Vector
      mirroring=NoMirror
      resourceName=../../../Engine/Content/Slate/Starship/CoreWidgets/CheckBox/indeterminate.svg
      resourceObject=null
      tiling=NoTile
      imageSize=<16 16>
      margin=<ltrb 0 0 0 0>
      outlineSettings {
       bUseBrushTransparency=false
       roundingType=HalfHeightRadius
       width=0
       color {colorUseRule=UseColor_Specified, specifiedColor=<rgba 0 0 0 0>}
       cornerRadii=<0 0 0 0>
      }
      tintColor {colorUseRule=UseColor_Specified, specifiedColor=<rgba 0 0.0802 0.3515 1>}
      uVRegion {bIsValid=false, max=<0 0>, min=<0 0>}
     }
    }}
   slot {class=VerticalBoxSlot, properties {verticalAlignment=VAlign_Top, padding=<ltrb 0 10 0 0>}}
  }
  UEFN_TextBlock_C_21 {
   widgetClassPath=&UEFN_TextBlock
   parent=renderNow
   widget=UEFN_TextBlock_C_21
   properties {
    justification=Center
    text=Render Queue
    toolTipText=Begins rendering of the generated Level Sequence using MRQ
    font {
     bForceMonospaced=false
     bMaterialIsStencil=false
     fontMaterial=null
     fontObject=&BurbankSmall
     letterSpacing=0
     monospacedWidth=1
     size=9
     skewAmount=0
     typefaceFontName=Medium
     outlineSettings {
      bApplyOutlineToDropShadows=false
      bMiteredCorners=false
      bSeparateFillAlpha=false
      outlineMaterial=null
      outlineSize=0
      outlineColor=<rgba 0 0 0 1>
     }
    }
   }
   slot {class=PanelSlot}
  }
  openSequence {
   widgetClassPath=&EditorUtilityCheckBox
   bIsVariable=true
   parent=VerticalBox_110
   widget=openSequence
   properties {
    checkedState=Checked
    widgetStyle {
     checkBoxType=CheckBox
     backgroundHoveredImage {
      bIsDynamicallyLoaded=false
      drawAs=RoundedBox
      imageType=NoImage
      mirroring=NoMirror
      resourceName=None
      resourceObject=null
      tiling=NoTile
      imageSize=<18 18>
      margin=<ltrb 0 0 0 0>
      outlineSettings {
       bUseBrushTransparency=false
       roundingType=FixedRadius
       width=1
       color {colorUseRule=UseColor_ColorTable, specifiedColor=<rgba 0 0 0 0>}
       cornerRadii=<3 3 3 3>
      }
      tintColor {colorUseRule=UseColor_Specified, specifiedColor=<rgba 0.0048 0.0048 0.0048 1>}
      uVRegion {bIsValid=false, max=<0 0>, min=<0 0>}
     }
     backgroundImage {
      bIsDynamicallyLoaded=false
      drawAs=RoundedBox
      imageType=NoImage
      mirroring=NoMirror
      resourceName=None
      resourceObject=null
      tiling=NoTile
      imageSize=<18 18>
      margin=<ltrb 0 0 0 0>
      outlineSettings {
       bUseBrushTransparency=false
       roundingType=FixedRadius
       width=1
       color {colorUseRule=UseColor_ColorTable, specifiedColor=<rgba 0 0 0 0>}
       cornerRadii=<3 3 3 3>
      }
      tintColor {colorUseRule=UseColor_Specified, specifiedColor=<rgba 0.0048 0.0048 0.0048 1>}
      uVRegion {bIsValid=false, max=<0 0>, min=<0 0>}
     }
     backgroundPressedImage {
      bIsDynamicallyLoaded=false
      drawAs=RoundedBox
      imageType=NoImage
      mirroring=NoMirror
      resourceName=None
      resourceObject=null
      tiling=NoTile
      imageSize=<18 18>
      margin=<ltrb 0 0 0 0>
      outlineSettings {
       bUseBrushTransparency=false
       roundingType=FixedRadius
       width=1
       color {colorUseRule=UseColor_ColorTable, specifiedColor=<rgba 0 0 0 0>}
       cornerRadii=<3 3 3 3>
      }
      tintColor {colorUseRule=UseColor_Specified, specifiedColor=<rgba 0.0048 0.0048 0.0048 1>}
      uVRegion {bIsValid=false, max=<0 0>, min=<0 0>}
     }
     borderBackgroundColor {colorUseRule=UseColor_Specified, specifiedColor=<rgba 1 1 1 1>}
     checkedForeground {colorUseRule=UseColor_Specified, specifiedColor=<rgba 1 1 1 1>}
     checkedHoveredForeground {colorUseRule=UseColor_Specified, specifiedColor=<rgba 1 1 1 1>}
     checkedHoveredImage {
      bIsDynamicallyLoaded=false
      drawAs=Image
      imageType=Vector
      mirroring=NoMirror
      resourceName=../../../Engine/Content/Slate/Starship/CoreWidgets/CheckBox/check.svg
      resourceObject=null
      tiling=NoTile
      imageSize=<16 16>
      margin=<ltrb 0 0 0 0>
      outlineSettings {
       bUseBrushTransparency=false
       roundingType=HalfHeightRadius
       width=0
       color {colorUseRule=UseColor_Specified, specifiedColor=<rgba 0 0 0 0>}
       cornerRadii=<0 0 0 0>
      }
      tintColor {colorUseRule=UseColor_Specified, specifiedColor=<rgba 0.0044 0.2384 1 1>}
      uVRegion {bIsValid=false, max=<0 0>, min=<0 0>}
     }
     checkedImage {
      bIsDynamicallyLoaded=false
      drawAs=Image
      imageType=Vector
      mirroring=NoMirror
      resourceName=../../../Engine/Content/Slate/Starship/CoreWidgets/CheckBox/check.svg
      resourceObject=null
      tiling=NoTile
      imageSize=<16 16>
      margin=<ltrb 0 0 0 0>
      outlineSettings {
       bUseBrushTransparency=false
       roundingType=HalfHeightRadius
       width=0
       color {colorUseRule=UseColor_Specified, specifiedColor=<rgba 0 0 0 0>}
       cornerRadii=<0 0 0 0>
      }
      tintColor {colorUseRule=UseColor_Specified, specifiedColor=<rgba 0 0.162 0.7454 1>}
      uVRegion {bIsValid=false, max=<0 0>, min=<0 0>}
     }
     checkedPressedForeground {colorUseRule=UseColor_Specified, specifiedColor=<rgba 1 1 1 1>}
     checkedPressedImage {
      bIsDynamicallyLoaded=false
      drawAs=Image
      imageType=Vector
      mirroring=NoMirror
      resourceName=../../../Engine/Content/Slate/Starship/CoreWidgets/CheckBox/check.svg
      resourceObject=null
      tiling=NoTile
      imageSize=<16 16>
      margin=<ltrb 0 0 0 0>
      outlineSettings {
       bUseBrushTransparency=false
       roundingType=HalfHeightRadius
       width=0
       color {colorUseRule=UseColor_Specified, specifiedColor=<rgba 0 0 0 0>}
       cornerRadii=<0 0 0 0>
      }
      tintColor {colorUseRule=UseColor_Specified, specifiedColor=<rgba 0 0.0802 0.3515 1>}
      uVRegion {bIsValid=false, max=<0 0>, min=<0 0>}
     }
     checkedSlateSound {resourceObject=null}
     foregroundColor {colorUseRule=UseColor_Specified, specifiedColor=<rgba 1 1 1 1>}
     hoveredForeground {colorUseRule=UseColor_Specified, specifiedColor=<rgba 1 1 1 1>}
     hoveredSlateSound {resourceObject=null}
     padding=<ltrb 2 0 0 0>
     pressedForeground {colorUseRule=UseColor_Specified, specifiedColor=<rgba 1 1 1 1>}
     uncheckedHoveredImage {
      bIsDynamicallyLoaded=false
      drawAs=NoDrawType
      imageType=NoImage
      mirroring=NoMirror
      resourceName=None
      resourceObject=null
      tiling=NoTile
      imageSize=<0 0>
      margin=<ltrb 0 0 0 0>
      outlineSettings {
       bUseBrushTransparency=false
       roundingType=HalfHeightRadius
       width=0
       color {colorUseRule=UseColor_Specified, specifiedColor=<rgba 0 0 0 0>}
       cornerRadii=<0 0 0 0>
      }
      tintColor {colorUseRule=UseColor_Specified, specifiedColor=<rgba 1 1 1 1>}
      uVRegion {bIsValid=false, max=<0 0>, min=<0 0>}
     }
     uncheckedImage {
      bIsDynamicallyLoaded=false
      drawAs=NoDrawType
      imageType=NoImage
      mirroring=NoMirror
      resourceName=None
      resourceObject=null
      tiling=NoTile
      imageSize=<0 0>
      margin=<ltrb 0 0 0 0>
      outlineSettings {
       bUseBrushTransparency=false
       roundingType=HalfHeightRadius
       width=0
       color {colorUseRule=UseColor_Specified, specifiedColor=<rgba 0 0 0 0>}
       cornerRadii=<0 0 0 0>
      }
      tintColor {colorUseRule=UseColor_Specified, specifiedColor=<rgba 1 1 1 1>}
      uVRegion {bIsValid=false, max=<0 0>, min=<0 0>}
     }
     uncheckedPressedImage {
      bIsDynamicallyLoaded=false
      drawAs=NoDrawType
      imageType=NoImage
      mirroring=NoMirror
      resourceName=None
      resourceObject=null
      tiling=NoTile
      imageSize=<0 0>
      margin=<ltrb 0 0 0 0>
      outlineSettings {
       bUseBrushTransparency=false
       roundingType=HalfHeightRadius
       width=0
       color {colorUseRule=UseColor_Specified, specifiedColor=<rgba 0 0 0 0>}
       cornerRadii=<0 0 0 0>
      }
      tintColor {colorUseRule=UseColor_Specified, specifiedColor=<rgba 1 1 1 1>}
      uVRegion {bIsValid=false, max=<0 0>, min=<0 0>}
     }
     uncheckedSlateSound {resourceObject=null}
     undeterminedForeground {colorUseRule=UseColor_Specified, specifiedColor=<rgba 1 1 1 1>}
     undeterminedHoveredImage {
      bIsDynamicallyLoaded=false
      drawAs=Image
      imageType=Vector
      mirroring=NoMirror
      resourceName=../../../Engine/Content/Slate/Starship/CoreWidgets/CheckBox/indeterminate.svg
      resourceObject=null
      tiling=NoTile
      imageSize=<16 16>
      margin=<ltrb 0 0 0 0>
      outlineSettings {
       bUseBrushTransparency=false
       roundingType=HalfHeightRadius
       width=0
       color {colorUseRule=UseColor_Specified, specifiedColor=<rgba 0 0 0 0>}
       cornerRadii=<0 0 0 0>
      }
      tintColor {colorUseRule=UseColor_Specified, specifiedColor=<rgba 0.0044 0.2384 1 1>}
      uVRegion {bIsValid=false, max=<0 0>, min=<0 0>}
     }
     undeterminedImage {
      bIsDynamicallyLoaded=false
      drawAs=Image
      imageType=Vector
      mirroring=NoMirror
      resourceName=../../../Engine/Content/Slate/Starship/CoreWidgets/CheckBox/indeterminate.svg
      resourceObject=null
      tiling=NoTile
      imageSize=<16 16>
      margin=<ltrb 0 0 0 0>
      outlineSettings {
       bUseBrushTransparency=false
       roundingType=HalfHeightRadius
       width=0
       color {colorUseRule=UseColor_Specified, specifiedColor=<rgba 0 0 0 0>}
       cornerRadii=<0 0 0 0>
      }
      tintColor {colorUseRule=UseColor_Specified, specifiedColor=<rgba 0 0.162 0.7454 1>}
      uVRegion {bIsValid=false, max=<0 0>, min=<0 0>}
     }
     undeterminedPressedImage {
      bIsDynamicallyLoaded=false
      drawAs=Image
      imageType=Vector
      mirroring=NoMirror
      resourceName=../../../Engine/Content/Slate/Starship/CoreWidgets/CheckBox/indeterminate.svg
      resourceObject=null
      tiling=NoTile
      imageSize=<16 16>
      margin=<ltrb 0 0 0 0>
      outlineSettings {
       bUseBrushTransparency=false
       roundingType=HalfHeightRadius
       width=0
       color {colorUseRule=UseColor_Specified, specifiedColor=<rgba 0 0 0 0>}
       cornerRadii=<0 0 0 0>
      }
      tintColor {colorUseRule=UseColor_Specified, specifiedColor=<rgba 0 0.0802 0.3515 1>}
      uVRegion {bIsValid=false, max=<0 0>, min=<0 0>}
     }
    }
   }
   slot {class=VerticalBoxSlot, properties {verticalAlignment=VAlign_Top, padding=<ltrb 0 5 0 0>}}
  }
  UEFN_TextBlock_C_22 {
   widgetClassPath=&UEFN_TextBlock
   parent=openSequence
   widget=UEFN_TextBlock_C_22
   properties {
    justification=Center
    text=Open Sequence
    toolTipText=Opens the generated Level Sequence in Sequencer
    font {
     bForceMonospaced=false
     bMaterialIsStencil=false
     fontMaterial=null
     fontObject=&BurbankSmall
     letterSpacing=0
     monospacedWidth=1
     size=9
     skewAmount=0
     typefaceFontName=Medium
     outlineSettings {
      bApplyOutlineToDropShadows=false
      bMiteredCorners=false
      bSeparateFillAlpha=false
      outlineMaterial=null
      outlineSize=0
      outlineColor=<rgba 0 0 0 1>
     }
    }
   }
   slot {class=PanelSlot}
  }
  SizeRef {
   widgetClassPath=&EditorUtilityCheckBox
   bIsVariable=true
   parent=VerticalBox_110
   widget=SizeRef
   properties {
    checkedState=Checked
    widgetStyle {
     checkBoxType=CheckBox
     backgroundHoveredImage {
      bIsDynamicallyLoaded=false
      drawAs=RoundedBox
      imageType=NoImage
      mirroring=NoMirror
      resourceName=None
      resourceObject=null
      tiling=NoTile
      imageSize=<18 18>
      margin=<ltrb 0 0 0 0>
      outlineSettings {
       bUseBrushTransparency=false
       roundingType=FixedRadius
       width=1
       color {colorUseRule=UseColor_ColorTable, specifiedColor=<rgba 0 0 0 0>}
       cornerRadii=<3 3 3 3>
      }
      tintColor {colorUseRule=UseColor_Specified, specifiedColor=<rgba 0.0048 0.0048 0.0048 1>}
      uVRegion {bIsValid=false, max=<0 0>, min=<0 0>}
     }
     backgroundImage {
      bIsDynamicallyLoaded=false
      drawAs=RoundedBox
      imageType=NoImage
      mirroring=NoMirror
      resourceName=None
      resourceObject=null
      tiling=NoTile
      imageSize=<18 18>
      margin=<ltrb 0 0 0 0>
      outlineSettings {
       bUseBrushTransparency=false
       roundingType=FixedRadius
       width=1
       color {colorUseRule=UseColor_ColorTable, specifiedColor=<rgba 0 0 0 0>}
       cornerRadii=<3 3 3 3>
      }
      tintColor {colorUseRule=UseColor_Specified, specifiedColor=<rgba 0.0048 0.0048 0.0048 1>}
      uVRegion {bIsValid=false, max=<0 0>, min=<0 0>}
     }
     backgroundPressedImage {
      bIsDynamicallyLoaded=false
      drawAs=RoundedBox
      imageType=NoImage
      mirroring=NoMirror
      resourceName=None
      resourceObject=null
      tiling=NoTile
      imageSize=<18 18>
      margin=<ltrb 0 0 0 0>
      outlineSettings {
       bUseBrushTransparency=false
       roundingType=FixedRadius
       width=1
       color {colorUseRule=UseColor_ColorTable, specifiedColor=<rgba 0 0 0 0>}
       cornerRadii=<3 3 3 3>
      }
      tintColor {colorUseRule=UseColor_Specified, specifiedColor=<rgba 0.0048 0.0048 0.0048 1>}
      uVRegion {bIsValid=false, max=<0 0>, min=<0 0>}
     }
     borderBackgroundColor {colorUseRule=UseColor_Specified, specifiedColor=<rgba 1 1 1 1>}
     checkedForeground {colorUseRule=UseColor_Specified, specifiedColor=<rgba 1 1 1 1>}
     checkedHoveredForeground {colorUseRule=UseColor_Specified, specifiedColor=<rgba 1 1 1 1>}
     checkedHoveredImage {
      bIsDynamicallyLoaded=false
      drawAs=Image
      imageType=Vector
      mirroring=NoMirror
      resourceName=../../../Engine/Content/Slate/Starship/CoreWidgets/CheckBox/check.svg
      resourceObject=null
      tiling=NoTile
      imageSize=<16 16>
      margin=<ltrb 0 0 0 0>
      outlineSettings {
       bUseBrushTransparency=false
       roundingType=HalfHeightRadius
       width=0
       color {colorUseRule=UseColor_Specified, specifiedColor=<rgba 0 0 0 0>}
       cornerRadii=<0 0 0 0>
      }
      tintColor {colorUseRule=UseColor_Specified, specifiedColor=<rgba 0.0044 0.2384 1 1>}
      uVRegion {bIsValid=false, max=<0 0>, min=<0 0>}
     }
     checkedImage {
      bIsDynamicallyLoaded=false
      drawAs=Image
      imageType=Vector
      mirroring=NoMirror
      resourceName=../../../Engine/Content/Slate/Starship/CoreWidgets/CheckBox/check.svg
      resourceObject=null
      tiling=NoTile
      imageSize=<16 16>
      margin=<ltrb 0 0 0 0>
      outlineSettings {
       bUseBrushTransparency=false
       roundingType=HalfHeightRadius
       width=0
       color {colorUseRule=UseColor_Specified, specifiedColor=<rgba 0 0 0 0>}
       cornerRadii=<0 0 0 0>
      }
      tintColor {colorUseRule=UseColor_Specified, specifiedColor=<rgba 0 0.162 0.7454 1>}
      uVRegion {bIsValid=false, max=<0 0>, min=<0 0>}
     }
     checkedPressedForeground {colorUseRule=UseColor_Specified, specifiedColor=<rgba 1 1 1 1>}
     checkedPressedImage {
      bIsDynamicallyLoaded=false
      drawAs=Image
      imageType=Vector
      mirroring=NoMirror
      resourceName=../../../Engine/Content/Slate/Starship/CoreWidgets/CheckBox/check.svg
      resourceObject=null
      tiling=NoTile
      imageSize=<16 16>
      margin=<ltrb 0 0 0 0>
      outlineSettings {
       bUseBrushTransparency=false
       roundingType=HalfHeightRadius
       width=0
       color {colorUseRule=UseColor_Specified, specifiedColor=<rgba 0 0 0 0>}
       cornerRadii=<0 0 0 0>
      }
      tintColor {colorUseRule=UseColor_Specified, specifiedColor=<rgba 0 0.0802 0.3515 1>}
      uVRegion {bIsValid=false, max=<0 0>, min=<0 0>}
     }
     checkedSlateSound {resourceObject=null}
     foregroundColor {colorUseRule=UseColor_Specified, specifiedColor=<rgba 1 1 1 1>}
     hoveredForeground {colorUseRule=UseColor_Specified, specifiedColor=<rgba 1 1 1 1>}
     hoveredSlateSound {resourceObject=null}
     padding=<ltrb 2 0 0 0>
     pressedForeground {colorUseRule=UseColor_Specified, specifiedColor=<rgba 1 1 1 1>}
     uncheckedHoveredImage {
      bIsDynamicallyLoaded=false
      drawAs=NoDrawType
      imageType=NoImage
      mirroring=NoMirror
      resourceName=None
      resourceObject=null
      tiling=NoTile
      imageSize=<0 0>
      margin=<ltrb 0 0 0 0>
      outlineSettings {
       bUseBrushTransparency=false
       roundingType=HalfHeightRadius
       width=0
       color {colorUseRule=UseColor_Specified, specifiedColor=<rgba 0 0 0 0>}
       cornerRadii=<0 0 0 0>
      }
      tintColor {colorUseRule=UseColor_Specified, specifiedColor=<rgba 1 1 1 1>}
      uVRegion {bIsValid=false, max=<0 0>, min=<0 0>}
     }
     uncheckedImage {
      bIsDynamicallyLoaded=false
      drawAs=NoDrawType
      imageType=NoImage
      mirroring=NoMirror
      resourceName=None
      resourceObject=null
      tiling=NoTile
      imageSize=<0 0>
      margin=<ltrb 0 0 0 0>
      outlineSettings {
       bUseBrushTransparency=false
       roundingType=HalfHeightRadius
       width=0
       color {colorUseRule=UseColor_Specified, specifiedColor=<rgba 0 0 0 0>}
       cornerRadii=<0 0 0 0>
      }
      tintColor {colorUseRule=UseColor_Specified, specifiedColor=<rgba 1 1 1 1>}
      uVRegion {bIsValid=false, max=<0 0>, min=<0 0>}
     }
     uncheckedPressedImage {
      bIsDynamicallyLoaded=false
      drawAs=NoDrawType
      imageType=NoImage
      mirroring=NoMirror
      resourceName=None
      resourceObject=null
      tiling=NoTile
      imageSize=<0 0>
      margin=<ltrb 0 0 0 0>
      outlineSettings {
       bUseBrushTransparency=false
       roundingType=HalfHeightRadius
       width=0
       color {colorUseRule=UseColor_Specified, specifiedColor=<rgba 0 0 0 0>}
       cornerRadii=<0 0 0 0>
      }
      tintColor {colorUseRule=UseColor_Specified, specifiedColor=<rgba 1 1 1 1>}
      uVRegion {bIsValid=false, max=<0 0>, min=<0 0>}
     }
     uncheckedSlateSound {resourceObject=null}
     undeterminedForeground {colorUseRule=UseColor_Specified, specifiedColor=<rgba 1 1 1 1>}
     undeterminedHoveredImage {
      bIsDynamicallyLoaded=false
      drawAs=Image
      imageType=Vector
      mirroring=NoMirror
      resourceName=../../../Engine/Content/Slate/Starship/CoreWidgets/CheckBox/indeterminate.svg
      resourceObject=null
      tiling=NoTile
      imageSize=<16 16>
      margin=<ltrb 0 0 0 0>
      outlineSettings {
       bUseBrushTransparency=false
       roundingType=HalfHeightRadius
       width=0
       color {colorUseRule=UseColor_Specified, specifiedColor=<rgba 0 0 0 0>}
       cornerRadii=<0 0 0 0>
      }
      tintColor {colorUseRule=UseColor_Specified, specifiedColor=<rgba 0.0044 0.2384 1 1>}
      uVRegion {bIsValid=false, max=<0 0>, min=<0 0>}
     }
     undeterminedImage {
      bIsDynamicallyLoaded=false
      drawAs=Image
      imageType=Vector
      mirroring=NoMirror
      resourceName=../../../Engine/Content/Slate/Starship/CoreWidgets/CheckBox/indeterminate.svg
      resourceObject=null
      tiling=NoTile
      imageSize=<16 16>
      margin=<ltrb 0 0 0 0>
      outlineSettings {
       bUseBrushTransparency=false
       roundingType=HalfHeightRadius
       width=0
       color {colorUseRule=UseColor_Specified, specifiedColor=<rgba 0 0 0 0>}
       cornerRadii=<0 0 0 0>
      }
      tintColor {colorUseRule=UseColor_Specified, specifiedColor=<rgba 0 0.162 0.7454 1>}
      uVRegion {bIsValid=false, max=<0 0>, min=<0 0>}
     }
     undeterminedPressedImage {
      bIsDynamicallyLoaded=false
      drawAs=Image
      imageType=Vector
      mirroring=NoMirror
      resourceName=../../../Engine/Content/Slate/Starship/CoreWidgets/CheckBox/indeterminate.svg
      resourceObject=null
      tiling=NoTile
      imageSize=<16 16>
      margin=<ltrb 0 0 0 0>
      outlineSettings {
       bUseBrushTransparency=false
       roundingType=HalfHeightRadius
       width=0
       color {colorUseRule=UseColor_Specified, specifiedColor=<rgba 0 0 0 0>}
       cornerRadii=<0 0 0 0>
      }
      tintColor {colorUseRule=UseColor_Specified, specifiedColor=<rgba 0 0.0802 0.3515 1>}
      uVRegion {bIsValid=false, max=<0 0>, min=<0 0>}
     }
    }
   }
   slot {class=VerticalBoxSlot, properties {verticalAlignment=VAlign_Bottom, padding=<ltrb 0 25 0 0>}}
  }
  UEFN_TextBlock_C_24 {
   widgetClassPath=&UEFN_TextBlock
   parent=SizeRef
   widget=UEFN_TextBlock_C_24
   properties {
    justification=Center
    text=Size Reference
    toolTipText=Adds a Jonesy for size reference
    font {
     bForceMonospaced=false
     bMaterialIsStencil=false
     fontMaterial=null
     fontObject=&BurbankSmall
     letterSpacing=0
     monospacedWidth=1
     size=9
     skewAmount=0
     typefaceFontName=Medium
     outlineSettings {
      bApplyOutlineToDropShadows=false
      bMiteredCorners=false
      bSeparateFillAlpha=false
      outlineMaterial=null
      outlineSize=0
      outlineColor=<rgba 0 0 0 1>
     }
    }
   }
   slot {class=PanelSlot}
  }
  JunoRef {
   widgetClassPath=&EditorUtilityCheckBox
   bIsVariable=true
   parent=VerticalBox_110
   widget=JunoRef
   properties {widgetStyle {
     checkBoxType=CheckBox
     backgroundHoveredImage {
      bIsDynamicallyLoaded=false
      drawAs=RoundedBox
      imageType=NoImage
      mirroring=NoMirror
      resourceName=None
      resourceObject=null
      tiling=NoTile
      imageSize=<18 18>
      margin=<ltrb 0 0 0 0>
      outlineSettings {
       bUseBrushTransparency=false
       roundingType=FixedRadius
       width=1
       color {colorUseRule=UseColor_ColorTable, specifiedColor=<rgba 0 0 0 0>}
       cornerRadii=<3 3 3 3>
      }
      tintColor {colorUseRule=UseColor_Specified, specifiedColor=<rgba 0.0048 0.0048 0.0048 1>}
      uVRegion {bIsValid=false, max=<0 0>, min=<0 0>}
     }
     backgroundImage {
      bIsDynamicallyLoaded=false
      drawAs=RoundedBox
      imageType=NoImage
      mirroring=NoMirror
      resourceName=None
      resourceObject=null
      tiling=NoTile
      imageSize=<18 18>
      margin=<ltrb 0 0 0 0>
      outlineSettings {
       bUseBrushTransparency=false
       roundingType=FixedRadius
       width=1
       color {colorUseRule=UseColor_ColorTable, specifiedColor=<rgba 0 0 0 0>}
       cornerRadii=<3 3 3 3>
      }
      tintColor {colorUseRule=UseColor_Specified, specifiedColor=<rgba 0.0048 0.0048 0.0048 1>}
      uVRegion {bIsValid=false, max=<0 0>, min=<0 0>}
     }
     backgroundPressedImage {
      bIsDynamicallyLoaded=false
      drawAs=RoundedBox
      imageType=NoImage
      mirroring=NoMirror
      resourceName=None
      resourceObject=null
      tiling=NoTile
      imageSize=<18 18>
      margin=<ltrb 0 0 0 0>
      outlineSettings {
       bUseBrushTransparency=false
       roundingType=FixedRadius
       width=1
       color {colorUseRule=UseColor_ColorTable, specifiedColor=<rgba 0 0 0 0>}
       cornerRadii=<3 3 3 3>
      }
      tintColor {colorUseRule=UseColor_Specified, specifiedColor=<rgba 0.0048 0.0048 0.0048 1>}
      uVRegion {bIsValid=false, max=<0 0>, min=<0 0>}
     }
     borderBackgroundColor {colorUseRule=UseColor_Specified, specifiedColor=<rgba 1 1 1 1>}
     checkedForeground {colorUseRule=UseColor_Specified, specifiedColor=<rgba 1 1 1 1>}
     checkedHoveredForeground {colorUseRule=UseColor_Specified, specifiedColor=<rgba 1 1 1 1>}
     checkedHoveredImage {
      bIsDynamicallyLoaded=false
      drawAs=Image
      imageType=Vector
      mirroring=NoMirror
      resourceName=../../../Engine/Content/Slate/Starship/CoreWidgets/CheckBox/check.svg
      resourceObject=null
      tiling=NoTile
      imageSize=<16 16>
      margin=<ltrb 0 0 0 0>
      outlineSettings {
       bUseBrushTransparency=false
       roundingType=HalfHeightRadius
       width=0
       color {colorUseRule=UseColor_Specified, specifiedColor=<rgba 0 0 0 0>}
       cornerRadii=<0 0 0 0>
      }
      tintColor {colorUseRule=UseColor_Specified, specifiedColor=<rgba 0.0044 0.2384 1 1>}
      uVRegion {bIsValid=false, max=<0 0>, min=<0 0>}
     }
     checkedImage {
      bIsDynamicallyLoaded=false
      drawAs=Image
      imageType=Vector
      mirroring=NoMirror
      resourceName=../../../Engine/Content/Slate/Starship/CoreWidgets/CheckBox/check.svg
      resourceObject=null
      tiling=NoTile
      imageSize=<16 16>
      margin=<ltrb 0 0 0 0>
      outlineSettings {
       bUseBrushTransparency=false
       roundingType=HalfHeightRadius
       width=0
       color {colorUseRule=UseColor_Specified, specifiedColor=<rgba 0 0 0 0>}
       cornerRadii=<0 0 0 0>
      }
      tintColor {colorUseRule=UseColor_Specified, specifiedColor=<rgba 0 0.162 0.7454 1>}
      uVRegion {bIsValid=false, max=<0 0>, min=<0 0>}
     }
     checkedPressedForeground {colorUseRule=UseColor_Specified, specifiedColor=<rgba 1 1 1 1>}
     checkedPressedImage {
      bIsDynamicallyLoaded=false
      drawAs=Image
      imageType=Vector
      mirroring=NoMirror
      resourceName=../../../Engine/Content/Slate/Starship/CoreWidgets/CheckBox/check.svg
      resourceObject=null
      tiling=NoTile
      imageSize=<16 16>
      margin=<ltrb 0 0 0 0>
      outlineSettings {
       bUseBrushTransparency=false
       roundingType=HalfHeightRadius
       width=0
       color {colorUseRule=UseColor_Specified, specifiedColor=<rgba 0 0 0 0>}
       cornerRadii=<0 0 0 0>
      }
      tintColor {colorUseRule=UseColor_Specified, specifiedColor=<rgba 0 0.0802 0.3515 1>}
      uVRegion {bIsValid=false, max=<0 0>, min=<0 0>}
     }
     checkedSlateSound {resourceObject=null}
     foregroundColor {colorUseRule=UseColor_Specified, specifiedColor=<rgba 1 1 1 1>}
     hoveredForeground {colorUseRule=UseColor_Specified, specifiedColor=<rgba 1 1 1 1>}
     hoveredSlateSound {resourceObject=null}
     padding=<ltrb 2 0 0 0>
     pressedForeground {colorUseRule=UseColor_Specified, specifiedColor=<rgba 1 1 1 1>}
     uncheckedHoveredImage {
      bIsDynamicallyLoaded=false
      drawAs=NoDrawType
      imageType=NoImage
      mirroring=NoMirror
      resourceName=None
      resourceObject=null
      tiling=NoTile
      imageSize=<0 0>
      margin=<ltrb 0 0 0 0>
      outlineSettings {
       bUseBrushTransparency=false
       roundingType=HalfHeightRadius
       width=0
       color {colorUseRule=UseColor_Specified, specifiedColor=<rgba 0 0 0 0>}
       cornerRadii=<0 0 0 0>
      }
      tintColor {colorUseRule=UseColor_Specified, specifiedColor=<rgba 1 1 1 1>}
      uVRegion {bIsValid=false, max=<0 0>, min=<0 0>}
     }
     uncheckedImage {
      bIsDynamicallyLoaded=false
      drawAs=NoDrawType
      imageType=NoImage
      mirroring=NoMirror
      resourceName=None
      resourceObject=null
      tiling=NoTile
      imageSize=<0 0>
      margin=<ltrb 0 0 0 0>
      outlineSettings {
       bUseBrushTransparency=false
       roundingType=HalfHeightRadius
       width=0
       color {colorUseRule=UseColor_Specified, specifiedColor=<rgba 0 0 0 0>}
       cornerRadii=<0 0 0 0>
      }
      tintColor {colorUseRule=UseColor_Specified, specifiedColor=<rgba 1 1 1 1>}
      uVRegion {bIsValid=false, max=<0 0>, min=<0 0>}
     }
     uncheckedPressedImage {
      bIsDynamicallyLoaded=false
      drawAs=NoDrawType
      imageType=NoImage
      mirroring=NoMirror
      resourceName=None
      resourceObject=null
      tiling=NoTile
      imageSize=<0 0>
      margin=<ltrb 0 0 0 0>
      outlineSettings {
       bUseBrushTransparency=false
       roundingType=HalfHeightRadius
       width=0
       color {colorUseRule=UseColor_Specified, specifiedColor=<rgba 0 0 0 0>}
       cornerRadii=<0 0 0 0>
      }
      tintColor {colorUseRule=UseColor_Specified, specifiedColor=<rgba 1 1 1 1>}
      uVRegion {bIsValid=false, max=<0 0>, min=<0 0>}
     }
     uncheckedSlateSound {resourceObject=null}
     undeterminedForeground {colorUseRule=UseColor_Specified, specifiedColor=<rgba 1 1 1 1>}
     undeterminedHoveredImage {
      bIsDynamicallyLoaded=false
      drawAs=Image
      imageType=Vector
      mirroring=NoMirror
      resourceName=../../../Engine/Content/Slate/Starship/CoreWidgets/CheckBox/indeterminate.svg
      resourceObject=null
      tiling=NoTile
      imageSize=<16 16>
      margin=<ltrb 0 0 0 0>
      outlineSettings {
       bUseBrushTransparency=false
       roundingType=HalfHeightRadius
       width=0
       color {colorUseRule=UseColor_Specified, specifiedColor=<rgba 0 0 0 0>}
       cornerRadii=<0 0 0 0>
      }
      tintColor {colorUseRule=UseColor_Specified, specifiedColor=<rgba 0.0044 0.2384 1 1>}
      uVRegion {bIsValid=false, max=<0 0>, min=<0 0>}
     }
     undeterminedImage {
      bIsDynamicallyLoaded=false
      drawAs=Image
      imageType=Vector
      mirroring=NoMirror
      resourceName=../../../Engine/Content/Slate/Starship/CoreWidgets/CheckBox/indeterminate.svg
      resourceObject=null
      tiling=NoTile
      imageSize=<16 16>
      margin=<ltrb 0 0 0 0>
      outlineSettings {
       bUseBrushTransparency=false
       roundingType=HalfHeightRadius
       width=0
       color {colorUseRule=UseColor_Specified, specifiedColor=<rgba 0 0 0 0>}
       cornerRadii=<0 0 0 0>
      }
      tintColor {colorUseRule=UseColor_Specified, specifiedColor=<rgba 0 0.162 0.7454 1>}
      uVRegion {bIsValid=false, max=<0 0>, min=<0 0>}
     }
     undeterminedPressedImage {
      bIsDynamicallyLoaded=false
      drawAs=Image
      imageType=Vector
      mirroring=NoMirror
      resourceName=../../../Engine/Content/Slate/Starship/CoreWidgets/CheckBox/indeterminate.svg
      resourceObject=null
      tiling=NoTile
      imageSize=<16 16>
      margin=<ltrb 0 0 0 0>
      outlineSettings {
       bUseBrushTransparency=false
       roundingType=HalfHeightRadius
       width=0
       color {colorUseRule=UseColor_Specified, specifiedColor=<rgba 0 0 0 0>}
       cornerRadii=<0 0 0 0>
      }
      tintColor {colorUseRule=UseColor_Specified, specifiedColor=<rgba 0 0.0802 0.3515 1>}
      uVRegion {bIsValid=false, max=<0 0>, min=<0 0>}
     }
    }}
   slot {class=VerticalBoxSlot, properties {verticalAlignment=VAlign_Bottom, padding=<ltrb 0 5 0 0>}}
  }
  UEFN_TextBlock_C_23 {
   widgetClassPath=&UEFN_TextBlock
   parent=JunoRef
   widget=UEFN_TextBlock_C_23
   properties {
    justification=Center
    text=Juno Minifig
    toolTipText=Uses a Juno Minifig as a size reference
    font {
     bForceMonospaced=false
     bMaterialIsStencil=false
     fontMaterial=null
     fontObject=&BurbankSmall
     letterSpacing=0
     monospacedWidth=1
     size=9
     skewAmount=0
     typefaceFontName=Medium
     outlineSettings {
      bApplyOutlineToDropShadows=false
      bMiteredCorners=false
      bSeparateFillAlpha=false
      outlineMaterial=null
      outlineSize=0
      outlineColor=<rgba 0 0 0 1>
     }
    }
   }
   slot {class=PanelSlot}
  }
  VerticalBox_187 {
   widgetClassPath=&VerticalBox
   parent=HorizontalBox_195
   widget=VerticalBox_187
   slot {class=HorizontalBoxSlot, properties {verticalAlignment=VAlign_Top, padding=<ltrb 5 5 10 10>}}
  }
  renderSelected {
   widgetClassPath=&EditorUtilityButton
   parent=VerticalBox_187
   widget=renderSelected
   slot {class=VerticalBoxSlot, properties {padding=<ltrb 0 10 0 0>}}
  }
  UEFN_TextBlock_C {
   widgetClassPath=&UEFN_TextBlock
   parent=renderSelected
   widget=UEFN_TextBlock_C
   properties {
    text=Make Sequence
    toolTipText=Creates a Level Sequence using the selected Actor
    font {
     bForceMonospaced=false
     bMaterialIsStencil=false
     fontMaterial=null
     fontObject=&BurbankSmall
     letterSpacing=0
     monospacedWidth=1
     size=12.75
     skewAmount=0
     typefaceFontName=Black
     outlineSettings {
      bApplyOutlineToDropShadows=false
      bMiteredCorners=false
      bSeparateFillAlpha=false
      outlineMaterial=null
      outlineSize=0
      outlineColor=<rgba 0 0 0 1>
     }
    }
   }
   slot {class=ButtonSlot, properties {verticalAlignment=VAlign_Fill, padding=<ltrb 0 0 0 0>}}
  }
  renderLODs {
   widgetClassPath=&EditorUtilityButton
   parent=VerticalBox_187
   widget=renderLODs
   properties {toolTipText=WIP}
   slot {class=VerticalBoxSlot, properties {padding=<ltrb 0 10 0 0>}}
  }
  UEFN_TextBlock_C_19 {
   widgetClassPath=&UEFN_TextBlock
   parent=renderLODs
   widget=UEFN_TextBlock_C_19
   properties {
    text=Make LODs Sequence
    toolTipText=Creates a Level Sequence using the selected Actor switching LODs in order
    font {
     bForceMonospaced=false
     bMaterialIsStencil=false
     fontMaterial=null
     fontObject=&BurbankSmall
     letterSpacing=0
     monospacedWidth=1
     size=12.75
     skewAmount=0
     typefaceFontName=Black
     outlineSettings {
      bApplyOutlineToDropShadows=false
      bMiteredCorners=false
      bSeparateFillAlpha=false
      outlineMaterial=null
      outlineSize=0
      outlineColor=<rgba 0 0 0 1>
     }
    }
   }
   slot {class=ButtonSlot, properties {verticalAlignment=VAlign_Fill, padding=<ltrb 0 0 0 0>}}
  }
  renderBP {
   widgetClassPath=&EditorUtilityButton
   parent=VerticalBox_187
   widget=renderBP
   properties {toolTipText=WIP}
   slot {class=VerticalBoxSlot, properties {padding=<ltrb 0 10 0 0>}}
  }
  UEFN_TextBlock_C_20 {
   widgetClassPath=&UEFN_TextBlock
   parent=renderBP
   widget=UEFN_TextBlock_C_20
   properties {
    text=Make BP Sequence
    toolTipText=Creates a Level Sequence using the selected Actor switching Nanite and LODs in order
    font {
     bForceMonospaced=false
     bMaterialIsStencil=false
     fontMaterial=null
     fontObject=&BurbankSmall
     letterSpacing=0
     monospacedWidth=1
     size=12.75
     skewAmount=0
     typefaceFontName=Black
     outlineSettings {
      bApplyOutlineToDropShadows=false
      bMiteredCorners=false
      bSeparateFillAlpha=false
      outlineMaterial=null
      outlineSize=0
      outlineColor=<rgba 0 0 0 1>
     }
    }
   }
   slot {class=ButtonSlot, properties {verticalAlignment=VAlign_Fill, padding=<ltrb 0 0 0 0>}}
  }
  cutDuration {
   widgetClassPath=&UEFN_TextBlock
   bIsVariable=true
   parent=VerticalBox_187
   widget=cutDuration
   properties {
    justification=Center
    text=Cut Duration
    colorAndOpacity {colorUseRule=UseColor_Specified, specifiedColor=<rgba 0.3229 0.3229 0.3229 1>}
    font {
     bForceMonospaced=false
     bMaterialIsStencil=false
     fontMaterial=null
     fontObject=&BurbankSmall
     letterSpacing=0
     monospacedWidth=1
     size=8.25
     skewAmount=0
     typefaceFontName=Black
     outlineSettings {
      bApplyOutlineToDropShadows=false
      bMiteredCorners=false
      bSeparateFillAlpha=false
      outlineMaterial=null
      outlineSize=0
      outlineColor=<rgba 0 0 0 1>
     }
    }
   }
   slot {class=VerticalBoxSlot, properties {padding=<ltrb 0 5 0 0>}}
  }
  cutDurationSlider {
   widgetClassPath=/Script/Blutility.EditorUtilitySlider
   bIsVariable=true
   parent=VerticalBox_187
   widget=cutDurationSlider
   properties {maxValue=900, minValue=90, stepSize=10, value=90}
   slot {class=VerticalBoxSlot, properties {padding=<ltrb 0 5 0 0>}}
  }
  HorizontalBox_482 {
   widgetClassPath=&HorizontalBox
   parent=VerticalBox_187
   widget=HorizontalBox_482
   slot {class=VerticalBoxSlot, properties {padding=<ltrb 0 10 0 5>}}
  }
  secs_10 {
   widgetClassPath=&EditorUtilityButton
   parent=HorizontalBox_482
   widget=secs_10
   slot {class=HorizontalBoxSlot, properties {size {sizeRule=Fill, value=1}}}
  }
  UEFN_TextBlock_C_197 {
   widgetClassPath=&UEFN_TextBlock
   parent=secs_10
   widget=UEFN_TextBlock_C_197
   properties {
    text=10
    font {
     bForceMonospaced=false
     bMaterialIsStencil=false
     fontMaterial=null
     fontObject=&BurbankSmall
     letterSpacing=0
     monospacedWidth=1
     size=9
     skewAmount=0
     typefaceFontName=Black
     outlineSettings {
      bApplyOutlineToDropShadows=false
      bMiteredCorners=false
      bSeparateFillAlpha=false
      outlineMaterial=null
      outlineSize=0
      outlineColor=<rgba 0 0 0 1>
     }
    }
   }
   slot {class=ButtonSlot}
  }
  secs_20 {
   widgetClassPath=&EditorUtilityButton
   parent=HorizontalBox_482
   widget=secs_20
   slot {class=HorizontalBoxSlot, properties {size {sizeRule=Fill, value=1}}}
  }
  UEFN_TextBlock_C_4 {
   widgetClassPath=&UEFN_TextBlock
   parent=secs_20
   widget=UEFN_TextBlock_C_4
   properties {
    text=20
    font {
     bForceMonospaced=false
     bMaterialIsStencil=false
     fontMaterial=null
     fontObject=&BurbankSmall
     letterSpacing=0
     monospacedWidth=1
     size=9
     skewAmount=0
     typefaceFontName=Black
     outlineSettings {
      bApplyOutlineToDropShadows=false
      bMiteredCorners=false
      bSeparateFillAlpha=false
      outlineMaterial=null
      outlineSize=0
      outlineColor=<rgba 0 0 0 1>
     }
    }
   }
   slot {class=ButtonSlot}
  }
  secs_30 {
   widgetClassPath=&EditorUtilityButton
   parent=HorizontalBox_482
   widget=secs_30
   slot {class=HorizontalBoxSlot, properties {size {sizeRule=Fill, value=1}}}
  }
  UEFN_TextBlock_C_5 {
   widgetClassPath=&UEFN_TextBlock
   parent=secs_30
   widget=UEFN_TextBlock_C_5
   properties {
    text=30
    font {
     bForceMonospaced=false
     bMaterialIsStencil=false
     fontMaterial=null
     fontObject=&BurbankSmall
     letterSpacing=0
     monospacedWidth=1
     size=9
     skewAmount=0
     typefaceFontName=Black
     outlineSettings {
      bApplyOutlineToDropShadows=false
      bMiteredCorners=false
      bSeparateFillAlpha=false
      outlineMaterial=null
      outlineSize=0
      outlineColor=<rgba 0 0 0 1>
     }
    }
   }
   slot {class=ButtonSlot}
  }
  yOffsetText {
   widgetClassPath=&UEFN_TextBlock
   bIsVariable=true
   parent=VerticalBox_187
   widget=yOffsetText
   properties {
    justification=Center
    text=Offset: 0
    colorAndOpacity {colorUseRule=UseColor_Specified, specifiedColor=<rgba 0.3229 0.3229 0.3229 1>}
    font {
     bForceMonospaced=false
     bMaterialIsStencil=false
     fontMaterial=null
     fontObject=&BurbankSmall
     letterSpacing=0
     monospacedWidth=1
     size=8.25
     skewAmount=0
     typefaceFontName=Black
     outlineSettings {
      bApplyOutlineToDropShadows=false
      bMiteredCorners=false
      bSeparateFillAlpha=false
      outlineMaterial=null
      outlineSize=0
      outlineColor=<rgba 0 0 0 1>
     }
    }
   }
   slot {class=VerticalBoxSlot, properties {padding=<ltrb 0 5 0 0>}}
  }
  yOffsetSlider {
   widgetClassPath=&EditorUtilitySlider
   bIsVariable=true
   parent=VerticalBox_187
   widget=yOffsetSlider
   properties {maxValue=500, minValue=-500, stepSize=5, value=10}
   slot {class=VerticalBoxSlot, properties {padding=<ltrb 0 5 0 0>}}
  }
  UserComment {
   widgetClassPath=/Game/Valkyrie/UMG/UEFN_EditableText
   bIsVariable=true
   parent=VerticalBox_187
   widget=UserComment
   properties {
    bIsEnabled=false
    hintText=Enter User Comment
    visibility=Collapsed
    widgetStyle {
     backgroundImageComposing {
      bIsDynamicallyLoaded=false
      drawAs=Border
      imageType=FullColor
      mirroring=NoMirror
      resourceName=../../../Engine/Content/Slate/Old/HyperlinkDotted.png
      resourceObject=null
      tiling=Both
      imageSize=<0 0>
      margin=<ltrb 0 0 0 0.1875>
      outlineSettings {
       bUseBrushTransparency=false
       roundingType=HalfHeightRadius
       width=0
       color {colorUseRule=UseColor_Specified, specifiedColor=<rgba 0 0 0 0>}
       cornerRadii=<0 0 0 0>
      }
      tintColor {colorUseRule=UseColor_Specified, specifiedColor=<rgba 1 1 1 1>}
      uVRegion {bIsValid=false, max=<0 0>, min=<0 0>}
     }
     backgroundImageSelected {
      bIsDynamicallyLoaded=false
      drawAs=Image
      imageType=NoImage
      mirroring=NoMirror
      resourceName=None
      resourceObject=null
      tiling=NoTile
      imageSize=<0 0>
      margin=<ltrb 0 0 0 0>
      outlineSettings {
       bUseBrushTransparency=false
       roundingType=HalfHeightRadius
       width=0
       color {colorUseRule=UseColor_Specified, specifiedColor=<rgba 0 0 0 0>}
       cornerRadii=<0 0 0 0>
      }
      tintColor {colorUseRule=UseColor_Specified, specifiedColor=<rgba 0 0.162 0.7454 1>}
      uVRegion {bIsValid=false, max=<0 0>, min=<0 0>}
     }
     caretImage {
      bIsDynamicallyLoaded=false
      drawAs=Image
      imageType=NoImage
      mirroring=NoMirror
      resourceName=None
      resourceObject=null
      tiling=NoTile
      imageSize=<0 0>
      margin=<ltrb 0 0 0 0>
      outlineSettings {
       bUseBrushTransparency=false
       roundingType=HalfHeightRadius
       width=0
       color {colorUseRule=UseColor_Specified, specifiedColor=<rgba 0 0 0 0>}
       cornerRadii=<0 0 0 0>
      }
      tintColor {colorUseRule=UseColor_Specified, specifiedColor=<rgba 1 1 1 1>}
      uVRegion {bIsValid=false, max=<0 0>, min=<0 0>}
     }
     colorAndOpacity {colorUseRule=UseColor_Specified, specifiedColor=<rgba 1 1 1 1>}
     font {
      bForceMonospaced=false
      bMaterialIsStencil=false
      fontMaterial=null
      fontObject=&BurbankSmall
      letterSpacing=0
      monospacedWidth=1
      size=9
      skewAmount=0
      typefaceFontName=Black
      outlineSettings {
       bApplyOutlineToDropShadows=false
       bMiteredCorners=false
       bSeparateFillAlpha=false
       outlineMaterial=null
       outlineSize=0
       outlineColor=<rgba 0 0 0 1>
      }
     }
     hintColorAndOpacity {colorUseRule=UseColor_Foreground, specifiedColor=<rgba 1 0 1 1>}
     selectedBackgroundColor {colorUseRule=UseColor_Foreground, specifiedColor=<rgba 1 0 1 1>}
    }
   }
   slot {class=VerticalBoxSlot, properties {padding=<ltrb 5 10 5 5>}}
  }
  cleanupTempFiles {
   widgetClassPath=&EditorUtilityButton
   parent=VerticalBox_187
   widget=cleanupTempFiles
   slot {class=VerticalBoxSlot, properties {padding=<ltrb 0 10 0 0>}}
  }
  UEFN_TextBlock_C_1 {
   widgetClassPath=&UEFN_TextBlock
   parent=cleanupTempFiles
   widget=UEFN_TextBlock_C_1
   properties {
    text=Cleanup Temp
    colorAndOpacity {colorUseRule=UseColor_Specified, specifiedColor=<rgba 0.3073 0.3073 0.3073 1>}
    font {
     bForceMonospaced=false
     bMaterialIsStencil=false
     fontMaterial=null
     fontObject=&BurbankSmall
     letterSpacing=0
     monospacedWidth=1
     size=12.75
     skewAmount=0
     typefaceFontName=Black
     outlineSettings {
      bApplyOutlineToDropShadows=false
      bMiteredCorners=false
      bSeparateFillAlpha=false
      outlineMaterial=null
      outlineSize=0
      outlineColor=<rgba 0 0 0 1>
     }
    }
   }
   slot {class=ButtonSlot, properties {verticalAlignment=VAlign_Fill, padding=<ltrb 0 0 0 0>}}
  }
  placeScaleActor {
   widgetClassPath=&EditorUtilityButton
   parent=VerticalBox_187
   widget=placeScaleActor
   properties {bIsEnabled=false, visibility=Collapsed}
   slot {class=VerticalBoxSlot, properties {padding=<ltrb 0 10 0 0>}}
  }
  UEFN_TextBlock_C_2 {
   widgetClassPath=&UEFN_TextBlock
   parent=placeScaleActor
   widget=UEFN_TextBlock_C_2
   properties {
    text="(debug) Place Jonesy"
    colorAndOpacity {colorUseRule=UseColor_Specified, specifiedColor=<rgba 0.3073 0.3073 0.3073 1>}
    font {
     bForceMonospaced=false
     bMaterialIsStencil=false
     fontMaterial=null
     fontObject=&BurbankSmall
     letterSpacing=0
     monospacedWidth=1
     size=10.5
     skewAmount=0
     typefaceFontName=Black
     outlineSettings {
      bApplyOutlineToDropShadows=false
      bMiteredCorners=false
      bSeparateFillAlpha=false
      outlineMaterial=null
      outlineSize=0
      outlineColor=<rgba 0 0 0 1>
     }
    }
   }
   slot {class=ButtonSlot, properties {verticalAlignment=VAlign_Fill, padding=<ltrb 0 0 0 0>}}
  }
  disableNaniteButton {
   widgetClassPath=&EditorUtilityButton
   parent=VerticalBox_187
   widget=disableNaniteButton
   properties {bIsEnabled=false, visibility=Collapsed}
   slot {class=VerticalBoxSlot, properties {padding=<ltrb 0 10 0 0>}}
  }
  UEFN_TextBlock_C_3 {
   widgetClassPath=&UEFN_TextBlock
   parent=disableNaniteButton
   widget=UEFN_TextBlock_C_3
   properties {
    text="(debug) Disable Nanite"
    colorAndOpacity {colorUseRule=UseColor_Specified, specifiedColor=<rgba 0.3073 0.3073 0.3073 1>}
    font {
     bForceMonospaced=false
     bMaterialIsStencil=false
     fontMaterial=null
     fontObject=&BurbankSmall
     letterSpacing=0
     monospacedWidth=1
     size=10.5
     skewAmount=0
     typefaceFontName=Black
     outlineSettings {
      bApplyOutlineToDropShadows=false
      bMiteredCorners=false
      bSeparateFillAlpha=false
      outlineMaterial=null
      outlineSize=0
      outlineColor=<rgba 0 0 0 1>
     }
    }
   }
   slot {class=ButtonSlot, properties {verticalAlignment=VAlign_Fill, padding=<ltrb 0 0 0 0>}}
  }
  VerticalBox_131 {
   widgetClassPath=&VerticalBox
   parent=HorizontalBox_195
   widget=VerticalBox_131
   properties {visibility=Visible}
   slot {class=HorizontalBoxSlot, properties {padding=<ltrb 5 5 10 0>}}
  }
  VerticalBox_299 {
   widgetClassPath=&VerticalBox
   parent=VerticalBox_131
   widget=VerticalBox_299
   slot {class=VerticalBoxSlot, properties {padding=<ltrb 0 10 0 10>}}
  }
  rotateActors {
   widgetClassPath=&EditorUtilityCheckBox
   bIsVariable=true
   parent=VerticalBox_299
   widget=rotateActors
   properties {
    checkedState=Checked
    widgetStyle {
     checkBoxType=CheckBox
     backgroundHoveredImage {
      bIsDynamicallyLoaded=false
      drawAs=RoundedBox
      imageType=NoImage
      mirroring=NoMirror
      resourceName=None
      resourceObject=null
      tiling=NoTile
      imageSize=<18 18>
      margin=<ltrb 0 0 0 0>
      outlineSettings {
       bUseBrushTransparency=false
       roundingType=FixedRadius
       width=1
       color {colorUseRule=UseColor_ColorTable, specifiedColor=<rgba 0 0 0 0>}
       cornerRadii=<3 3 3 3>
      }
      tintColor {colorUseRule=UseColor_Specified, specifiedColor=<rgba 0.0048 0.0048 0.0048 1>}
      uVRegion {bIsValid=false, max=<0 0>, min=<0 0>}
     }
     backgroundImage {
      bIsDynamicallyLoaded=false
      drawAs=RoundedBox
      imageType=NoImage
      mirroring=NoMirror
      resourceName=None
      resourceObject=null
      tiling=NoTile
      imageSize=<18 18>
      margin=<ltrb 0 0 0 0>
      outlineSettings {
       bUseBrushTransparency=false
       roundingType=FixedRadius
       width=1
       color {colorUseRule=UseColor_ColorTable, specifiedColor=<rgba 0 0 0 0>}
       cornerRadii=<3 3 3 3>
      }
      tintColor {colorUseRule=UseColor_Specified, specifiedColor=<rgba 0.0048 0.0048 0.0048 1>}
      uVRegion {bIsValid=false, max=<0 0>, min=<0 0>}
     }
     backgroundPressedImage {
      bIsDynamicallyLoaded=false
      drawAs=RoundedBox
      imageType=NoImage
      mirroring=NoMirror
      resourceName=None
      resourceObject=null
      tiling=NoTile
      imageSize=<18 18>
      margin=<ltrb 0 0 0 0>
      outlineSettings {
       bUseBrushTransparency=false
       roundingType=FixedRadius
       width=1
       color {colorUseRule=UseColor_ColorTable, specifiedColor=<rgba 0 0 0 0>}
       cornerRadii=<3 3 3 3>
      }
      tintColor {colorUseRule=UseColor_Specified, specifiedColor=<rgba 0.0048 0.0048 0.0048 1>}
      uVRegion {bIsValid=false, max=<0 0>, min=<0 0>}
     }
     borderBackgroundColor {colorUseRule=UseColor_Specified, specifiedColor=<rgba 1 1 1 1>}
     checkedForeground {colorUseRule=UseColor_Specified, specifiedColor=<rgba 1 1 1 1>}
     checkedHoveredForeground {colorUseRule=UseColor_Specified, specifiedColor=<rgba 1 1 1 1>}
     checkedHoveredImage {
      bIsDynamicallyLoaded=false
      drawAs=Image
      imageType=Vector
      mirroring=NoMirror
      resourceName=../../../Engine/Content/Slate/Starship/CoreWidgets/CheckBox/check.svg
      resourceObject=null
      tiling=NoTile
      imageSize=<16 16>
      margin=<ltrb 0 0 0 0>
      outlineSettings {
       bUseBrushTransparency=false
       roundingType=HalfHeightRadius
       width=0
       color {colorUseRule=UseColor_Specified, specifiedColor=<rgba 0 0 0 0>}
       cornerRadii=<0 0 0 0>
      }
      tintColor {colorUseRule=UseColor_Specified, specifiedColor=<rgba 0.0044 0.2384 1 1>}
      uVRegion {bIsValid=false, max=<0 0>, min=<0 0>}
     }
     checkedImage {
      bIsDynamicallyLoaded=false
      drawAs=Image
      imageType=Vector
      mirroring=NoMirror
      resourceName=../../../Engine/Content/Slate/Starship/CoreWidgets/CheckBox/check.svg
      resourceObject=null
      tiling=NoTile
      imageSize=<16 16>
      margin=<ltrb 0 0 0 0>
      outlineSettings {
       bUseBrushTransparency=false
       roundingType=HalfHeightRadius
       width=0
       color {colorUseRule=UseColor_Specified, specifiedColor=<rgba 0 0 0 0>}
       cornerRadii=<0 0 0 0>
      }
      tintColor {colorUseRule=UseColor_Specified, specifiedColor=<rgba 0 0.162 0.7454 1>}
      uVRegion {bIsValid=false, max=<0 0>, min=<0 0>}
     }
     checkedPressedForeground {colorUseRule=UseColor_Specified, specifiedColor=<rgba 1 1 1 1>}
     checkedPressedImage {
      bIsDynamicallyLoaded=false
      drawAs=Image
      imageType=Vector
      mirroring=NoMirror
      resourceName=../../../Engine/Content/Slate/Starship/CoreWidgets/CheckBox/check.svg
      resourceObject=null
      tiling=NoTile
      imageSize=<16 16>
      margin=<ltrb 0 0 0 0>
      outlineSettings {
       bUseBrushTransparency=false
       roundingType=HalfHeightRadius
       width=0
       color {colorUseRule=UseColor_Specified, specifiedColor=<rgba 0 0 0 0>}
       cornerRadii=<0 0 0 0>
      }
      tintColor {colorUseRule=UseColor_Specified, specifiedColor=<rgba 0 0.0802 0.3515 1>}
      uVRegion {bIsValid=false, max=<0 0>, min=<0 0>}
     }
     checkedSlateSound {resourceObject=null}
     foregroundColor {colorUseRule=UseColor_Specified, specifiedColor=<rgba 1 1 1 1>}
     hoveredForeground {colorUseRule=UseColor_Specified, specifiedColor=<rgba 1 1 1 1>}
     hoveredSlateSound {resourceObject=null}
     padding=<ltrb 2 0 0 0>
     pressedForeground {colorUseRule=UseColor_Specified, specifiedColor=<rgba 1 1 1 1>}
     uncheckedHoveredImage {
      bIsDynamicallyLoaded=false
      drawAs=NoDrawType
      imageType=NoImage
      mirroring=NoMirror
      resourceName=None
      resourceObject=null
      tiling=NoTile
      imageSize=<0 0>
      margin=<ltrb 0 0 0 0>
      outlineSettings {
       bUseBrushTransparency=false
       roundingType=HalfHeightRadius
       width=0
       color {colorUseRule=UseColor_Specified, specifiedColor=<rgba 0 0 0 0>}
       cornerRadii=<0 0 0 0>
      }
      tintColor {colorUseRule=UseColor_Specified, specifiedColor=<rgba 1 1 1 1>}
      uVRegion {bIsValid=false, max=<0 0>, min=<0 0>}
     }
     uncheckedImage {
      bIsDynamicallyLoaded=false
      drawAs=NoDrawType
      imageType=NoImage
      mirroring=NoMirror
      resourceName=None
      resourceObject=null
      tiling=NoTile
      imageSize=<0 0>
      margin=<ltrb 0 0 0 0>
      outlineSettings {
       bUseBrushTransparency=false
       roundingType=HalfHeightRadius
       width=0
       color {colorUseRule=UseColor_Specified, specifiedColor=<rgba 0 0 0 0>}
       cornerRadii=<0 0 0 0>
      }
      tintColor {colorUseRule=UseColor_Specified, specifiedColor=<rgba 1 1 1 1>}
      uVRegion {bIsValid=false, max=<0 0>, min=<0 0>}
     }
     uncheckedPressedImage {
      bIsDynamicallyLoaded=false
      drawAs=NoDrawType
      imageType=NoImage
      mirroring=NoMirror
      resourceName=None
      resourceObject=null
      tiling=NoTile
      imageSize=<0 0>
      margin=<ltrb 0 0 0 0>
      outlineSettings {
       bUseBrushTransparency=false
       roundingType=HalfHeightRadius
       width=0
       color {colorUseRule=UseColor_Specified, specifiedColor=<rgba 0 0 0 0>}
       cornerRadii=<0 0 0 0>
      }
      tintColor {colorUseRule=UseColor_Specified, specifiedColor=<rgba 1 1 1 1>}
      uVRegion {bIsValid=false, max=<0 0>, min=<0 0>}
     }
     uncheckedSlateSound {resourceObject=null}
     undeterminedForeground {colorUseRule=UseColor_Specified, specifiedColor=<rgba 1 1 1 1>}
     undeterminedHoveredImage {
      bIsDynamicallyLoaded=false
      drawAs=Image
      imageType=Vector
      mirroring=NoMirror
      resourceName=../../../Engine/Content/Slate/Starship/CoreWidgets/CheckBox/indeterminate.svg
      resourceObject=null
      tiling=NoTile
      imageSize=<16 16>
      margin=<ltrb 0 0 0 0>
      outlineSettings {
       bUseBrushTransparency=false
       roundingType=HalfHeightRadius
       width=0
       color {colorUseRule=UseColor_Specified, specifiedColor=<rgba 0 0 0 0>}
       cornerRadii=<0 0 0 0>
      }
      tintColor {colorUseRule=UseColor_Specified, specifiedColor=<rgba 0.0044 0.2384 1 1>}
      uVRegion {bIsValid=false, max=<0 0>, min=<0 0>}
     }
     undeterminedImage {
      bIsDynamicallyLoaded=false
      drawAs=Image
      imageType=Vector
      mirroring=NoMirror
      resourceName=../../../Engine/Content/Slate/Starship/CoreWidgets/CheckBox/indeterminate.svg
      resourceObject=null
      tiling=NoTile
      imageSize=<16 16>
      margin=<ltrb 0 0 0 0>
      outlineSettings {
       bUseBrushTransparency=false
       roundingType=HalfHeightRadius
       width=0
       color {colorUseRule=UseColor_Specified, specifiedColor=<rgba 0 0 0 0>}
       cornerRadii=<0 0 0 0>
      }
      tintColor {colorUseRule=UseColor_Specified, specifiedColor=<rgba 0 0.162 0.7454 1>}
      uVRegion {bIsValid=false, max=<0 0>, min=<0 0>}
     }
     undeterminedPressedImage {
      bIsDynamicallyLoaded=false
      drawAs=Image
      imageType=Vector
      mirroring=NoMirror
      resourceName=../../../Engine/Content/Slate/Starship/CoreWidgets/CheckBox/indeterminate.svg
      resourceObject=null
      tiling=NoTile
      imageSize=<16 16>
      margin=<ltrb 0 0 0 0>
      outlineSettings {
       bUseBrushTransparency=false
       roundingType=HalfHeightRadius
       width=0
       color {colorUseRule=UseColor_Specified, specifiedColor=<rgba 0 0 0 0>}
       cornerRadii=<0 0 0 0>
      }
      tintColor {colorUseRule=UseColor_Specified, specifiedColor=<rgba 0 0.0802 0.3515 1>}
      uVRegion {bIsValid=false, max=<0 0>, min=<0 0>}
     }
    }
   }
   slot {class=VerticalBoxSlot, properties {size {sizeRule=Fill, value=1}}}
  }
  UEFN_TextBlock_C_376 {
   widgetClassPath=&UEFN_TextBlock
   parent=rotateActors
   widget=UEFN_TextBlock_C_376
   properties {
    justification=Center
    text=Turn Table
    font {
     bForceMonospaced=false
     bMaterialIsStencil=false
     fontMaterial=null
     fontObject=&BurbankSmall
     letterSpacing=0
     monospacedWidth=1
     size=10.5
     skewAmount=0
     typefaceFontName=Black
     outlineSettings {
      bApplyOutlineToDropShadows=false
      bMiteredCorners=false
      bSeparateFillAlpha=false
      outlineMaterial=null
      outlineSize=0
      outlineColor=<rgba 0 0 0 1>
     }
    }
   }
   slot {class=PanelSlot}
  }
  rotateCamera {
   widgetClassPath=&EditorUtilityCheckBox
   bIsVariable=true
   parent=VerticalBox_299
   widget=rotateCamera
   properties {
    bIsEnabled=false
    widgetStyle {
     checkBoxType=CheckBox
     backgroundHoveredImage {
      bIsDynamicallyLoaded=false
      drawAs=RoundedBox
      imageType=NoImage
      mirroring=NoMirror
      resourceName=None
      resourceObject=null
      tiling=NoTile
      imageSize=<18 18>
      margin=<ltrb 0 0 0 0>
      outlineSettings {
       bUseBrushTransparency=false
       roundingType=FixedRadius
       width=1
       color {colorUseRule=UseColor_ColorTable, specifiedColor=<rgba 0 0 0 0>}
       cornerRadii=<3 3 3 3>
      }
      tintColor {colorUseRule=UseColor_Specified, specifiedColor=<rgba 0.0048 0.0048 0.0048 1>}
      uVRegion {bIsValid=false, max=<0 0>, min=<0 0>}
     }
     backgroundImage {
      bIsDynamicallyLoaded=false
      drawAs=RoundedBox
      imageType=NoImage
      mirroring=NoMirror
      resourceName=None
      resourceObject=null
      tiling=NoTile
      imageSize=<18 18>
      margin=<ltrb 0 0 0 0>
      outlineSettings {
       bUseBrushTransparency=false
       roundingType=FixedRadius
       width=1
       color {colorUseRule=UseColor_ColorTable, specifiedColor=<rgba 0 0 0 0>}
       cornerRadii=<3 3 3 3>
      }
      tintColor {colorUseRule=UseColor_Specified, specifiedColor=<rgba 0.0048 0.0048 0.0048 1>}
      uVRegion {bIsValid=false, max=<0 0>, min=<0 0>}
     }
     backgroundPressedImage {
      bIsDynamicallyLoaded=false
      drawAs=RoundedBox
      imageType=NoImage
      mirroring=NoMirror
      resourceName=None
      resourceObject=null
      tiling=NoTile
      imageSize=<18 18>
      margin=<ltrb 0 0 0 0>
      outlineSettings {
       bUseBrushTransparency=false
       roundingType=FixedRadius
       width=1
       color {colorUseRule=UseColor_ColorTable, specifiedColor=<rgba 0 0 0 0>}
       cornerRadii=<3 3 3 3>
      }
      tintColor {colorUseRule=UseColor_Specified, specifiedColor=<rgba 0.0048 0.0048 0.0048 1>}
      uVRegion {bIsValid=false, max=<0 0>, min=<0 0>}
     }
     borderBackgroundColor {colorUseRule=UseColor_Specified, specifiedColor=<rgba 1 1 1 1>}
     checkedForeground {colorUseRule=UseColor_Specified, specifiedColor=<rgba 1 1 1 1>}
     checkedHoveredForeground {colorUseRule=UseColor_Specified, specifiedColor=<rgba 1 1 1 1>}
     checkedHoveredImage {
      bIsDynamicallyLoaded=false
      drawAs=Image
      imageType=Vector
      mirroring=NoMirror
      resourceName=../../../Engine/Content/Slate/Starship/CoreWidgets/CheckBox/check.svg
      resourceObject=null
      tiling=NoTile
      imageSize=<16 16>
      margin=<ltrb 0 0 0 0>
      outlineSettings {
       bUseBrushTransparency=false
       roundingType=HalfHeightRadius
       width=0
       color {colorUseRule=UseColor_Specified, specifiedColor=<rgba 0 0 0 0>}
       cornerRadii=<0 0 0 0>
      }
      tintColor {colorUseRule=UseColor_Specified, specifiedColor=<rgba 0.0044 0.2384 1 1>}
      uVRegion {bIsValid=false, max=<0 0>, min=<0 0>}
     }
     checkedImage {
      bIsDynamicallyLoaded=false
      drawAs=Image
      imageType=Vector
      mirroring=NoMirror
      resourceName=../../../Engine/Content/Slate/Starship/CoreWidgets/CheckBox/check.svg
      resourceObject=null
      tiling=NoTile
      imageSize=<16 16>
      margin=<ltrb 0 0 0 0>
      outlineSettings {
       bUseBrushTransparency=false
       roundingType=HalfHeightRadius
       width=0
       color {colorUseRule=UseColor_Specified, specifiedColor=<rgba 0 0 0 0>}
       cornerRadii=<0 0 0 0>
      }
      tintColor {colorUseRule=UseColor_Specified, specifiedColor=<rgba 0 0.162 0.7454 1>}
      uVRegion {bIsValid=false, max=<0 0>, min=<0 0>}
     }
     checkedPressedForeground {colorUseRule=UseColor_Specified, specifiedColor=<rgba 1 1 1 1>}
     checkedPressedImage {
      bIsDynamicallyLoaded=false
      drawAs=Image
      imageType=Vector
      mirroring=NoMirror
      resourceName=../../../Engine/Content/Slate/Starship/CoreWidgets/CheckBox/check.svg
      resourceObject=null
      tiling=NoTile
      imageSize=<16 16>
      margin=<ltrb 0 0 0 0>
      outlineSettings {
       bUseBrushTransparency=false
       roundingType=HalfHeightRadius
       width=0
       color {colorUseRule=UseColor_Specified, specifiedColor=<rgba 0 0 0 0>}
       cornerRadii=<0 0 0 0>
      }
      tintColor {colorUseRule=UseColor_Specified, specifiedColor=<rgba 0 0.0802 0.3515 1>}
      uVRegion {bIsValid=false, max=<0 0>, min=<0 0>}
     }
     checkedSlateSound {resourceObject=null}
     foregroundColor {colorUseRule=UseColor_Specified, specifiedColor=<rgba 1 1 1 1>}
     hoveredForeground {colorUseRule=UseColor_Specified, specifiedColor=<rgba 1 1 1 1>}
     hoveredSlateSound {resourceObject=null}
     padding=<ltrb 2 0 0 0>
     pressedForeground {colorUseRule=UseColor_Specified, specifiedColor=<rgba 1 1 1 1>}
     uncheckedHoveredImage {
      bIsDynamicallyLoaded=false
      drawAs=NoDrawType
      imageType=NoImage
      mirroring=NoMirror
      resourceName=None
      resourceObject=null
      tiling=NoTile
      imageSize=<0 0>
      margin=<ltrb 0 0 0 0>
      outlineSettings {
       bUseBrushTransparency=false
       roundingType=HalfHeightRadius
       width=0
       color {colorUseRule=UseColor_Specified, specifiedColor=<rgba 0 0 0 0>}
       cornerRadii=<0 0 0 0>
      }
      tintColor {colorUseRule=UseColor_Specified, specifiedColor=<rgba 1 1 1 1>}
      uVRegion {bIsValid=false, max=<0 0>, min=<0 0>}
     }
     uncheckedImage {
      bIsDynamicallyLoaded=false
      drawAs=NoDrawType
      imageType=NoImage
      mirroring=NoMirror
      resourceName=None
      resourceObject=null
      tiling=NoTile
      imageSize=<0 0>
      margin=<ltrb 0 0 0 0>
      outlineSettings {
       bUseBrushTransparency=false
       roundingType=HalfHeightRadius
       width=0
       color {colorUseRule=UseColor_Specified, specifiedColor=<rgba 0 0 0 0>}
       cornerRadii=<0 0 0 0>
      }
      tintColor {colorUseRule=UseColor_Specified, specifiedColor=<rgba 1 1 1 1>}
      uVRegion {bIsValid=false, max=<0 0>, min=<0 0>}
     }
     uncheckedPressedImage {
      bIsDynamicallyLoaded=false
      drawAs=NoDrawType
      imageType=NoImage
      mirroring=NoMirror
      resourceName=None
      resourceObject=null
      tiling=NoTile
      imageSize=<0 0>
      margin=<ltrb 0 0 0 0>
      outlineSettings {
       bUseBrushTransparency=false
       roundingType=HalfHeightRadius
       width=0
       color {colorUseRule=UseColor_Specified, specifiedColor=<rgba 0 0 0 0>}
       cornerRadii=<0 0 0 0>
      }
      tintColor {colorUseRule=UseColor_Specified, specifiedColor=<rgba 1 1 1 1>}
      uVRegion {bIsValid=false, max=<0 0>, min=<0 0>}
     }
     uncheckedSlateSound {resourceObject=null}
     undeterminedForeground {colorUseRule=UseColor_Specified, specifiedColor=<rgba 1 1 1 1>}
     undeterminedHoveredImage {
      bIsDynamicallyLoaded=false
      drawAs=Image
      imageType=Vector
      mirroring=NoMirror
      resourceName=../../../Engine/Content/Slate/Starship/CoreWidgets/CheckBox/indeterminate.svg
      resourceObject=null
      tiling=NoTile
      imageSize=<16 16>
      margin=<ltrb 0 0 0 0>
      outlineSettings {
       bUseBrushTransparency=false
       roundingType=HalfHeightRadius
       width=0
       color {colorUseRule=UseColor_Specified, specifiedColor=<rgba 0 0 0 0>}
       cornerRadii=<0 0 0 0>
      }
      tintColor {colorUseRule=UseColor_Specified, specifiedColor=<rgba 0.0044 0.2384 1 1>}
      uVRegion {bIsValid=false, max=<0 0>, min=<0 0>}
     }
     undeterminedImage {
      bIsDynamicallyLoaded=false
      drawAs=Image
      imageType=Vector
      mirroring=NoMirror
      resourceName=../../../Engine/Content/Slate/Starship/CoreWidgets/CheckBox/indeterminate.svg
      resourceObject=null
      tiling=NoTile
      imageSize=<16 16>
      margin=<ltrb 0 0 0 0>
      outlineSettings {
       bUseBrushTransparency=false
       roundingType=HalfHeightRadius
       width=0
       color {colorUseRule=UseColor_Specified, specifiedColor=<rgba 0 0 0 0>}
       cornerRadii=<0 0 0 0>
      }
      tintColor {colorUseRule=UseColor_Specified, specifiedColor=<rgba 0 0.162 0.7454 1>}
      uVRegion {bIsValid=false, max=<0 0>, min=<0 0>}
     }
     undeterminedPressedImage {
      bIsDynamicallyLoaded=false
      drawAs=Image
      imageType=Vector
      mirroring=NoMirror
      resourceName=../../../Engine/Content/Slate/Starship/CoreWidgets/CheckBox/indeterminate.svg
      resourceObject=null
      tiling=NoTile
      imageSize=<16 16>
      margin=<ltrb 0 0 0 0>
      outlineSettings {
       bUseBrushTransparency=false
       roundingType=HalfHeightRadius
       width=0
       color {colorUseRule=UseColor_Specified, specifiedColor=<rgba 0 0 0 0>}
       cornerRadii=<0 0 0 0>
      }
      tintColor {colorUseRule=UseColor_Specified, specifiedColor=<rgba 0 0.0802 0.3515 1>}
      uVRegion {bIsValid=false, max=<0 0>, min=<0 0>}
     }
    }
   }
   slot {class=VerticalBoxSlot, properties {padding=<ltrb 0 5 0 0>, size {sizeRule=Fill, value=1}}}
  }
  UEFN_TextBlock_C_15 {
   widgetClassPath=&UEFN_TextBlock
   parent=rotateCamera
   widget=UEFN_TextBlock_C_15
   properties {
    justification=Center
    text=Orbit Camera
    font {
     bForceMonospaced=false
     bMaterialIsStencil=false
     fontMaterial=null
     fontObject=&BurbankSmall
     letterSpacing=0
     monospacedWidth=1
     size=10.5
     skewAmount=0
     typefaceFontName=Black
     outlineSettings {
      bApplyOutlineToDropShadows=false
      bMiteredCorners=false
      bSeparateFillAlpha=false
      outlineMaterial=null
      outlineSize=0
      outlineColor=<rgba 0 0 0 1>
     }
    }
   }
   slot {class=PanelSlot}
  }
  turnTableOptions_grp {widgetClassPath=&VerticalBox, bIsVariable=true, parent=VerticalBox_299, widget=turnTableOptions_grp, slot {class=VerticalBoxSlot}}
  turnTableAngle_text {
   widgetClassPath=&UEFN_TextBlock
   bIsVariable=true
   parent=turnTableOptions_grp
   widget=turnTableAngle_text
   properties {
    text=Angle
    colorAndOpacity {colorUseRule=UseColor_Specified, specifiedColor=<rgba 0.3073 0.3073 0.3073 1>}
    font {
     bForceMonospaced=false
     bMaterialIsStencil=false
     fontMaterial=null
     fontObject=&BurbankSmall
     letterSpacing=0
     monospacedWidth=1
     size=8.25
     skewAmount=0
     typefaceFontName=Black
     outlineSettings {
      bApplyOutlineToDropShadows=false
      bMiteredCorners=false
      bSeparateFillAlpha=false
      outlineMaterial=null
      outlineSize=0
      outlineColor=<rgba 0 0 0 1>
     }
    }
   }
   slot {class=VerticalBoxSlot, properties {horizontalAlignment=HAlign_Center, padding=<ltrb 0 5 0 0>}}
  }
  turnAngle {
   widgetClassPath=&EditorUtilitySlider
   bIsVariable=true
   parent=turnTableOptions_grp
   widget=turnAngle
   properties {maxValue=360, minValue=90, stepSize=1, value=360}
   slot {class=VerticalBoxSlot, properties {padding=<ltrb 0 5 0 0>}}
  }
  HorizontalBox {
   widgetClassPath=&HorizontalBox
   parent=turnTableOptions_grp
   widget=HorizontalBox
   slot {class=VerticalBoxSlot, properties {padding=<ltrb 0 10 0 5>}}
  }
  angle90 {
   widgetClassPath=&EditorUtilityButton
   parent=HorizontalBox
   widget=angle90
   slot {class=HorizontalBoxSlot, properties {size {sizeRule=Fill, value=1}}}
  }
  UEFN_TextBlock_C_6 {
   widgetClassPath=&UEFN_TextBlock
   parent=angle90
   widget=UEFN_TextBlock_C_6
   properties {
    text=90
    font {
     bForceMonospaced=false
     bMaterialIsStencil=false
     fontMaterial=null
     fontObject=&BurbankSmall
     letterSpacing=0
     monospacedWidth=1
     size=9
     skewAmount=0
     typefaceFontName=Black
     outlineSettings {
      bApplyOutlineToDropShadows=false
      bMiteredCorners=false
      bSeparateFillAlpha=false
      outlineMaterial=null
      outlineSize=0
      outlineColor=<rgba 0 0 0 1>
     }
    }
   }
   slot {class=ButtonSlot}
  }
  angle180 {
   widgetClassPath=&EditorUtilityButton
   parent=HorizontalBox
   widget=angle180
   slot {class=HorizontalBoxSlot, properties {size {sizeRule=Fill, value=1}}}
  }
  UEFN_TextBlock_C_7 {
   widgetClassPath=&UEFN_TextBlock
   parent=angle180
   widget=UEFN_TextBlock_C_7
   properties {
    text=180
    font {
     bForceMonospaced=false
     bMaterialIsStencil=false
     fontMaterial=null
     fontObject=&BurbankSmall
     letterSpacing=0
     monospacedWidth=1
     size=9
     skewAmount=0
     typefaceFontName=Black
     outlineSettings {
      bApplyOutlineToDropShadows=false
      bMiteredCorners=false
      bSeparateFillAlpha=false
      outlineMaterial=null
      outlineSize=0
      outlineColor=<rgba 0 0 0 1>
     }
    }
   }
   slot {class=ButtonSlot}
  }
  angle270 {
   widgetClassPath=&EditorUtilityButton
   parent=HorizontalBox
   widget=angle270
   slot {class=HorizontalBoxSlot, properties {size {sizeRule=Fill, value=1}}}
  }
  UEFN_TextBlock_C_10 {
   widgetClassPath=&UEFN_TextBlock
   parent=angle270
   widget=UEFN_TextBlock_C_10
   properties {
    text=270
    font {
     bForceMonospaced=false
     bMaterialIsStencil=false
     fontMaterial=null
     fontObject=&BurbankSmall
     letterSpacing=0
     monospacedWidth=1
     size=9
     skewAmount=0
     typefaceFontName=Black
     outlineSettings {
      bApplyOutlineToDropShadows=false
      bMiteredCorners=false
      bSeparateFillAlpha=false
      outlineMaterial=null
      outlineSize=0
      outlineColor=<rgba 0 0 0 1>
     }
    }
   }
   slot {class=ButtonSlot}
  }
  angle360 {
   widgetClassPath=&EditorUtilityButton
   parent=HorizontalBox
   widget=angle360
   slot {class=HorizontalBoxSlot, properties {size {sizeRule=Fill, value=1}}}
  }
  UEFN_TextBlock_C_8 {
   widgetClassPath=&UEFN_TextBlock
   parent=angle360
   widget=UEFN_TextBlock_C_8
   properties {
    text=360
    font {
     bForceMonospaced=false
     bMaterialIsStencil=false
     fontMaterial=null
     fontObject=&BurbankSmall
     letterSpacing=0
     monospacedWidth=1
     size=9
     skewAmount=0
     typefaceFontName=Black
     outlineSettings {
      bApplyOutlineToDropShadows=false
      bMiteredCorners=false
      bSeparateFillAlpha=false
      outlineMaterial=null
      outlineSize=0
      outlineColor=<rgba 0 0 0 1>
     }
    }
   }
   slot {class=ButtonSlot}
  }
  VerticalBox_113 {
   widgetClassPath=&VerticalBox
   parent=VerticalBox_131
   widget=VerticalBox_113
   properties {toolTipText=Experimental}
   slot {class=VerticalBoxSlot}
  }
  camFOV_text {
   widgetClassPath=&UEFN_TextBlock
   bIsVariable=true
   parent=VerticalBox_113
   widget=camFOV_text
   properties {
    justification=Center
    text=Camera FOV: 75
    colorAndOpacity {colorUseRule=UseColor_Specified, specifiedColor=<rgba 0.3073 0.3073 0.3073 1>}
    font {
     bForceMonospaced=false
     bMaterialIsStencil=false
     fontMaterial=null
     fontObject=&BurbankSmall
     letterSpacing=0
     monospacedWidth=1
     size=8.25
     skewAmount=0
     typefaceFontName=Black
     outlineSettings {
      bApplyOutlineToDropShadows=false
      bMiteredCorners=false
      bSeparateFillAlpha=false
      outlineMaterial=null
      outlineSize=0
      outlineColor=<rgba 0 0 0 1>
     }
    }
   }
   slot {class=VerticalBoxSlot, properties {padding=<ltrb 0 5 0 0>}}
  }
  camFOV_slider {
   widgetClassPath=&EditorUtilitySlider
   bIsVariable=true
   parent=VerticalBox_113
   widget=camFOV_slider
   properties {maxValue=160, minValue=10, stepSize=1, value=75}
   slot {class=VerticalBoxSlot, properties {padding=<ltrb 0 5 0 0>}}
  }
  HorizontalBox_1 {
   widgetClassPath=&HorizontalBox
   parent=VerticalBox_113
   widget=HorizontalBox_1
   slot {class=VerticalBoxSlot, properties {padding=<ltrb 0 10 0 5>}}
  }
  fov35 {
   widgetClassPath=&EditorUtilityButton
   parent=HorizontalBox_1
   widget=fov35
   slot {class=HorizontalBoxSlot, properties {size {sizeRule=Fill, value=1}}}
  }
  UEFN_TextBlock_C_11 {
   widgetClassPath=&UEFN_TextBlock
   parent=fov35
   widget=UEFN_TextBlock_C_11
   properties {
    text=35
    font {
     bForceMonospaced=false
     bMaterialIsStencil=false
     fontMaterial=null
     fontObject=&BurbankSmall
     letterSpacing=0
     monospacedWidth=1
     size=9
     skewAmount=0
     typefaceFontName=Black
     outlineSettings {
      bApplyOutlineToDropShadows=false
      bMiteredCorners=false
      bSeparateFillAlpha=false
      outlineMaterial=null
      outlineSize=0
      outlineColor=<rgba 0 0 0 1>
     }
    }
   }
   slot {class=ButtonSlot}
  }
  fov75 {
   widgetClassPath=&EditorUtilityButton
   parent=HorizontalBox_1
   widget=fov75
   slot {class=HorizontalBoxSlot, properties {size {sizeRule=Fill, value=1}}}
  }
  UEFN_TextBlock_C_12 {
   widgetClassPath=&UEFN_TextBlock
   parent=fov75
   widget=UEFN_TextBlock_C_12
   properties {
    text=75
    font {
     bForceMonospaced=false
     bMaterialIsStencil=false
     fontMaterial=null
     fontObject=&BurbankSmall
     letterSpacing=0
     monospacedWidth=1
     size=9
     skewAmount=0
     typefaceFontName=Black
     outlineSettings {
      bApplyOutlineToDropShadows=false
      bMiteredCorners=false
      bSeparateFillAlpha=false
      outlineMaterial=null
      outlineSize=0
      outlineColor=<rgba 0 0 0 1>
     }
    }
   }
   slot {class=ButtonSlot}
  }
  fov90 {
   widgetClassPath=&EditorUtilityButton
   parent=HorizontalBox_1
   widget=fov90
   slot {class=HorizontalBoxSlot, properties {size {sizeRule=Fill, value=1}}}
  }
  UEFN_TextBlock_C_13 {
   widgetClassPath=&UEFN_TextBlock
   parent=fov90
   widget=UEFN_TextBlock_C_13
   properties {
    text=90
    font {
     bForceMonospaced=false
     bMaterialIsStencil=false
     fontMaterial=null
     fontObject=&BurbankSmall
     letterSpacing=0
     monospacedWidth=1
     size=9
     skewAmount=0
     typefaceFontName=Black
     outlineSettings {
      bApplyOutlineToDropShadows=false
      bMiteredCorners=false
      bSeparateFillAlpha=false
      outlineMaterial=null
      outlineSize=0
      outlineColor=<rgba 0 0 0 1>
     }
    }
   }
   slot {class=ButtonSlot}
  }
  fov120 {
   widgetClassPath=&EditorUtilityButton
   parent=HorizontalBox_1
   widget=fov120
   slot {class=HorizontalBoxSlot, properties {size {sizeRule=Fill, value=1}}}
  }
  UEFN_TextBlock_C_14 {
   widgetClassPath=&UEFN_TextBlock
   parent=fov120
   widget=UEFN_TextBlock_C_14
   properties {
    text=120
    font {
     bForceMonospaced=false
     bMaterialIsStencil=false
     fontMaterial=null
     fontObject=&BurbankSmall
     letterSpacing=0
     monospacedWidth=1
     size=9
     skewAmount=0
     typefaceFontName=Black
     outlineSettings {
      bApplyOutlineToDropShadows=false
      bMiteredCorners=false
      bSeparateFillAlpha=false
      outlineMaterial=null
      outlineSize=0
      outlineColor=<rgba 0 0 0 1>
     }
    }
   }
   slot {class=ButtonSlot}
  }
  VerticalBox_165 {
   widgetClassPath=&VerticalBox
   parent=VerticalBox_131
   widget=VerticalBox_165
   slot {class=VerticalBoxSlot, properties {size {sizeRule=Fill, value=1}}}
  }
  MLineComments {
   widgetClassPath=/Script/Blutility.EditorUtilityMultiLineEditableTextBox
   bIsVariable=true
   parent=VerticalBox_165
   widget=MLineComments
   properties {
    hintText=Enter User Comments
    widgetStyle {
     backgroundColor {colorUseRule=UseColor_Specified, specifiedColor=<rgba 1 1 1 1>}
     backgroundImageFocused {
      bIsDynamicallyLoaded=false
      drawAs=RoundedBox
      imageType=NoImage
      mirroring=NoMirror
      resourceName=None
      resourceObject=null
      tiling=NoTile
      imageSize=<0 0>
      margin=<ltrb 0 0 0 0>
      outlineSettings {
       bUseBrushTransparency=false
       roundingType=FixedRadius
       width=1
       color {colorUseRule=UseColor_ColorTable, specifiedColor=<rgba 0 0 0 0>}
       cornerRadii=<4 4 4 4>
      }
      tintColor {colorUseRule=UseColor_Specified, specifiedColor=<rgba 0.0048 0.0048 0.0048 1>}
      uVRegion {bIsValid=false, max=<0 0>, min=<0 0>}
     }
     backgroundImageHovered {
      bIsDynamicallyLoaded=false
      drawAs=RoundedBox
      imageType=NoImage
      mirroring=NoMirror
      resourceName=None
      resourceObject=null
      tiling=NoTile
      imageSize=<0 0>
      margin=<ltrb 0 0 0 0>
      outlineSettings {
       bUseBrushTransparency=false
       roundingType=FixedRadius
       width=1
       color {colorUseRule=UseColor_ColorTable, specifiedColor=<rgba 0 0 0 0>}
       cornerRadii=<4 4 4 4>
      }
      tintColor {colorUseRule=UseColor_Specified, specifiedColor=<rgba 0.0048 0.0048 0.0048 1>}
      uVRegion {bIsValid=false, max=<0 0>, min=<0 0>}
     }
     backgroundImageNormal {
      bIsDynamicallyLoaded=false
      drawAs=RoundedBox
      imageType=NoImage
      mirroring=NoMirror
      resourceName=None
      resourceObject=null
      tiling=NoTile
      imageSize=<0 0>
      margin=<ltrb 0 0 0 0>
      outlineSettings {
       bUseBrushTransparency=false
       roundingType=FixedRadius
       width=1
       color {colorUseRule=UseColor_ColorTable, specifiedColor=<rgba 0 0 0 0>}
       cornerRadii=<4 4 4 4>
      }
      tintColor {colorUseRule=UseColor_Specified, specifiedColor=<rgba 0.0048 0.0048 0.0048 1>}
      uVRegion {bIsValid=false, max=<0 0>, min=<0 0>}
     }
     backgroundImageReadOnly {
      bIsDynamicallyLoaded=false
      drawAs=RoundedBox
      imageType=NoImage
      mirroring=NoMirror
      resourceName=None
      resourceObject=null
      tiling=NoTile
      imageSize=<0 0>
      margin=<ltrb 0 0 0 0>
      outlineSettings {
       bUseBrushTransparency=false
       roundingType=FixedRadius
       width=1
       color {colorUseRule=UseColor_ColorTable, specifiedColor=<rgba 0 0 0 0>}
       cornerRadii=<4 4 4 4>
      }
      tintColor {colorUseRule=UseColor_Specified, specifiedColor=<rgba 0.0284 0.0284 0.0284 1>}
      uVRegion {bIsValid=false, max=<0 0>, min=<0 0>}
     }
     focusedForegroundColor {colorUseRule=UseColor_Specified, specifiedColor=<rgba 1 1 1 1>}
     foregroundColor {colorUseRule=UseColor_Specified, specifiedColor=<rgba 0.5271 0.5271 0.5271 1>}
     hScrollBarPadding=<ltrb 0 0 0 0>
     padding=<ltrb 8 3 8 4>
     readOnlyForegroundColor {colorUseRule=UseColor_Foreground, specifiedColor=<rgba 1 0 1 1>}
     scrollBarStyle {
      thickness=8
      draggedThumbImage {
       bIsDynamicallyLoaded=false
       drawAs=RoundedBox
       imageType=NoImage
       mirroring=NoMirror
       resourceName=None
       resourceObject=null
       tiling=NoTile
       imageSize=<0 0>
       margin=<ltrb 0 0 0 0>
       outlineSettings {
        bUseBrushTransparency=false
        roundingType=FixedRadius
        width=0
        color {colorUseRule=UseColor_Specified, specifiedColor=<rgba 0 0 0 0>}
        cornerRadii=<4 4 4 4>
       }
       tintColor {colorUseRule=UseColor_Specified, specifiedColor=<rgba 0.2159 0.2159 0.2159 1>}
       uVRegion {bIsValid=false, max=<0 0>, min=<0 0>}
      }
      horizontalBackgroundImage {
       bIsDynamicallyLoaded=false
       drawAs=NoDrawType
       imageType=NoImage
       mirroring=NoMirror
       resourceName=None
       resourceObject=null
       tiling=NoTile
       imageSize=<0 0>
       margin=<ltrb 0 0 0 0>
       outlineSettings {
        bUseBrushTransparency=false
        roundingType=HalfHeightRadius
        width=0
        color {colorUseRule=UseColor_Specified, specifiedColor=<rgba 0 0 0 0>}
        cornerRadii=<0 0 0 0>
       }
       tintColor {colorUseRule=UseColor_Specified, specifiedColor=<rgba 1 1 1 1>}
       uVRegion {bIsValid=false, max=<0 0>, min=<0 0>}
      }
      horizontalBottomSlotImage {
       bIsDynamicallyLoaded=false
       drawAs=NoDrawType
       imageType=NoImage
       mirroring=NoMirror
       resourceName=None
       resourceObject=null
       tiling=NoTile
       imageSize=<0 0>
       margin=<ltrb 0 0 0 0>
       outlineSettings {
        bUseBrushTransparency=false
        roundingType=HalfHeightRadius
        width=0
        color {colorUseRule=UseColor_Specified, specifiedColor=<rgba 0 0 0 0>}
        cornerRadii=<0 0 0 0>
       }
       tintColor {colorUseRule=UseColor_Specified, specifiedColor=<rgba 1 1 1 1>}
       uVRegion {bIsValid=false, max=<0 0>, min=<0 0>}
      }
      horizontalTopSlotImage {
       bIsDynamicallyLoaded=false
       drawAs=NoDrawType
       imageType=NoImage
       mirroring=NoMirror
       resourceName=None
       resourceObject=null
       tiling=NoTile
       imageSize=<0 0>
       margin=<ltrb 0 0 0 0>
       outlineSettings {
        bUseBrushTransparency=false
        roundingType=HalfHeightRadius
        width=0
        color {colorUseRule=UseColor_Specified, specifiedColor=<rgba 0 0 0 0>}
        cornerRadii=<0 0 0 0>
       }
       tintColor {colorUseRule=UseColor_Specified, specifiedColor=<rgba 1 1 1 1>}
       uVRegion {bIsValid=false, max=<0 0>, min=<0 0>}
      }
      hoveredThumbImage {
       bIsDynamicallyLoaded=false
       drawAs=RoundedBox
       imageType=NoImage
       mirroring=NoMirror
       resourceName=None
       resourceObject=null
       tiling=NoTile
       imageSize=<0 0>
       margin=<ltrb 0 0 0 0>
       outlineSettings {
        bUseBrushTransparency=false
        roundingType=FixedRadius
        width=0
        color {colorUseRule=UseColor_Specified, specifiedColor=<rgba 0 0 0 0>}
        cornerRadii=<4 4 4 4>
       }
       tintColor {colorUseRule=UseColor_Specified, specifiedColor=<rgba 0.2159 0.2159 0.2159 1>}
       uVRegion {bIsValid=false, max=<0 0>, min=<0 0>}
      }
      normalThumbImage {
       bIsDynamicallyLoaded=false
       drawAs=RoundedBox
       imageType=NoImage
       mirroring=NoMirror
       resourceName=None
       resourceObject=null
       tiling=NoTile
       imageSize=<0 0>
       margin=<ltrb 0 0 0 0>
       outlineSettings {
        bUseBrushTransparency=false
        roundingType=FixedRadius
        width=0
        color {colorUseRule=UseColor_Specified, specifiedColor=<rgba 0 0 0 0>}
        cornerRadii=<4 4 4 4>
       }
       tintColor {colorUseRule=UseColor_Specified, specifiedColor=<rgba 0.0953 0.0953 0.0953 1>}
       uVRegion {bIsValid=false, max=<0 0>, min=<0 0>}
      }
      verticalBackgroundImage {
       bIsDynamicallyLoaded=false
       drawAs=NoDrawType
       imageType=NoImage
       mirroring=NoMirror
       resourceName=None
       resourceObject=null
       tiling=NoTile
       imageSize=<0 0>
       margin=<ltrb 0 0 0 0>
       outlineSettings {
        bUseBrushTransparency=false
        roundingType=HalfHeightRadius
        width=0
        color {colorUseRule=UseColor_Specified, specifiedColor=<rgba 0 0 0 0>}
        cornerRadii=<0 0 0 0>
       }
       tintColor {colorUseRule=UseColor_Specified, specifiedColor=<rgba 1 1 1 1>}
       uVRegion {bIsValid=false, max=<0 0>, min=<0 0>}
      }
      verticalBottomSlotImage {
       bIsDynamicallyLoaded=false
       drawAs=NoDrawType
       imageType=NoImage
       mirroring=NoMirror
       resourceName=None
       resourceObject=null
       tiling=NoTile
       imageSize=<0 0>
       margin=<ltrb 0 0 0 0>
       outlineSettings {
        bUseBrushTransparency=false
        roundingType=HalfHeightRadius
        width=0
        color {colorUseRule=UseColor_Specified, specifiedColor=<rgba 0 0 0 0>}
        cornerRadii=<0 0 0 0>
       }
       tintColor {colorUseRule=UseColor_Specified, specifiedColor=<rgba 1 1 1 1>}
       uVRegion {bIsValid=false, max=<0 0>, min=<0 0>}
      }
      verticalTopSlotImage {
       bIsDynamicallyLoaded=false
       drawAs=NoDrawType
       imageType=NoImage
       mirroring=NoMirror
       resourceName=None
       resourceObject=null
       tiling=NoTile
       imageSize=<0 0>
       margin=<ltrb 0 0 0 0>
       outlineSettings {
        bUseBrushTransparency=false
        roundingType=HalfHeightRadius
        width=0
        color {colorUseRule=UseColor_Specified, specifiedColor=<rgba 0 0 0 0>}
        cornerRadii=<0 0 0 0>
       }
       tintColor {colorUseRule=UseColor_Specified, specifiedColor=<rgba 1 1 1 1>}
       uVRegion {bIsValid=false, max=<0 0>, min=<0 0>}
      }
     }
     textStyle {
      overflowPolicy=Clip
      transformPolicy=None
      colorAndOpacity {colorUseRule=UseColor_Specified, specifiedColor=<rgba 1 1 1 1>}
      font {
       bForceMonospaced=false
       bMaterialIsStencil=false
       fontMaterial=null
       fontObject=&BurbankSmall
       letterSpacing=0
       monospacedWidth=1
       size=8.25
       skewAmount=0
       typefaceFontName=Bold
       outlineSettings {
        bApplyOutlineToDropShadows=false
        bMiteredCorners=false
        bSeparateFillAlpha=false
        outlineMaterial=null
        outlineSize=0
        outlineColor=<rgba 0 0 0 1>
       }
      }
      highlightColor {colorUseRule=UseColor_ColorTable, specifiedColor=<rgba 0 0 0 0>}
      highlightShape {
       bIsDynamicallyLoaded=false
       drawAs=Image
       imageType=NoImage
       mirroring=NoMirror
       resourceName=None
       resourceObject=null
       tiling=NoTile
       imageSize=<0 0>
       margin=<ltrb 0 0 0 0>
       outlineSettings {
        bUseBrushTransparency=false
        roundingType=HalfHeightRadius
        width=0
        color {colorUseRule=UseColor_Specified, specifiedColor=<rgba 0 0 0 0>}
        cornerRadii=<0 0 0 0>
       }
       tintColor {colorUseRule=UseColor_ColorTable, specifiedColor=<rgba 0 0 0 0>}
       uVRegion {bIsValid=false, max=<0 0>, min=<0 0>}
      }
      selectedBackgroundColor {colorUseRule=UseColor_ColorTable, specifiedColor=<rgba 0 0 0 0>}
      shadowColorAndOpacity=<rgba 0 0 0 1>
      shadowOffset=<0 0>
      strikeBrush {
       bIsDynamicallyLoaded=false
       drawAs=Image
       imageType=NoImage
       mirroring=NoMirror
       resourceName=None
       resourceObject=null
       tiling=NoTile
       imageSize=<32 32>
       margin=<ltrb 0 0 0 0>
       outlineSettings {
        bUseBrushTransparency=false
        roundingType=HalfHeightRadius
        width=0
        color {colorUseRule=UseColor_Specified, specifiedColor=<rgba 0 0 0 0>}
        cornerRadii=<0 0 0 0>
       }
       tintColor {colorUseRule=UseColor_Specified, specifiedColor=<rgba 1 1 1 1>}
       uVRegion {bIsValid=false, max=<0 0>, min=<0 0>}
      }
      underlineBrush {
       bIsDynamicallyLoaded=false
       drawAs=Image
       imageType=NoImage
       mirroring=NoMirror
       resourceName=None
       resourceObject=null
       tiling=NoTile
       imageSize=<32 32>
       margin=<ltrb 0 0 0 0>
       outlineSettings {
        bUseBrushTransparency=false
        roundingType=HalfHeightRadius
        width=0
        color {colorUseRule=UseColor_Specified, specifiedColor=<rgba 0 0 0 0>}
        cornerRadii=<0 0 0 0>
       }
       tintColor {colorUseRule=UseColor_Specified, specifiedColor=<rgba 1 1 1 1>}
       uVRegion {bIsValid=false, max=<0 0>, min=<0 0>}
      }
     }
     vScrollBarPadding=<ltrb 0 0 0 0>
    }
   }
   slot {class=VerticalBoxSlot, properties {padding=<ltrb 0 5 0 0>, size {sizeRule=Fill, value=1}}}
  }
  VerticalBox_124 {
   widgetClassPath=&VerticalBox
   parent=HorizontalBox_195
   widget=VerticalBox_124
   slot {class=HorizontalBoxSlot, properties {padding=<ltrb 5 10 10 0>}}
  }
  TODSliderBox {widgetClassPath=&VerticalBox, bIsVariable=true, parent=VerticalBox_124, widget=TODSliderBox, slot {class=VerticalBoxSlot}}
  CurrentTODText {
   widgetClassPath=&UEFN_TextBlock
   bIsVariable=true
   parent=TODSliderBox
   widget=CurrentTODText
   properties {
    text=Time of Day
    font {
     bForceMonospaced=false
     bMaterialIsStencil=false
     fontMaterial=null
     fontObject=&BurbankSmall
     letterSpacing=0
     monospacedWidth=1
     size=10.5
     skewAmount=0
     typefaceFontName=Black
     outlineSettings {
      bApplyOutlineToDropShadows=false
      bMiteredCorners=false
      bSeparateFillAlpha=false
      outlineMaterial=null
      outlineSize=0
      outlineColor=<rgba 0 0 0 1>
     }
    }
   }
   slot {class=VerticalBoxSlot, properties {horizontalAlignment=HAlign_Center, padding=<ltrb 0 0 0 5>}}
  }
  TODSlider {
   widgetClassPath=/Script/AdvancedWidgets.RadialSlider
   bIsVariable=true
   parent=TODSliderBox
   widget=TODSlider
   properties {
    customDefaultValue=0.5
    showSliderHand=true
    value=0.5
    widgetStyle {
     barThickness=8
     disabledBarImage {
      bIsDynamicallyLoaded=false
      drawAs=RoundedBox
      imageType=NoImage
      mirroring=NoMirror
      resourceName=None
      resourceObject=null
      tiling=NoTile
      imageSize=<0 0>
      margin=<ltrb 0 0 0 0>
      outlineSettings {
       bUseBrushTransparency=false
       roundingType=FixedRadius
       width=1
       color {colorUseRule=UseColor_ColorTable, specifiedColor=<rgba 0 0 0 0>}
       cornerRadii=<2 2 2 2>
      }
      tintColor {colorUseRule=UseColor_Specified, specifiedColor=<rgba 0.0048 0.0048 0.0048 1>}
      uVRegion {bIsValid=false, max=<0 0>, min=<0 0>}
     }
     disabledThumbImage {
      bIsDynamicallyLoaded=false
      drawAs=RoundedBox
      imageType=NoImage
      mirroring=NoMirror
      resourceName=None
      resourceObject=null
      tiling=NoTile
      imageSize=<8 8>
      margin=<ltrb 0 0 0 0>
      outlineSettings {
       bUseBrushTransparency=false
       roundingType=HalfHeightRadius
       width=0
       color {colorUseRule=UseColor_Specified, specifiedColor=<rgba 0 0 0 0>}
       cornerRadii=<0 0 0 0>
      }
      tintColor {colorUseRule=UseColor_Specified, specifiedColor=<rgba 0.5271 0.5271 0.5271 1>}
      uVRegion {bIsValid=false, max=<0 0>, min=<0 0>}
     }
     hoveredBarImage {
      bIsDynamicallyLoaded=false
      drawAs=RoundedBox
      imageType=NoImage
      mirroring=NoMirror
      resourceName=None
      resourceObject=null
      tiling=NoTile
      imageSize=<0 0>
      margin=<ltrb 0 0 0 0>
      outlineSettings {
       bUseBrushTransparency=false
       roundingType=FixedRadius
       width=1
       color {colorUseRule=UseColor_ColorTable, specifiedColor=<rgba 0 0 0 0>}
       cornerRadii=<2 2 2 2>
      }
      tintColor {colorUseRule=UseColor_Specified, specifiedColor=<rgba 0.0048 0.0048 0.0048 1>}
      uVRegion {bIsValid=false, max=<0 0>, min=<0 0>}
     }
     hoveredThumbImage {
      bIsDynamicallyLoaded=false
      drawAs=RoundedBox
      imageType=NoImage
      mirroring=NoMirror
      resourceName=None
      resourceObject=null
      tiling=NoTile
      imageSize=<8 8>
      margin=<ltrb 0 0 0 0>
      outlineSettings {
       bUseBrushTransparency=false
       roundingType=HalfHeightRadius
       width=0
       color {colorUseRule=UseColor_Specified, specifiedColor=<rgba 0 0 0 0>}
       cornerRadii=<0 0 0 0>
      }
      tintColor {colorUseRule=UseColor_Specified, specifiedColor=<rgba 1 1 1 1>}
      uVRegion {bIsValid=false, max=<0 0>, min=<0 0>}
     }
     normalBarImage {
      bIsDynamicallyLoaded=false
      drawAs=RoundedBox
      imageType=NoImage
      mirroring=NoMirror
      resourceName=None
      resourceObject=null
      tiling=NoTile
      imageSize=<0 0>
      margin=<ltrb 0 0 0 0>
      outlineSettings {
       bUseBrushTransparency=false
       roundingType=FixedRadius
       width=1
       color {colorUseRule=UseColor_ColorTable, specifiedColor=<rgba 0 0 0 0>}
       cornerRadii=<2 2 2 2>
      }
      tintColor {colorUseRule=UseColor_Specified, specifiedColor=<rgba 0.0048 0.0048 0.0048 1>}
      uVRegion {bIsValid=false, max=<0 0>, min=<0 0>}
     }
     normalThumbImage {
      bIsDynamicallyLoaded=false
      drawAs=RoundedBox
      imageType=NoImage
      mirroring=NoMirror
      resourceName=None
      resourceObject=null
      tiling=NoTile
      imageSize=<8 8>
      margin=<ltrb 0 0 0 0>
      outlineSettings {
       bUseBrushTransparency=false
       roundingType=HalfHeightRadius
       width=0
       color {colorUseRule=UseColor_Specified, specifiedColor=<rgba 0 0 0 0>}
       cornerRadii=<0 0 0 0>
      }
      tintColor {colorUseRule=UseColor_Specified, specifiedColor=<rgba 0.5271 0.5271 0.5271 1>}
      uVRegion {bIsValid=false, max=<0 0>, min=<0 0>}
     }
    }
   }
   slot {class=VerticalBoxSlot, properties {horizontalAlignment=HAlign_Center, verticalAlignment=VAlign_Top}}
  }
  VerticalBox_139 {
   widgetClassPath=&VerticalBox
   parent=VerticalBox_124
   widget=VerticalBox_139
   slot {class=VerticalBoxSlot, properties {padding=<ltrb 0 -30 0 0>, size {sizeRule=Fill, value=1}}}
  }
  TODRange {
   widgetClassPath=&EditorUtilityCheckBox
   bIsVariable=true
   parent=VerticalBox_139
   widget=TODRange
   properties {widgetStyle {
     checkBoxType=CheckBox
     backgroundHoveredImage {
      bIsDynamicallyLoaded=false
      drawAs=RoundedBox
      imageType=NoImage
      mirroring=NoMirror
      resourceName=None
      resourceObject=null
      tiling=NoTile
      imageSize=<18 18>
      margin=<ltrb 0 0 0 0>
      outlineSettings {
       bUseBrushTransparency=false
       roundingType=FixedRadius
       width=1
       color {colorUseRule=UseColor_ColorTable, specifiedColor=<rgba 0 0 0 0>}
       cornerRadii=<3 3 3 3>
      }
      tintColor {colorUseRule=UseColor_Specified, specifiedColor=<rgba 0.0048 0.0048 0.0048 1>}
      uVRegion {bIsValid=false, max=<0 0>, min=<0 0>}
     }
     backgroundImage {
      bIsDynamicallyLoaded=false
      drawAs=RoundedBox
      imageType=NoImage
      mirroring=NoMirror
      resourceName=None
      resourceObject=null
      tiling=NoTile
      imageSize=<18 18>
      margin=<ltrb 0 0 0 0>
      outlineSettings {
       bUseBrushTransparency=false
       roundingType=FixedRadius
       width=1
       color {colorUseRule=UseColor_ColorTable, specifiedColor=<rgba 0 0 0 0>}
       cornerRadii=<3 3 3 3>
      }
      tintColor {colorUseRule=UseColor_Specified, specifiedColor=<rgba 0.0048 0.0048 0.0048 1>}
      uVRegion {bIsValid=false, max=<0 0>, min=<0 0>}
     }
     backgroundPressedImage {
      bIsDynamicallyLoaded=false
      drawAs=RoundedBox
      imageType=NoImage
      mirroring=NoMirror
      resourceName=None
      resourceObject=null
      tiling=NoTile
      imageSize=<18 18>
      margin=<ltrb 0 0 0 0>
      outlineSettings {
       bUseBrushTransparency=false
       roundingType=FixedRadius
       width=1
       color {colorUseRule=UseColor_ColorTable, specifiedColor=<rgba 0 0 0 0>}
       cornerRadii=<3 3 3 3>
      }
      tintColor {colorUseRule=UseColor_Specified, specifiedColor=<rgba 0.0048 0.0048 0.0048 1>}
      uVRegion {bIsValid=false, max=<0 0>, min=<0 0>}
     }
     borderBackgroundColor {colorUseRule=UseColor_Specified, specifiedColor=<rgba 1 1 1 1>}
     checkedForeground {colorUseRule=UseColor_Specified, specifiedColor=<rgba 1 1 1 1>}
     checkedHoveredForeground {colorUseRule=UseColor_Specified, specifiedColor=<rgba 1 1 1 1>}
     checkedHoveredImage {
      bIsDynamicallyLoaded=false
      drawAs=Image
      imageType=Vector
      mirroring=NoMirror
      resourceName=../../../Engine/Content/Slate/Starship/CoreWidgets/CheckBox/check.svg
      resourceObject=null
      tiling=NoTile
      imageSize=<16 16>
      margin=<ltrb 0 0 0 0>
      outlineSettings {
       bUseBrushTransparency=false
       roundingType=HalfHeightRadius
       width=0
       color {colorUseRule=UseColor_Specified, specifiedColor=<rgba 0 0 0 0>}
       cornerRadii=<0 0 0 0>
      }
      tintColor {colorUseRule=UseColor_Specified, specifiedColor=<rgba 0.0044 0.2384 1 1>}
      uVRegion {bIsValid=false, max=<0 0>, min=<0 0>}
     }
     checkedImage {
      bIsDynamicallyLoaded=false
      drawAs=Image
      imageType=Vector
      mirroring=NoMirror
      resourceName=../../../Engine/Content/Slate/Starship/CoreWidgets/CheckBox/check.svg
      resourceObject=null
      tiling=NoTile
      imageSize=<16 16>
      margin=<ltrb 0 0 0 0>
      outlineSettings {
       bUseBrushTransparency=false
       roundingType=HalfHeightRadius
       width=0
       color {colorUseRule=UseColor_Specified, specifiedColor=<rgba 0 0 0 0>}
       cornerRadii=<0 0 0 0>
      }
      tintColor {colorUseRule=UseColor_Specified, specifiedColor=<rgba 0 0.162 0.7454 1>}
      uVRegion {bIsValid=false, max=<0 0>, min=<0 0>}
     }
     checkedPressedForeground {colorUseRule=UseColor_Specified, specifiedColor=<rgba 1 1 1 1>}
     checkedPressedImage {
      bIsDynamicallyLoaded=false
      drawAs=Image
      imageType=Vector
      mirroring=NoMirror
      resourceName=../../../Engine/Content/Slate/Starship/CoreWidgets/CheckBox/check.svg
      resourceObject=null
      tiling=NoTile
      imageSize=<16 16>
      margin=<ltrb 0 0 0 0>
      outlineSettings {
       bUseBrushTransparency=false
       roundingType=HalfHeightRadius
       width=0
       color {colorUseRule=UseColor_Specified, specifiedColor=<rgba 0 0 0 0>}
       cornerRadii=<0 0 0 0>
      }
      tintColor {colorUseRule=UseColor_Specified, specifiedColor=<rgba 0 0.0802 0.3515 1>}
      uVRegion {bIsValid=false, max=<0 0>, min=<0 0>}
     }
     checkedSlateSound {resourceObject=null}
     foregroundColor {colorUseRule=UseColor_Specified, specifiedColor=<rgba 1 1 1 1>}
     hoveredForeground {colorUseRule=UseColor_Specified, specifiedColor=<rgba 1 1 1 1>}
     hoveredSlateSound {resourceObject=null}
     padding=<ltrb 2 0 0 0>
     pressedForeground {colorUseRule=UseColor_Specified, specifiedColor=<rgba 1 1 1 1>}
     uncheckedHoveredImage {
      bIsDynamicallyLoaded=false
      drawAs=NoDrawType
      imageType=NoImage
      mirroring=NoMirror
      resourceName=None
      resourceObject=null
      tiling=NoTile
      imageSize=<0 0>
      margin=<ltrb 0 0 0 0>
      outlineSettings {
       bUseBrushTransparency=false
       roundingType=HalfHeightRadius
       width=0
       color {colorUseRule=UseColor_Specified, specifiedColor=<rgba 0 0 0 0>}
       cornerRadii=<0 0 0 0>
      }
      tintColor {colorUseRule=UseColor_Specified, specifiedColor=<rgba 1 1 1 1>}
      uVRegion {bIsValid=false, max=<0 0>, min=<0 0>}
     }
     uncheckedImage {
      bIsDynamicallyLoaded=false
      drawAs=NoDrawType
      imageType=NoImage
      mirroring=NoMirror
      resourceName=None
      resourceObject=null
      tiling=NoTile
      imageSize=<0 0>
      margin=<ltrb 0 0 0 0>
      outlineSettings {
       bUseBrushTransparency=false
       roundingType=HalfHeightRadius
       width=0
       color {colorUseRule=UseColor_Specified, specifiedColor=<rgba 0 0 0 0>}
       cornerRadii=<0 0 0 0>
      }
      tintColor {colorUseRule=UseColor_Specified, specifiedColor=<rgba 1 1 1 1>}
      uVRegion {bIsValid=false, max=<0 0>, min=<0 0>}
     }
     uncheckedPressedImage {
      bIsDynamicallyLoaded=false
      drawAs=NoDrawType
      imageType=NoImage
      mirroring=NoMirror
      resourceName=None
      resourceObject=null
      tiling=NoTile
      imageSize=<0 0>
      margin=<ltrb 0 0 0 0>
      outlineSettings {
       bUseBrushTransparency=false
       roundingType=HalfHeightRadius
       width=0
       color {colorUseRule=UseColor_Specified, specifiedColor=<rgba 0 0 0 0>}
       cornerRadii=<0 0 0 0>
      }
      tintColor {colorUseRule=UseColor_Specified, specifiedColor=<rgba 1 1 1 1>}
      uVRegion {bIsValid=false, max=<0 0>, min=<0 0>}
     }
     uncheckedSlateSound {resourceObject=null}
     undeterminedForeground {colorUseRule=UseColor_Specified, specifiedColor=<rgba 1 1 1 1>}
     undeterminedHoveredImage {
      bIsDynamicallyLoaded=false
      drawAs=Image
      imageType=Vector
      mirroring=NoMirror
      resourceName=../../../Engine/Content/Slate/Starship/CoreWidgets/CheckBox/indeterminate.svg
      resourceObject=null
      tiling=NoTile
      imageSize=<16 16>
      margin=<ltrb 0 0 0 0>
      outlineSettings {
       bUseBrushTransparency=false
       roundingType=HalfHeightRadius
       width=0
       color {colorUseRule=UseColor_Specified, specifiedColor=<rgba 0 0 0 0>}
       cornerRadii=<0 0 0 0>
      }
      tintColor {colorUseRule=UseColor_Specified, specifiedColor=<rgba 0.0044 0.2384 1 1>}
      uVRegion {bIsValid=false, max=<0 0>, min=<0 0>}
     }
     undeterminedImage {
      bIsDynamicallyLoaded=false
      drawAs=Image
      imageType=Vector
      mirroring=NoMirror
      resourceName=../../../Engine/Content/Slate/Starship/CoreWidgets/CheckBox/indeterminate.svg
      resourceObject=null
      tiling=NoTile
      imageSize=<16 16>
      margin=<ltrb 0 0 0 0>
      outlineSettings {
       bUseBrushTransparency=false
       roundingType=HalfHeightRadius
       width=0
       color {colorUseRule=UseColor_Specified, specifiedColor=<rgba 0 0 0 0>}
       cornerRadii=<0 0 0 0>
      }
      tintColor {colorUseRule=UseColor_Specified, specifiedColor=<rgba 0 0.162 0.7454 1>}
      uVRegion {bIsValid=false, max=<0 0>, min=<0 0>}
     }
     undeterminedPressedImage {
      bIsDynamicallyLoaded=false
      drawAs=Image
      imageType=Vector
      mirroring=NoMirror
      resourceName=../../../Engine/Content/Slate/Starship/CoreWidgets/CheckBox/indeterminate.svg
      resourceObject=null
      tiling=NoTile
      imageSize=<16 16>
      margin=<ltrb 0 0 0 0>
      outlineSettings {
       bUseBrushTransparency=false
       roundingType=HalfHeightRadius
       width=0
       color {colorUseRule=UseColor_Specified, specifiedColor=<rgba 0 0 0 0>}
       cornerRadii=<0 0 0 0>
      }
      tintColor {colorUseRule=UseColor_Specified, specifiedColor=<rgba 0 0.0802 0.3515 1>}
      uVRegion {bIsValid=false, max=<0 0>, min=<0 0>}
     }
    }}
   slot {class=VerticalBoxSlot}
  }
  UEFN_TextBlock_C_16 {
   widgetClassPath=&UEFN_TextBlock
   parent=TODRange
   widget=UEFN_TextBlock_C_16
   properties {
    justification=Center
    text=Use Time Range
    font {
     bForceMonospaced=false
     bMaterialIsStencil=false
     fontMaterial=null
     fontObject=&BurbankSmall
     letterSpacing=0
     monospacedWidth=1
     size=10.5
     skewAmount=0
     typefaceFontName=Black
     outlineSettings {
      bApplyOutlineToDropShadows=false
      bMiteredCorners=false
      bSeparateFillAlpha=false
      outlineMaterial=null
      outlineSize=0
      outlineColor=<rgba 0 0 0 1>
     }
    }
   }
   slot {class=PanelSlot}
  }
  TODRangeBox {
   widgetClassPath=&HorizontalBox
   bIsVariable=true
   parent=VerticalBox_139
   widget=TODRangeBox
   properties {bIsEnabled=false, toolTipText=Range between 0 and 23.59}
   slot {class=VerticalBoxSlot, properties {horizontalAlignment=HAlign_Center, padding=<ltrb 0 5 0 0>}}
  }
  UEFN_TextBlock_C_17 {
   widgetClassPath=&UEFN_TextBlock
   parent=TODRangeBox
   widget=UEFN_TextBlock_C_17
   properties {
    justification=Center
    text=Min
    font {
     bForceMonospaced=false
     bMaterialIsStencil=false
     fontMaterial=null
     fontObject=&BurbankSmall
     letterSpacing=0
     monospacedWidth=1
     size=9
     skewAmount=0
     typefaceFontName=Black
     outlineSettings {
      bApplyOutlineToDropShadows=false
      bMiteredCorners=false
      bSeparateFillAlpha=false
      outlineMaterial=null
      outlineSize=0
      outlineColor=<rgba 0 0 0 1>
     }
    }
   }
   slot {class=HorizontalBoxSlot, properties {verticalAlignment=VAlign_Center}}
  }
  TODMinTime {
   widgetClassPath=/Script/Blutility.EditorUtilitySpinBox
   bIsVariable=true
   parent=TODRangeBox
   widget=TODMinTime
   properties {
    bOverride_MaxSliderValue=true
    bOverride_MaxValue=true
    bOverride_MinSliderValue=true
    bOverride_MinValue=true
    delta=0.1
    justification=Center
    maxFractionalDigits=2
    maxSliderValue=23.99
    maxValue=23.99
    minDesiredWidth=32
    minFractionalDigits=2
    minValue=0.01
    value=0.01
    font {
     bForceMonospaced=false
     bMaterialIsStencil=false
     fontMaterial=null
     fontObject=&Roboto
     letterSpacing=0
     monospacedWidth=1
     size=9
     skewAmount=0
     typefaceFontName=Bold
     outlineSettings {
      bApplyOutlineToDropShadows=false
      bMiteredCorners=false
      bSeparateFillAlpha=false
      outlineMaterial=null
      outlineSize=0
      outlineColor=<rgba 0 0 0 1>
     }
    }
    foregroundColor {colorUseRule=UseColor_Specified, specifiedColor=<rgba 1 1 1 1>}
   }
   slot {class=HorizontalBoxSlot, properties {padding=<ltrb 5 0 5 0>}}
  }
  UEFN_TextBlock_C_18 {
   widgetClassPath=&UEFN_TextBlock
   parent=TODRangeBox
   widget=UEFN_TextBlock_C_18
   properties {
    justification=Center
    text=Max
    font {
     bForceMonospaced=false
     bMaterialIsStencil=false
     fontMaterial=null
     fontObject=&BurbankSmall
     letterSpacing=0
     monospacedWidth=1
     size=9
     skewAmount=0
     typefaceFontName=Black
     outlineSettings {
      bApplyOutlineToDropShadows=false
      bMiteredCorners=false
      bSeparateFillAlpha=false
      outlineMaterial=null
      outlineSize=0
      outlineColor=<rgba 0 0 0 1>
     }
    }
   }
   slot {class=HorizontalBoxSlot, properties {verticalAlignment=VAlign_Center}}
  }
  TODMaxTime {
   widgetClassPath=&EditorUtilitySpinBox
   bIsVariable=true
   parent=TODRangeBox
   widget=TODMaxTime
   properties {
    bOverride_MaxSliderValue=true
    bOverride_MaxValue=true
    bOverride_MinSliderValue=true
    bOverride_MinValue=true
    delta=0.01
    justification=Center
    maxFractionalDigits=2
    maxSliderValue=23.99
    maxValue=23.99
    minDesiredWidth=32
    minFractionalDigits=2
    value=23.59
    font {
     bForceMonospaced=false
     bMaterialIsStencil=false
     fontMaterial=null
     fontObject=&Roboto
     letterSpacing=0
     monospacedWidth=1
     size=9
     skewAmount=0
     typefaceFontName=Bold
     outlineSettings {
      bApplyOutlineToDropShadows=false
      bMiteredCorners=false
      bSeparateFillAlpha=false
      outlineMaterial=null
      outlineSize=0
      outlineColor=<rgba 0 0 0 1>
     }
    }
    foregroundColor {colorUseRule=UseColor_Specified, specifiedColor=<rgba 1 1 1 1>}
   }
   slot {class=HorizontalBoxSlot, properties {padding=<ltrb 5 0 5 0>}}
  }
  getTODActor {
   widgetClassPath=&EditorUtilityButton
   parent=VerticalBox_124
   widget=getTODActor
   slot {class=VerticalBoxSlot, properties {verticalAlignment=VAlign_Top, padding=<ltrb 0 10 0 0>}}
  }
  UEFN_TextBlock_C_9 {
   widgetClassPath=&UEFN_TextBlock
   parent=getTODActor
   widget=UEFN_TextBlock_C_9
   properties {
    text=Get TOD Actor
    colorAndOpacity {colorUseRule=UseColor_Specified, specifiedColor=<rgba 0.3073 0.3073 0.3073 1>}
    font {
     bForceMonospaced=false
     bMaterialIsStencil=false
     fontMaterial=null
     fontObject=&BurbankSmall
     letterSpacing=0
     monospacedWidth=1
     size=10.5
     skewAmount=0
     typefaceFontName=Black
     outlineSettings {
      bApplyOutlineToDropShadows=false
      bMiteredCorners=false
      bSeparateFillAlpha=false
      outlineMaterial=null
      outlineSize=0
      outlineColor=<rgba 0 0 0 1>
     }
    }
   }
   slot {class=ButtonSlot, properties {verticalAlignment=VAlign_Fill, padding=<ltrb 0 0 0 0>}}
  }
  MessageBorder {
   widgetClassPath=&Border
   bIsVariable=true
   parent=VerticalBox_53
   widget=MessageBorder
   properties {
    horizontalAlignment=HAlign_Center
    background {
     bIsDynamicallyLoaded=false
     drawAs=RoundedBox
     imageType=NoImage
     mirroring=NoMirror
     resourceName=None
     resourceObject=null
     tiling=NoTile
     imageSize=<32 32>
     margin=<ltrb 1 1 1 1>
     outlineSettings {
      bUseBrushTransparency=false
      roundingType=HalfHeightRadius
      width=0
      color {colorUseRule=UseColor_Specified, specifiedColor=<rgba 0 0 0 0>}
      cornerRadii=<0 0 0 0>
     }
     tintColor {colorUseRule=UseColor_Specified, specifiedColor=<rgba 1 1 1 1>}
     uVRegion {bIsValid=false, max=<0 0>, min=<0 0>}
    }
    brushColor=<rgba 0.0381 0.0729 0.0417 0.5>
    padding=<ltrb 0 0 0 0>
   }
   slot {class=VerticalBoxSlot, properties {padding=<ltrb 0 10 0 0>, size {sizeRule=Fill, value=1}}}
  }
  MessageBox {
   widgetClassPath=/Script/UMG.RichTextBlock
   bIsVariable=true
   parent=MessageBorder
   widget=MessageBox
   properties {justification=Center, text=Ready..., textStyleSet=&AS_InfoStyle}
   slot {class=BorderSlot, properties {horizontalAlignment=HAlign_Center, padding=<ltrb 0 0 0 0>}}
  }
 ]
}
